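# Material function that builds a light mask for EchoFrost rock pillars in Hera terrain.
# An offset local-Z height drives separate upper and lower fades, with the lower squared, forming a vertical band sharpened by a falloff power.
# That band is multiplied by view-dependent shaping from horizontal camera, pivot-to-pixel and normal directions, slightly damped where normals face the camera directly.
# It outputs the combined mask and the height gradient alone.
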MaterialFunction /Hera/Terrain/Rocks/Materials/EchoFrost/MF_Hera_EchoFrost_PillarLightMask
graph {nodes [
  n1 = FunctionInput() {bUsePreviewValueAsDefault=true, inputName=Light Offset, inputType=FunctionInput_Scalar, previewValue=<-200 -200 -200 -200>}
  n2 = Add(A=LocalPosition().Z, B=@n1)
  n3 = ScalarParameter() {defaultValue=7909.7642, parameterName=Light Upper Fade}
  n4 = Saturate(None=Divide(A=Multiply(A=@n2) {constB=-1}, B=ScalarParameter() {defaultValue=309.0102, parameterName=Light Lower Fade}))
  n5 = Add(A=Saturate(None=Divide(A=@n2, B=@n3)), B=Multiply(A=@n4, B=@n4))
  n6 = Power(Base=OneMinus(None=Saturate(None=@n5)), Exp=ScalarParameter() {defaultValue=30, parameterName=Light Falloff})
  n7 = Normalize(VectorInput=ComponentMask(None=CameraVectorWS()) {g=true, r=true})
  n8 = LocalPosition()
  n9 = Normalize(VectorInput=ComponentMask(None=Transform(None=@n8.XYZ) {transformSourceType=TRANSFORMSOURCE_Instance}) {g=true, r=true})
  n10 = Abs(None=Add(A=DotProduct(A=@n7, B=@n9)) {constB=-0.3})
  n11 = FunctionInput(Preview=VertexNormalWS()) {bUsePreviewValueAsDefault=true, inputName=Normal}
  n12 = DotProduct(A=@n7, B=Normalize(VectorInput=ComponentMask(None=@n11) {g=true, r=true}))
  n13 = Multiply(A=Saturate(None=Power(Base=@n12) {constExponent=16})) {constB=0.3}
  n14 = Multiply(A=Multiply(A=Saturate(None=@n10)) {constB=1.5}, B=Multiply(A=Multiply(A=@n12, B=@n12), B=OneMinus(None=@n13)))
  n15 = Multiply(A=@n14, B=@n14)
  FunctionOutput(None=Saturate(None=Multiply(A=@n6, B=@n15)))
  FunctionOutput(None=Saturate(None=@n6)) {outputName=Gradient Only}
 ]}
