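# Translucent unlit ice material for an EchoFrost trophy display-case exterior, hand-shading its lighting from baked normal and mask textures.
# Color sums curvature tints, fixed-point light, cubemap reflection, specular and a height-masked emissive glow, scaled by time of day.
# Refraction is composited manually by blending normal-offset SceneColor with that color using depth- and height-faded opacity.
# Mobile skips the scene blend; a static switch enables destructible cracks and WPO.
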Material /Hera/Terrain/Rocks/Materials/EchoFrost/M_EchoFrost_Ice_Baked_Trophies_Exterior2
graph {
 properties {
  bAutomaticallySetUsageInEditor=false
  bEnableNewHLSLGeneratorOverridden=true
  bTangentSpaceNormal=false
  blendMode=BLEND_Translucent
  physMaterial=/Game/FortressPhysicalMaterials/Stone
  shadingModel=MSM_Unlit
  usedShadingModels=Unlit
 }
 nodes [
  Constant()
  n1 = TextureSampleParameter2D() {parameterName=Normals, samplerType=SAMPLERTYPE_Normal, texture=/Hera/Sets/EchoFrost/Textures/T_EchoFrost_DisplayCase_A_A_N}
  NamedRerouteDeclaration(None=Transform(None=NamedRerouteDeclaration(None=@n1.RGB) {name=Tangent Normals, nodeColor=<rgba 1 0 0.8203 1>})) {name=Normals, nodeColor=<rgba 0 0.7187 1 1>}
  n2 = ScalarParameter() {defaultValue=10, parameterName=Normal Detail Tiling X}
  n3 = ScalarParameter() {defaultValue=10, parameterName=Normal Detail Tiling Y}
  n4 = TextureSampleParameter2D(UVs=Multiply(A=TextureCoordinate(), B=AppendVector(A=@n2, B=@n3))) {parameterName=Normal Detail, samplerType=SAMPLERTYPE_Normal, texture=/Hera/Sets/EchoFrost/Textures/T_EchoFrost_IceSurface_Flat_A_N}
  n5 = LinearInterpolate(A=Constant3Vector() {constant=<rgba 0 0 1 0>}, B=@n4.RGB, Alpha=ScalarParameter() {defaultValue=0.5, parameterName=Normal Detail Intensity})
  n6 = ComponentMask(None=@n5) {r=true}
  n7 = ComponentMask(None=@n5) {g=true}
  n8 = ComponentMask(None=@n5) {b=true}
  n9 = LinearInterpolate(
   A=Constant3Vector() {constant=<rgba 0 0 1 0>}
   B=MaterialFunctionCall(R=Add(A=@n1.R, B=@n6), G=Add(A=@n1.G, B=@n7), B=Multiply(A=@n1.B, B=@n8)) {materialFunction=&AppendMany}.RGB
   Alpha=ScalarParameter() {defaultValue=0.4, parameterName=Normal Intensity}
  ) {constAlpha=1}
  n10 = MaterialFunctionCall(Material_In=SetMaterialAttributes!(MaterialAttributes, Base Color=NamedRerouteUsage(), Normal=@n9) {attributeSetTypes [#1, #2]}) {materialFunction=&MF_FN_DestructibleDamage}
  n11 = GetMaterialAttributes(None=NamedRerouteDeclaration(None=@n10.Material_Out) {name=Crack Material, nodeColor=<rgba 0 0.9297 1 1>}) {attributeGetTypes [#2]}
  NamedRerouteDeclaration(None=Transform(None=StaticSwitchParameter(True=@n11.Normal, False=@n9) {parameterName=IsDestructible})) {name=Normals With Intensity, nodeColor=<rgba 1 0 0.0234 1>}
  n12 = Saturate(None=Divide(A=NamedRerouteUsage()) {constB=10})
  n13 = Multiply(A=LocalPosition().Z) {constB=-1}
  n14 = ScalarParameter() {defaultValue=150, parameterName=Fade Position}
  n15 = Saturate(None=Divide(A=Add(A=@n13, B=@n14), B=ScalarParameter() {defaultValue=26.2373, parameterName=Base Opacity}))
  NamedRerouteDeclaration(None=Saturate(None=LinearInterpolate(B=@n12, Alpha=@n15) {constA=1})) {name=Depth Fade, nodeColor=<rgba 0.3359 0 1 1>}
  NamedRerouteDeclaration(None=@n10.DamageAnimWPO) {name=WPO, nodeColor=<rgba 1 0 0.6094 1>}
  n16 = Saturate(None=DotProduct(A=NamedRerouteUsage(), B=CameraVectorWS()))
  n17 = Add(A=Multiply(A=OneMinus(None=@n16), B=ScalarParameter() {defaultValue=2, parameterName=Reflection Fresnel Intensity}) {constB=2})
  n18 = ScalarParameter() {defaultValue=0.1, parameterName=Specular Low}
  n19 = ScalarParameter() {defaultValue=1.6, parameterName=Specular High}
  n20 = NamedRerouteUsage()
  n21 = NamedRerouteUsage()
  n22 = Saturate(None=Multiply(A=Multiply(A=@n21, B=@n21), B=ScalarParameter() {defaultValue=20, parameterName=Curvature Specular Power}) {constB=20})
  n23 = Saturate(None=Multiply(A=LinearInterpolate(A=@n18, B=@n19, Alpha=@n20) {constA=0.2, constB=0.6}, B=OneMinus(None=@n22)))
  n24 = TextureSample(UVs=ReflectionVectorWS(CustomWorldNormal=NamedRerouteUsage())) {samplerType=SAMPLERTYPE_LinearColor, texture=/Hera/Sets/EchoFrost/Textures/T_EchoFrost_PillarCubeMap}
  n25 = LocalPosition()
  n26 = LinearInterpolate(
   A=ScalarParameter() {defaultValue=1, parameterName=Reflection Intensity Base}
   B=ScalarParameter() {defaultValue=1, parameterName=Reflection Intensity}
   Alpha=Saturate(None=Subtract(A=@n25.Z) {constB=400})
  )
  NamedRerouteDeclaration(None=Multiply(A=Multiply(A=@n17, B=@n23), B=Multiply(A=@n24.RGB, B=@n26) {constB=0.5})) {name=Reflection, nodeColor=<rgba 0 1 0.8594 1>}
  n27 = Add(A=NamedRerouteUsage(), B=CameraVectorWS())
  n28 = Power(Base=DotProduct(A=Normalize(VectorInput=@n27), B=NamedRerouteUsage())) {constExponent=270}
  NamedRerouteDeclaration(None=Multiply(A=Multiply(A=@n23, B=@n28), B=ScalarParameter() {defaultValue=0.6, parameterName=Specular Intensity}) {constB=1.5}) {name=Specular, nodeColor=<rgba 1 0.5391 0 1>}
  n29 = Add(A=LocalPosition().Z, B=ScalarParameter() {defaultValue=-149.8118, parameterName=Light Position})
  n30 = ScalarParameter() {defaultValue=125, parameterName=BaseColor Opacity}
  n31 = OneMinus(None=Saturate(None=Divide(A=@n29, B=@n30)))
  n32 = NamedRerouteUsage()
  n33 = Multiply(A=Multiply(A=@n32, B=@n32), B=ScalarParameter() {defaultValue=7, parameterName=Curvature Opacity Power}) {constB=7}
  n34 = Saturate(None=Add(A=Multiply(A=@n31, B=@n31), B=Saturate(None=@n33)))
  n35 = LinearInterpolate(
   A=ScalarParameter() {defaultValue=0.7, parameterName=Opacity1}
   B=ScalarParameter() {defaultValue=1.6, parameterName=Opacity2}
   Alpha=NamedRerouteDeclaration(None=@n34) {name=Light Mask, nodeColor=<rgba 0.6172 0 1 1>}
  )
  n36 = Divide(A=NamedRerouteUsage(), B=ScalarParameter() {defaultValue=40, parameterName=Depth Fade}) {constB=150}
  n37 = LinearInterpolate(B=Saturate(None=@n36), Alpha=@n15) {constA=1}
  NamedRerouteDeclaration(None=Multiply(A=Saturate(None=@n35), B=Saturate(None=@n37))) {name=Opacity, nodeColor=<rgba 1 0 0.1641 1>}
  n38 = NamedRerouteUsage()
  NamedRerouteDeclaration(None=Multiply(A=ComponentMask(None=@n38) {g=true, r=true}, B=ScalarParameter() {defaultValue=0.035, parameterName=Refraction Intensity}) {constB=0.05}) {name=Refraction, nodeColor=<rgba 0 1 0.8594 1>}
  n39 = TextureSampleParameter2D() {parameterName=SpecularMasks, samplerType=SAMPLERTYPE_Masks, texture=/Hera/Sets/EchoFrost/Textures/T_EchoFrost_DisplayCase_A_A_M}
  NamedRerouteDeclaration(None=StaticSwitchParameter(True=OneMinus(None=@n39.B), False=@n39.B) {parameterName=InvertThickness}) {name=Thickness, nodeColor=<rgba 1 0 0.5156 1>}
  NamedRerouteDeclaration(None=@n39.G) {name=Curvature, nodeColor=<rgba 0.2422 0 1 1>}
  NamedRerouteDeclaration(None=@n39.R) {name=Specular Mask, nodeColor=<rgba 0 1 0.1562 1>}
  Reroute(None=@n34)
  n40 = Subtract(A=SceneDepth!(OffsetFraction=NamedRerouteUsage()) {inputMode=OffsetFraction}.R, B=PixelDepth().R)
  Reroute(None=@n40)
  Reroute(None=@n36)
  Reroute(None=@n10.Material_Out)
  n41 = NamedRerouteUsage()
  n42 = NamedRerouteUsage()
  n43 = Multiply(A=@n42, B=@n42)
  n44 = ScalarParameter() {defaultValue=8, parameterName=Curvature BaseColor Power}
  n45 = LinearInterpolate(
   A=LinearInterpolate(
    A=VectorParameter() {parameterName=Color 1, defaultValue=<rgba 0 0.0086 0.026 1>}.RGB
    B=VectorParameter() {parameterName=Color 2, defaultValue=<rgba 0.1586 0.2261 0.3542 1>}.RGB
    Alpha=Multiply(A=@n43, B=@n44) {constB=8}
   )
   B=VectorParameter() {parameterName=Color Curvature Highlights, defaultValue=<rgba 0.068 0.089 0.1406 1>}.RGB
   Alpha=NamedRerouteUsage()
  )
  n46 = Normalize(VectorInput=Subtract(A=Constant3Vector() {constant=<rgba 5620 -50980 9070 0>}, B=ActorPositionWS()))
  n47 = Multiply(A=DotProduct(A=NamedRerouteDeclaration(None=@n46) {name=Light Vector, nodeColor=<rgba 0 1 0.7891 1>}, B=NamedRerouteUsage())) {constB=0.5}
  n48 = Multiply(A=Power(Base=Add(A=@n47) {constB=0.5}), B=VectorParameter() {parameterName=Color Light, defaultValue=<rgba 0.022 0.0815 0.2031 1>}.RGB)
  n49 = Add(A=Add(A=Add(A=@n45, B=@n48), B=NamedRerouteUsage()), B=NamedRerouteUsage())
  n50 = VectorParameter() {parameterName=Color Contact Shadow, defaultValue=<rgba 0.0192 0.1638 0.2969 1>}
  n51 = Add(A=Divide(A=LocalPosition().Z, B=ScalarParameter() {defaultValue=1556.8625, parameterName=Color Contact Bounds})) {constB=-0.47}
  n52 = Power(Base=Multiply(A=Abs(None=@n51)) {constB=2})
  n53 = MaterialFunctionCall(
   Color_Morning=VectorParameter() {parameterName=Multiplier Morning, defaultValue=<rgba 1.1589 0.7949 0.6597 1>}.RGB
   Color_Day=VectorParameter() {parameterName=Multiplier Day, defaultValue=<rgba 1 1 1 1>}.RGB
   Color_Evening=VectorParameter() {parameterName=Multiplier Evening, defaultValue=<rgba 1 0.7054 0.6094 1>}.RGB
   Color_Night=VectorParameter() {parameterName=Multiplier Night, defaultValue=<rgba 0.8177 0.8177 0.8177 1>}.RGB
  ) {materialFunction=&Color_Time_of_Day}
  n54 = MaterialFunctionCall(
   Color_Morning=VectorParameter() {parameterName=Mobile Multiplier Morning, defaultValue=<rgba 2 1.0809 0.7396 1>}.RGB
   Color_Day=VectorParameter() {parameterName=Mobile Multiplier Day, defaultValue=<rgba 1.7917 1.7917 1.7917 1>}.RGB
   Color_Evening=VectorParameter() {parameterName=Mobile Multiplier Evening, defaultValue=<rgba 2 1.1486 0.8021 1>}.RGB
   Color_Night=VectorParameter() {parameterName=Mobile Multiplier Night, defaultValue=<rgba 1.4 1.4 1.4 1>}.RGB
  ) {materialFunction=&Color_Time_of_Day}
  n55 = Multiply(
   A=LinearInterpolate(A=@n49, B=Multiply(A=@n49, B=@n50.RGB) {constB=0}, Alpha=Saturate(None=@n52))
   B=ShadingPathSwitch(Default=@n53.Result, Mobile=QualitySwitch(Default=@n53.Result, Low=@n54.Result))
  )
  n56 = Constant2Vector()
  n57 = ScalarParameter() {defaultValue=-149.8118, parameterName=Light Position}
  n58 = ScalarParameter() {defaultValue=0.1223, parameterName=Normal Light Intensity Sides}
  n59 = Multiply(
   A=Add(A=LocalPosition().XYZ, B=AppendVector(A=@n56, B=@n57))
   B=AppendVector(A=AppendVector(A=@n58, B=@n58), B=ScalarParameter() {defaultValue=1, parameterName=Normal Light Intensity Up})
  )
  n60 = DotProduct(
   A=LinearInterpolate(A=NamedRerouteUsage(), B=VertexNormalWS()) {constAlpha=-2}
   B=Transform(None=Multiply(A=@n59) {constB=0.01}) {transformSourceType=TRANSFORMSOURCE_Local}
  )
  n61 = MaterialFunctionCall!(Light Offset=ScalarParameter() {defaultValue=-149.8118, parameterName=Light Position}, Normal=NamedRerouteUsage()) {materialFunction=&/MF_Hera_EchoFrost_PillarLightMask}
  n62 = Multiply(A=@n61.Result, B=@n61.Result)
  n63 = LinearInterpolate(
   A=VectorParameter() {parameterName=Emissive Edges, defaultValue=<rgba 0.055 0.0082 0.0677 1>}.RGB
   B=VectorParameter() {parameterName=Emissive Center, defaultValue=<rgba 0.5625 0.1787 0.2499 1>}.RGB
   Alpha=Multiply(A=@n62, B=@n62)
  )
  n64 = Multiply(A=Multiply(A=Max(A=@n60) {constB=0}, B=@n63), B=Multiply(A=@n61.Gradient Only, B=@n61.Gradient Only))
  n65 = Multiply(A=@n61.Result, B=@n63)
  n66 = LinearInterpolate(
   A=ScalarParameter() {parameterName=BaseColor Depth Min}
   B=ScalarParameter() {defaultValue=1, parameterName=BaseColor Depth Max}
   Alpha=NamedRerouteUsage()
  )
  n67 = Divide(
   A=NamedRerouteDeclaration(None=@n40) {name=Depth Comparison, nodeColor=<rgba 0 1 0.2969 1>}
   B=ScalarParameter() {defaultValue=150, parameterName=Light Depth}
  ) {constB=150}
  n68 = Multiply(A=Multiply(A=Add(A=@n64, B=@n65)) {constB=7}, B=LinearInterpolate(A=Saturate(None=@n66), Alpha=Saturate(None=@n67))) {constB=0.5}
  n69 = StaticSwitchParameter(
   True=GetMaterialAttributes(None=NamedRerouteUsage()) {attributeGetTypes [#1]}.Base Color
   False=NamedRerouteDeclaration(None=Add(A=@n55, B=@n68)) {name=Final Color, nodeColor=<rgba 1 0 0.2812 1>}
  ) {parameterName=IsDestructible}
  n70 = NamedRerouteUsage()
  n71 = LinearInterpolate(
   A=ComponentMask(None=SceneColor!(OffsetFraction=@n41) {inputMode=OffsetFraction}.RGB) {b=true, g=true, r=true}
   B=EyeAdaptationInverse(LightValueInput=@n69, AlphaInput=Constant() {r=0.7}).EyeAdaptationInverse
   Alpha=@n70
  ) {constAlpha=1}
  n72 = Multiply(A=@n71) {constB=1.5}
  ShadingPathSwitch(Default=QualitySwitch(Default=@n71, Low=@n72, High=@n71, Medium=@n72, Epic=@n71), Mobile=Multiply(A=@n69) {constB=1.5})
  ShadingPathSwitch(Default=NamedRerouteUsage(), Mobile=@n70)
  StaticSwitchParameter(True=NamedRerouteUsage(), False=Constant()) {parameterName=IsDestructible}
 ]
}
