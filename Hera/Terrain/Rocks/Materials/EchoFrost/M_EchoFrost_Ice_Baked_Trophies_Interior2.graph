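# Unlit material for the EchoFrost ice rock set, named as a trophy interior variant, that produces a flat color shifting with time of day.
# Color_Time_of_Day blends morning, day, evening and night emissive color parameters, pale mauve at dawn and dusk and dark blue otherwise.
# EyeAdaptationInverse partly offsets auto exposure on that color.
# An MF_FN_DestructibleDamage call is also present, and the physical material is stone.
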Material /Hera/Terrain/Rocks/Materials/EchoFrost/M_EchoFrost_Ice_Baked_Trophies_Interior2
graph {
 properties {
  bAutomaticallySetUsageInEditor=false
  bTangentSpaceNormal=false
  physMaterial=/Game/FortressPhysicalMaterials/Stone
  shadingModel=MSM_Unlit
  usedShadingModels=Unlit
 }
 nodes [
  EyeAdaptationInverse(
   LightValueInput=MaterialFunctionCall(
    Color_Morning=VectorParameter() {parameterName=Emissive Morning, defaultValue=<rgba 0.0734 0.051 0.099 1>}.RGB
    Color_Day=VectorParameter() {parameterName=Emissive Day, defaultValue=<rgba 0.0042 0.0244 0.0573 1>}.RGB
    Color_Evening=VectorParameter() {parameterName=Emissive Evening, defaultValue=<rgba 0.0769 0.0672 0.0938 1>}.RGB
    Color_Night=VectorParameter() {parameterName=Emissive Night, defaultValue=<rgba 0.0035 0.02 0.0469 1>}.RGB
   ) {materialFunction=&Color_Time_of_Day}.Result
   AlphaInput=Constant() {r=0.75}
  )
  MaterialFunctionCall() {materialFunction=&MF_FN_DestructibleDamage}
 ]
}
