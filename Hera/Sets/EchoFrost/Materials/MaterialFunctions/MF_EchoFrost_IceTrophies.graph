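# EchoFrost ice material function for trophies, layering an ice look onto incoming Material Attributes.
# It takes baked curvature, baked AO and painted mask inputs alongside world-space axis projections and local-bounds gradients.
# Bump-offset parallax and camera-vector math support internal layers such as air bubbles.
# Sky-light N·L and N·H terms supply custom lighting, with sine-hash noise and 4x4 Bayer screen dithering also in the graph.
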MaterialFunction /Hera/Sets/EchoFrost/Materials/MaterialFunctions/MF_EchoFrost_IceTrophies
graph {expressions {
  MaterialExpressionAbs_2 {class=MaterialExpressionAbs, inputs [None {from=MaterialExpressionDotProduct_7, output=""}], properties {}}
  MaterialExpressionAbs_5 {class=MaterialExpressionAbs, inputs [None {from=MaterialExpressionVertexNormalWS_0, output=""}], properties {}}
  MaterialExpressionActorPositionWS_1 {class=MaterialExpressionActorPositionWS, inputs [], properties {}}
  MaterialExpressionAdd_0 {
   class=MaterialExpressionAdd
   inputs [
    A {from=MaterialExpressionLinearInterpolate_6}
    B {from=MaterialExpressionSkyAtmosphereLightDirection_0}
   ]
  }
  MaterialExpressionAdd_1 {
   class=MaterialExpressionAdd
   inputs [
    A {from=MaterialExpressionConvert_4}
    B {from=MaterialExpressionConvert_4}
   ]
  }
  MaterialExpressionAdd_11 {
   class=MaterialExpressionAdd
   inputs [
    A {from=MaterialExpressionAdd_1}
    B {from=MaterialExpressionConvert_4}
   ]
  }
  MaterialExpressionAdd_12 {
   class=MaterialExpressionAdd
   inputs [
    A {from=MaterialExpressionDivide_4}
    B {from=MaterialExpressionMultiply_24}
   ]
  }
  MaterialExpressionAdd_13 {
   class=MaterialExpressionAdd
   inputs [
    A {from=MaterialExpressionDivide_6}
    B {from=MaterialExpressionMultiply_24}
   ]
  }
  MaterialExpressionAdd_14 {
   class=MaterialExpressionAdd
   inputs [
    A {from=MaterialExpressionNamedRerouteUsage_69}
    B {from=MaterialExpressionStaticSwitchParameter_6}
   ]
  }
  MaterialExpressionAdd_15 {
   class=MaterialExpressionAdd
   inputs [
    A {from=MaterialExpressionNamedRerouteUsage_69}
    B {from=MaterialExpressionStaticSwitchParameter_5}
   ]
  }
  MaterialExpressionAdd_16 {
   class=MaterialExpressionAdd
   inputs [
    A {from=MaterialExpressionNamedRerouteUsage_71}
    B {from=MaterialExpressionReroute_51}
   ]
  }
  MaterialExpressionAdd_17 {
   class=MaterialExpressionAdd
   inputs [
    A {from=MaterialExpressionReroute_20}
    B {from=MaterialExpressionVectorParameter_10, output=RGB}
   ]
  }
  MaterialExpressionAdd_18 {
   class=MaterialExpressionAdd
   inputs [
    A {from=MaterialExpressionReroute_39}
    B {from=MaterialExpressionMultiply_14}
   ]
   properties {constB=0}
  }
  MaterialExpressionAdd_19 {
   class=MaterialExpressionAdd
   inputs [
    A {from=MaterialExpressionTextureSample_12, output=R}
    B {from=MaterialExpressionTextureSample_11, output=G}
   ]
  }
  MaterialExpressionAdd_20 {
   class=MaterialExpressionAdd
   inputs [
    A {from=MaterialExpressionAdd_19}
    B {from=MaterialExpressionTextureSample_9, output=B}
   ]
  }
  MaterialExpressionAdd_21 {
   class=MaterialExpressionAdd
   inputs [
    A {from=MaterialExpressionNamedRerouteUsage_0}
    B {from=MaterialExpressionReroute_58}
   ]
  }
  MaterialExpressionAdd_22 {
   class=MaterialExpressionAdd
   inputs [
    A {from=MaterialExpressionNamedRerouteUsage_71}
    B {from=MaterialExpressionReroute_47}
   ]
  }
  MaterialExpressionAdd_23 {
   class=MaterialExpressionAdd
   inputs [
    A {from=MaterialExpressionDivide_8}
    B {from=MaterialExpressionIf_1}
   ]
  }
  MaterialExpressionAdd_25 {
   class=MaterialExpressionAdd
   inputs [
    A {from=MaterialExpressionReroute_19}
    B {from=MaterialExpressionNamedRerouteUsage_36}
   ]
  }
  MaterialExpressionAdd_26 {
   class=MaterialExpressionAdd
   inputs [
    A {from=MaterialExpressionAdd_25}
    B {from=MaterialExpressionAdd_25}
   ]
  }
  MaterialExpressionAdd_28 {
   class=MaterialExpressionAdd
   inputs [
    A {from=MaterialExpressionAdd_26}
    B {from=MaterialExpressionReroute_44}
   ]
  }
  MaterialExpressionAdd_29 {
   class=MaterialExpressionAdd
   inputs [
    A {from=MaterialExpressionNamedRerouteUsage_3}
    B {from=MaterialExpressionReroute_61}
   ]
  }
  MaterialExpressionAdd_30 {
   class=MaterialExpressionAdd
   inputs [
    A {from=MaterialExpressionNamedRerouteUsage_67}
    B {from=MaterialExpressionScalarParameter_85}
   ]
  }
  MaterialExpressionAdd_31 {
   class=MaterialExpressionAdd
   inputs [
    A {from=MaterialExpressionScalarParameter_1}
    B {from=MaterialExpressionScalarParameter_16}
   ]
   properties {constB=0}
  }
  MaterialExpressionAdd_32 {
   class=MaterialExpressionAdd
   inputs [
    A {from=MaterialExpressionScreenPosition_3, output=PixelPosition}
    B {from=MaterialExpressionViewProperty_0, output=Property}
   ]
  }
  MaterialExpressionAdd_33 {
   class=MaterialExpressionAdd
   inputs [
    A {from=MaterialExpressionMultiply_15}
    B {from=MaterialExpressionCustom_1}
   ]
   properties {constB=0.5}
  }
  MaterialExpressionAdd_34 {
   class=MaterialExpressionAdd
   inputs [
    A {from=MaterialExpressionSmoothStep_0}
    B {from=MaterialExpressionScalarParameter_78}
   ]
  }
  MaterialExpressionAdd_35 {
   class=MaterialExpressionAdd
   inputs [
    A {from=MaterialExpressionMultiply_6}
    B {from=""}
   ]
   properties {constB=-0.5}
  }
  MaterialExpressionAdd_36 {
   class=MaterialExpressionAdd
   inputs [
    A {from=MaterialExpressionMultiply_11}
    B {from=MaterialExpressionMultiply_13}
   ]
  }
  MaterialExpressionAdd_37 {
   class=MaterialExpressionAdd
   inputs [
    A {from=MaterialExpressionMultiply_71}
    B {from=MaterialExpressionMultiply_70}
   ]
  }
  MaterialExpressionAdd_38 {
   class=MaterialExpressionAdd
   inputs [
    A {from=MaterialExpressionAdd_37}
    B {from=MaterialExpressionMultiply_68}
   ]
  }
  MaterialExpressionAdd_39 {
   class=MaterialExpressionAdd
   inputs [
    A {from=MaterialExpressionDivide_11}
    B {from=MaterialExpressionMultiply_65}
   ]
  }
  MaterialExpressionAdd_40 {
   class=MaterialExpressionAdd
   inputs [
    A {from=MaterialExpressionReroute_40}
    B {from=MaterialExpressionMultiply_72}
   ]
  }
  MaterialExpressionAdd_41 {
   class=MaterialExpressionAdd
   inputs [
    A {from=MaterialExpressionPower_9}
    B {from=MaterialExpressionScalarParameter_64}
   ]
   properties {constB=-0.2}
  }
  MaterialExpressionAdd_42 {
   class=MaterialExpressionAdd
   inputs [
    A {from=MaterialExpressionSubtract_8}
    B {from=MaterialExpressionScalarParameter_84}
   ]
  }
  MaterialExpressionAdd_43 {
   class=MaterialExpressionAdd
   inputs [
    A {from=MaterialExpressionSaturate_14}
    B {from=MaterialExpressionScalarParameter_74}
   ]
  }
  MaterialExpressionAdd_45 {
   class=MaterialExpressionAdd
   inputs [
    A {from=MaterialExpressionMultiply_77}
    B {from=MaterialExpressionReroute_72}
   ]
  }
  MaterialExpressionAdd_47 {
   class=MaterialExpressionAdd
   inputs [
    A {from=MaterialExpressionMultiply_61}
    B {from=MaterialExpressionReroute_57}
   ]
  }
  MaterialExpressionAdd_6 {
   class=MaterialExpressionAdd
   inputs [
    A {from=MaterialExpressionComponentMask_22}
    B {from=MaterialExpressionAppendVector_4}
   ]
   properties {constB=-0.5}
  }
  MaterialExpressionAdd_7 {
   class=MaterialExpressionAdd
   inputs [
    A {from=MaterialExpressionMultiply_18}
    B {from=MaterialExpressionReroute_48}
   ]
  }
  MaterialExpressionAdd_8 {
   class=MaterialExpressionAdd
   inputs [
    A {from=MaterialExpressionAppendVector_8}
    B {from=MaterialExpressionScalarParameter_102}
   ]
   properties {constB=0.2}
  }
  MaterialExpressionAppendVector_0 {
   class=MaterialExpressionAppendVector
   inputs [
    A {from=MaterialExpressionFrac_3, output=""}
    B {from=MaterialExpressionFrac_2, output=""}
   ]
  }
  MaterialExpressionAppendVector_1 {
   class=MaterialExpressionAppendVector
   inputs [
    A {from=MaterialExpressionConstant_4, output=""}
    B {from=MaterialExpressionMultiply_48, output=""}
   ]
  }
  MaterialExpressionAppendVector_10 {
   class=MaterialExpressionAppendVector
   inputs [
    A {from=MaterialExpressionOneMinus_9, output=""}
    B {from=MaterialExpressionOneMinus_3, output=""}
   ]
  }
  MaterialExpressionAppendVector_2 {
   class=MaterialExpressionAppendVector
   inputs [
    A {from=MaterialExpressionConvert_6, output=""}
    B {from=MaterialExpressionOneMinus_3, output=""}
   ]
  }
  MaterialExpressionAppendVector_3 {
   class=MaterialExpressionAppendVector
   inputs [
    A {from=MaterialExpressionScalarParameter_62, output=""}
    B {from=MaterialExpressionScalarParameter_73, output=""}
   ]
  }
  MaterialExpressionAppendVector_4 {
   class=MaterialExpressionAppendVector
   inputs [
    A {from=MaterialExpressionComponentMask_30, output=""}
    B {from=MaterialExpressionMultiply_9, output=""}
   ]
  }
  MaterialExpressionAppendVector_5 {
   class=MaterialExpressionAppendVector
   inputs [
    A {from=MaterialExpressionConvert_6, output=""}
    B {from=MaterialExpressionOneMinus_3, output=""}
   ]
  }
  MaterialExpressionAppendVector_6 {
   class=MaterialExpressionAppendVector
   inputs [
    A {from=MaterialExpressionOneMinus_11, output=""}
    B {from=MaterialExpressionOneMinus_3, output=""}
   ]
  }
  MaterialExpressionAppendVector_7 {
   class=MaterialExpressionAppendVector
   inputs [
    A {from=MaterialExpressionSubtract_2, output=""}
    B {from=MaterialExpressionSubtract_9, output=""}
   ]
  }
  MaterialExpressionAppendVector_8 {
   class=MaterialExpressionAppendVector
   inputs [
    A {from=MaterialExpressionAppendVector_7, output=""}
    B {from=MaterialExpressionSubtract_10, output=""}
   ]
  }
  MaterialExpressionAppendVector_9 {
   class=MaterialExpressionAppendVector
   inputs [
    A {from=MaterialExpressionRound_1, output=""}
    B {from=MaterialExpressionOneMinus_10, output=""}
   ]
  }
  MaterialExpressionBumpOffset_0 {
   class=MaterialExpressionBumpOffset
   inputs [
    Coordinate {from=MaterialExpressionReroute_35, output=""}
    Height {from=MaterialExpressionNamedRerouteUsage_31, output=""}
    HeightRatioInput {from=MaterialExpressionScalarParameter_13, output=""}
   ]
   properties {heightRatio=0.1}
  }
  MaterialExpressionCameraVectorWS_0 {class=MaterialExpressionCameraVectorWS, inputs [], properties {}}
  MaterialExpressionCameraVectorWS_1 {class=MaterialExpressionCameraVectorWS, inputs [], properties {}}
  MaterialExpressionCameraVectorWS_2 {class=MaterialExpressionCameraVectorWS, inputs [], properties {}}
  MaterialExpressionCameraVectorWS_3 {class=MaterialExpressionCameraVectorWS, inputs [], properties {}}
  MaterialExpressionCameraVectorWS_6 {class=MaterialExpressionCameraVectorWS, inputs [], properties {}}
  MaterialExpressionChannelMaskParameter_0 {
   class=MaterialExpressionChannelMaskParameter
   inputs [None {from=MaterialExpressionTextureSample_5, output=RGB}]
   properties {
    group="-       EchoFrost Ice"
    maskChannel=Green
    parameterName=* Internal Layers - Air Bubbles - Channel
    sortPriority=200
    defaultValue=<rgba 0 1 0 0>
   }
  }
  MaterialExpressionComponentMask_0 {class=MaterialExpressionComponentMask, inputs [None {from=MaterialExpressionNamedRerouteUsage_46}], properties {r=true}}
  MaterialExpressionComponentMask_1 {class=MaterialExpressionComponentMask, inputs [None {from=MaterialExpressionDivide_5}], properties {b=true}}
  MaterialExpressionComponentMask_10 {class=MaterialExpressionComponentMask, inputs [None {from=MaterialExpressionReroute_50}], properties {g=true, r=true}}
  MaterialExpressionComponentMask_11 {class=MaterialExpressionComponentMask, inputs [None {from=MaterialExpressionReroute_49}], properties {r=true}}
  MaterialExpressionComponentMask_12 {class=MaterialExpressionComponentMask, inputs [None {from=MaterialExpressionTransform_4}], properties {b=true}}
  MaterialExpressionComponentMask_13 {class=MaterialExpressionComponentMask, inputs [None {from=MaterialExpressionReroute_49}], properties {g=true}}
  MaterialExpressionComponentMask_15 {class=MaterialExpressionComponentMask, inputs [None {from=MaterialExpressionReroute_49}], properties {b=true}}
  MaterialExpressionComponentMask_16 {class=MaterialExpressionComponentMask, inputs [None {from=MaterialExpressionReroute_50}], properties {b=true, r=true}}
  MaterialExpressionComponentMask_2 {class=MaterialExpressionComponentMask, inputs [None {from=MaterialExpressionConstantBiasScale_1}], properties {g=true, r=true}}
  MaterialExpressionComponentMask_20 {class=MaterialExpressionComponentMask, inputs [None {from=MaterialExpressionReroute_50}], properties {b=true, g=true}}
  MaterialExpressionComponentMask_22 {class=MaterialExpressionComponentMask, inputs [None {from=MaterialExpressionDivide_2}], properties {g=true, r=true}}
  MaterialExpressionComponentMask_23 {class=MaterialExpressionComponentMask, inputs [None {from=MaterialExpressionRound_1}], properties {r=true}}
  MaterialExpressionComponentMask_24 {class=MaterialExpressionComponentMask, inputs [None {from=MaterialExpressionRound_1}], properties {g=true}}
  MaterialExpressionComponentMask_25 {class=MaterialExpressionComponentMask, inputs [None {from=MaterialExpressionDivide_7}], properties {g=true, r=true}}
  MaterialExpressionComponentMask_26 {class=MaterialExpressionComponentMask, inputs [None {from=MaterialExpressionDivide_7}], properties {b=true, r=true}}
  MaterialExpressionComponentMask_27 {class=MaterialExpressionComponentMask, inputs [None {from=MaterialExpressionDivide_7}], properties {b=true, g=true}}
  MaterialExpressionComponentMask_28 {class=MaterialExpressionComponentMask, inputs [None {from=MaterialExpressionNamedRerouteUsage_105}], properties {g=true}}
  MaterialExpressionComponentMask_29 {class=MaterialExpressionComponentMask, inputs [None {from=MaterialExpressionReroute_60}], properties {r=true}}
  MaterialExpressionComponentMask_3 {class=MaterialExpressionComponentMask, inputs [None {from=MaterialExpressionTextureSample_1, output=RGB}], properties {g=true, r=true}}
  MaterialExpressionComponentMask_30 {class=MaterialExpressionComponentMask, inputs [None {from=MaterialExpressionAdd_35}], properties {r=true}}
  MaterialExpressionComponentMask_31 {class=MaterialExpressionComponentMask, inputs [None {from=MaterialExpressionAdd_32}], properties {r=true}}
  MaterialExpressionComponentMask_32 {class=MaterialExpressionComponentMask, inputs [None {from=MaterialExpressionAdd_32}], properties {g=true}}
  MaterialExpressionComponentMask_33 {class=MaterialExpressionComponentMask, inputs [None {from=MaterialExpressionAdd_39}], properties {g=true, r=true}}
  MaterialExpressionComponentMask_34 {class=MaterialExpressionComponentMask, inputs [None {from=MaterialExpressionAdd_39}], properties {b=true, r=true}}
  MaterialExpressionComponentMask_35 {class=MaterialExpressionComponentMask, inputs [None {from=MaterialExpressionAdd_39}], properties {b=true, g=true}}
  MaterialExpressionComponentMask_36 {class=MaterialExpressionComponentMask, inputs [None {from=MaterialExpressionNamedRerouteUsage_106}], properties {g=true}}
  MaterialExpressionComponentMask_37 {class=MaterialExpressionComponentMask, inputs [None {from=MaterialExpressionReroute_69}], properties {r=true}}
  MaterialExpressionComponentMask_4 {class=MaterialExpressionComponentMask, inputs [None {from=MaterialExpressionNamedRerouteUsage_54}], properties {r=true}}
  MaterialExpressionComponentMask_7 {class=MaterialExpressionComponentMask, inputs [None {from=MaterialExpressionVertexNormalWS_3}], properties {g=true, r=true}}
  MaterialExpressionComponentMask_9 {class=MaterialExpressionComponentMask, inputs [None {from=MaterialExpressionDivide_17}], properties {g=true, r=true}}
  MaterialExpressionConstant2Vector_0 {class=MaterialExpressionConstant2Vector, properties {g=0.5, r=0.5}}
  MaterialExpressionConstant3Vector_0 {class=MaterialExpressionConstant3Vector, properties {constant=<rgba 1 1 -1 0>}}
  MaterialExpressionConstant3Vector_1 {class=MaterialExpressionConstant3Vector, properties {constant=<rgba 0 0 1 0>}}
  MaterialExpressionConstant3Vector_2 {class=MaterialExpressionConstant3Vector, properties {constant=<rgba 0 1 0 0>}}
  MaterialExpressionConstant4Vector_0 {class=MaterialExpressionConstant4Vector, inputs [], properties {constant=<rgba 4.6532 5.756 0 0>}}
  MaterialExpressionConstant4Vector_1 {class=MaterialExpressionConstant4Vector, inputs [], properties {constant=<rgba 9.654 9.256 0 0>}}
  MaterialExpressionConstantBiasScale_0 {class=MaterialExpressionConstantBiasScale, inputs [None {from=MaterialExpressionComponentMask_3}], properties {bias=-0.5, scale=2}}
  MaterialExpressionConstantBiasScale_1 {class=MaterialExpressionConstantBiasScale, inputs [None {from=MaterialExpressionTransform_1}]}
  MaterialExpressionConstant_0 {class=MaterialExpressionConstant, properties {r=0.2}}
  MaterialExpressionConstant_1 {class=MaterialExpressionConstant, properties {r=2}}
  MaterialExpressionConstant_2 {class=MaterialExpressionConstant}
  MaterialExpressionConstant_3 {class=MaterialExpressionConstant}
  MaterialExpressionConstant_4 {class=MaterialExpressionConstant}
  MaterialExpressionConstant_5 {class=MaterialExpressionConstant}
  MaterialExpressionConstant_7 {class=MaterialExpressionConstant, properties {r=0.25}}
  MaterialExpressionConstant_9 {class=MaterialExpressionConstant}
  MaterialExpressionConvert_0 {
   class=MaterialExpressionConvert
   inputs [
    None {from=MaterialExpressionNamedRerouteUsage_75}
    None {from=MaterialExpressionNamedRerouteUsage_74}
    None {from=MaterialExpressionNamedRerouteUsage_76}
   ]
   properties {
    nodeName=Make Vector3
    convertInputs [
     {
      type=Scalar
      expressionInput {
       expression=&self:MaterialExpressionNamedRerouteUsage_75
       inputName=None
       mask=0
       maskA=0
       maskB=0
       maskG=0
       maskR=0
       outputIndex=0
      }
     }
     {
      type=Scalar
      expressionInput {
       expression=&self:MaterialExpressionNamedRerouteUsage_74
       inputName=None
       mask=0
       maskA=0
       maskB=0
       maskG=0
       maskR=0
       outputIndex=0
      }
     }
     {
      type=Scalar
      expressionInput {
       expression=&self:MaterialExpressionNamedRerouteUsage_76
       inputName=None
       mask=0
       maskA=0
       maskB=0
       maskG=0
       maskR=0
       outputIndex=0
      }
     }
    ]
    convertOutputs [{type=Vector3}]
   }
  }
  MaterialExpressionConvert_1 {
   class=MaterialExpressionConvert
   inputs [
    None {from=MaterialExpressionVertexColor_0, output=R}
    None {from=MaterialExpressionVertexColor_0, output=R}
    None {from=MaterialExpressionVertexColor_0, output=R}
   ]
   properties {
    nodeName=Make Vector3
    convertInputs [
     {
      type=Scalar
      expressionInput {
       expression=&self:MaterialExpressionVertexColor_0
       inputName=None
       mask=1
       maskA=0
       maskB=0
       maskG=0
       maskR=1
       outputIndex=1
      }
     }
     {
      type=Scalar
      expressionInput {
       expression=&self:MaterialExpressionVertexColor_0
       inputName=None
       mask=1
       maskA=0
       maskB=0
       maskG=1
       maskR=0
       outputIndex=2
      }
     }
     {
      type=Scalar
      expressionInput {
       expression=&self:MaterialExpressionVertexColor_0
       inputName=None
       mask=1
       maskA=0
       maskB=1
       maskG=0
       maskR=0
       outputIndex=3
      }
     }
    ]
    convertOutputs [{type=Vector3}]
   }
  }
  MaterialExpressionConvert_10 {
   class=MaterialExpressionConvert
   inputs [None {from=MaterialExpressionSaturate_6}]
   properties {
    nodeName=Break Float2
    convertInputs [{
      type=Vector2
      expressionInput {
       expression=&self:MaterialExpressionSaturate_6
       inputName=None
       mask=0
       maskA=0
       maskB=0
       maskG=0
       maskR=0
       outputIndex=0
      }
     }]
    convertOutputs [{type=Scalar}, {type=Scalar}]
   }
  }
  MaterialExpressionConvert_2 {
   class=MaterialExpressionConvert
   inputs [None {from=MaterialExpressionAbs_5}]
   properties {
    nodeName=Break Float3
    convertInputs [{
      type=Vector3
      expressionInput {
       expression=&self:MaterialExpressionAbs_5
       inputName=None
       mask=0
       maskA=0
       maskB=0
       maskG=0
       maskR=0
       outputIndex=0
      }
     }]
    convertOutputs [{type=Scalar}, {type=Scalar}, {type=Scalar}]
   }
  }
  MaterialExpressionConvert_3 {
   class=MaterialExpressionConvert
   inputs [
    None {from=""}
    None {from=""}
    None {from=MaterialExpressionScalarParameter_38}
   ]
   properties {
    nodeName=Make Vector3
    convertInputs [
     {
      type=Scalar
      defaultValue=<rgba 1 0 0 1>
      expressionInput {expression=null, inputName=None, mask=0, maskA=0, maskB=0, maskG=0, maskR=0, outputIndex=0}
     }
     {
      type=Scalar
      defaultValue=<rgba 1 0 0 1>
      expressionInput {expression=null, inputName=None, mask=0, maskA=0, maskB=0, maskG=0, maskR=0, outputIndex=0}
     }
     {
      type=Scalar
      expressionInput {
       expression=&self:MaterialExpressionScalarParameter_38
       inputName=None
       mask=0
       maskA=0
       maskB=0
       maskG=0
       maskR=0
       outputIndex=0
      }
     }
    ]
    convertOutputs [{type=Vector3}]
   }
  }
  MaterialExpressionConvert_4 {
   class=MaterialExpressionConvert
   inputs [None {from=MaterialExpressionSaturate_21}]
   properties {
    nodeName=Break Float3
    convertInputs [{
      type=Vector3
      expressionInput {
       expression=&self:MaterialExpressionSaturate_21
       inputName=None
       mask=0
       maskA=0
       maskB=0
       maskG=0
       maskR=0
       outputIndex=0
      }
     }]
    convertOutputs [{type=Scalar}, {type=Scalar}, {type=Scalar}]
   }
  }
  MaterialExpressionConvert_6 {
   class=MaterialExpressionConvert
   inputs [None {from=MaterialExpressionShaderStageSwitch_0}]
   properties {
    nodeName=Break Float3
    convertInputs [{
      type=Vector3
      expressionInput {
       expression=&self:MaterialExpressionShaderStageSwitch_0
       inputName=None
       mask=0
       maskA=0
       maskB=0
       maskG=0
       maskR=0
       outputIndex=0
      }
     }]
    convertOutputs [{type=Scalar}, {type=Scalar}, {type=Scalar}]
   }
  }
  MaterialExpressionConvert_9 {
   class=MaterialExpressionConvert
   inputs [None {from=MaterialExpressionSaturate_11}]
   properties {
    nodeName=Break Float2
    convertInputs [{
      type=Vector2
      expressionInput {
       expression=&self:MaterialExpressionSaturate_11
       inputName=None
       mask=0
       maskA=0
       maskB=0
       maskG=0
       maskR=0
       outputIndex=0
      }
     }]
    convertOutputs [{type=Scalar}, {type=Scalar}]
   }
  }
  MaterialExpressionCrossProduct_0 {
   class=MaterialExpressionCrossProduct
   inputs [
    A {from=MaterialExpressionReroute_8, output=""}
    B {from=MaterialExpressionTransform_0, output=""}
   ]
   properties {}
  }
  MaterialExpressionCustom_1 {
   class=MaterialExpressionCustom
   inputs [PixelPosition {from=MaterialExpressionScreenPosition_5, output=PixelPosition}]
   properties {
    code="float4x4 bayerMatrix4x4 = float4x4(\r\n    0.0, 8.0, 2.0, 10.0,\r\n    12.0, 4.0, 14.0, 6.0,\r\n    3.0, 11.0, 1.0, 9.0,\r\n    15.0, 7.0, 13.0, 5.0\r\n) / 16.0f;\r\n\r\n    // Get the pixel's integer coordinates\r\n    int2 coords = Mod((int2)(PixelPosition.xy),4);\r\n\r\n    // HLSL accesses matrices using [row][column] syntax\r\n    float dither = bayerMatrix4x4[coords.y][coords.x];\r\n\r\n    return dither;"
    description=Bayer Matrix 4x4
    showCode=true
    inputs [{
      inputName=PixelPosition
      input {
       expression=&self:MaterialExpressionScreenPosition_5
       inputName=None
       mask=0
       maskA=0
       maskB=0
       maskG=0
       maskR=0
       outputIndex=1
      }
     }]
   }
  }
  MaterialExpressionDDX_0 {class=MaterialExpressionDDX, inputs [Value {from=MaterialExpressionMultiply_2, output=""}], properties {}}
  MaterialExpressionDDX_1 {class=MaterialExpressionDDX, inputs [Value {from=MaterialExpressionNamedRerouteUsage_109, output=""}], properties {}}
  MaterialExpressionDDY_0 {class=MaterialExpressionDDY, inputs [Value {from=MaterialExpressionMultiply_2, output=""}], properties {}}
  MaterialExpressionDDY_1 {class=MaterialExpressionDDY, inputs [Value {from=MaterialExpressionNamedRerouteUsage_109, output=""}], properties {}}
  MaterialExpressionDesaturation_0 {
   class=MaterialExpressionDesaturation
   inputs [
    None {from=MaterialExpressionTextureSampleParameter2D_4, output=RGB}
    Fraction {from=MaterialExpressionScalarParameter_29, output=""}
   ]
   properties {}
  }
  MaterialExpressionDistance_0 {
   class=MaterialExpressionDistance
   inputs [
    A {from=MaterialExpressionVectorParameter_3, output=RGB}
    B {from=MaterialExpressionObjectPositionWS_0, output=""}
   ]
   properties {}
  }
  MaterialExpressionDistance_1 {
   class=MaterialExpressionDistance
   inputs [
    A {from=MaterialExpressionConstant2Vector_0, output=""}
    B {from=MaterialExpressionTextureCoordinate_0, output=""}
   ]
   properties {}
  }
  MaterialExpressionDivide_0 {
   class=MaterialExpressionDivide
   inputs [
    A {from=MaterialExpressionSubtract_7}
    B {from=MaterialExpressionScalarParameter_0}
   ]
  }
  MaterialExpressionDivide_1 {
   class=MaterialExpressionDivide
   inputs [
    A {from=MaterialExpressionSubtract_4}
    B {from=MaterialExpressionVectorParameter_5, output=RGB}
   ]
  }
  MaterialExpressionDivide_10 {
   class=MaterialExpressionDivide
   inputs [
    A {from=""}
    B {from=MaterialExpressionNamedRerouteUsage_104}
   ]
   properties {constA=384}
  }
  MaterialExpressionDivide_11 {
   class=MaterialExpressionDivide
   inputs [
    A {from=MaterialExpressionWorldPosition_3, output=XYZ}
    B {from=MaterialExpressionNamedRerouteDeclaration_56}
   ]
  }
  MaterialExpressionDivide_12 {
   class=MaterialExpressionDivide
   inputs [
    A {from=""}
    B {from=MaterialExpressionNamedRerouteUsage_108}
   ]
   properties {constA=384}
  }
  MaterialExpressionDivide_13 {
   class=MaterialExpressionDivide
   inputs [
    A {from=MaterialExpressionLocalPosition_3, output=XYZ}
    B {from=MaterialExpressionVectorParameter_16, output=RGB}
   ]
  }
  MaterialExpressionDivide_14 {
   class=MaterialExpressionDivide
   inputs [
    A {from=MaterialExpressionPixelDepth_1, output=R}
    B {from=MaterialExpressionScalarParameter_69}
   ]
  }
  MaterialExpressionDivide_16 {
   class=MaterialExpressionDivide
   inputs [
    A {from=MaterialExpressionAdd_38}
    B {from=MaterialExpressionDotProduct_10}
   ]
  }
  MaterialExpressionDivide_17 {
   class=MaterialExpressionDivide
   inputs [
    A {from=MaterialExpressionActorPositionWS_1}
    B {from=MaterialExpressionScalarParameter_89}
   ]
  }
  MaterialExpressionDivide_2 {
   class=MaterialExpressionDivide
   inputs [
    A {from=MaterialExpressionSaturate_21}
    B {from=MaterialExpressionAdd_11}
   ]
  }
  MaterialExpressionDivide_3 {
   class=MaterialExpressionDivide
   inputs [
    A {from=MaterialExpressionDistance_0}
    B {from=MaterialExpressionScalarParameter_52}
   ]
  }
  MaterialExpressionDivide_4 {
   class=MaterialExpressionDivide
   inputs [
    A {from=MaterialExpressionWorldPosition_2, output=XY}
    B {from=MaterialExpressionScalarParameter_33}
   ]
   properties {constB=2000}
  }
  MaterialExpressionDivide_5 {
   class=MaterialExpressionDivide
   inputs [
    A {from=MaterialExpressionSubtract_3}
    B {from=MaterialExpressionMaterialFunctionCall_15, output=Local Bounds Size}
   ]
  }
  MaterialExpressionDivide_6 {
   class=MaterialExpressionDivide
   inputs [
    A {from=MaterialExpressionWorldPosition_2, output=XY}
    B {from=MaterialExpressionScalarParameter_26}
   ]
   properties {constB=2000}
  }
  MaterialExpressionDivide_7 {
   class=MaterialExpressionDivide
   inputs [
    A {from=MaterialExpressionWorldPosition_0, output=XYZ}
    B {from=MaterialExpressionNamedRerouteDeclaration_53}
   ]
  }
  MaterialExpressionDivide_8 {
   class=MaterialExpressionDivide
   inputs [
    A {from=MaterialExpressionQualitySwitch_1}
    B {from=MaterialExpressionAppendVector_3}
   ]
  }
  MaterialExpressionDivide_9 {
   class=MaterialExpressionDivide
   inputs [
    A {from=MaterialExpressionAdd_20}
    B {from=""}
   ]
   properties {constB=3}
  }
  MaterialExpressionDotProduct_0 {
   class=MaterialExpressionDotProduct
   inputs [
    A {from=MaterialExpressionLinearInterpolate_6, output=""}
    B {from=MaterialExpressionSkyAtmosphereLightDirection_0, output=""}
   ]
   properties {desc=N dot L}
  }
  MaterialExpressionDotProduct_1 {
   class=MaterialExpressionDotProduct
   inputs [
    A {from=MaterialExpressionNormalize_0, output=""}
    B {from=MaterialExpressionLinearInterpolate_6, output=""}
   ]
   properties {}
  }
  MaterialExpressionDotProduct_10 {
   class=MaterialExpressionDotProduct
   inputs [
    A {from=MaterialExpressionMultiply_63, output=""}
    B {from=MaterialExpressionTransform_10, output=""}
   ]
   properties {}
  }
  MaterialExpressionDotProduct_2 {
   class=MaterialExpressionDotProduct
   inputs [
    A {from=MaterialExpressionNamedRerouteUsage_5, output=""}
    B {from=MaterialExpressionConstant3Vector_1, output=""}
   ]
   properties {}
  }
  MaterialExpressionDotProduct_3 {
   class=MaterialExpressionDotProduct
   inputs [
    A {from=MaterialExpressionConstant4Vector_1, output=RGB}
    B {from=MaterialExpressionAdd_12, output=""}
   ]
   properties {}
  }
  MaterialExpressionDotProduct_4 {
   class=MaterialExpressionDotProduct
   inputs [
    A {from=MaterialExpressionVertexNormalWS_5, output=""}
    B {from=MaterialExpressionCameraVectorWS_0, output=""}
   ]
   properties {}
  }
  MaterialExpressionDotProduct_5 {
   class=MaterialExpressionDotProduct
   inputs [
    A {from=MaterialExpressionConstant4Vector_0, output=RGB}
    B {from=MaterialExpressionAdd_13, output=""}
   ]
   properties {}
  }
  MaterialExpressionDotProduct_6 {
   class=MaterialExpressionDotProduct
   inputs [
    A {from=MaterialExpressionLinearInterpolate_6, output=""}
    B {from=MaterialExpressionNormalize_1, output=""}
   ]
   properties {desc=N dot H}
  }
  MaterialExpressionDotProduct_7 {
   class=MaterialExpressionDotProduct
   inputs [
    A {from=MaterialExpressionPixelNormalWS_2, output=""}
    B {from=MaterialExpressionCameraVectorWS_3, output=""}
   ]
   properties {}
  }
  MaterialExpressionDotProduct_9 {
   class=MaterialExpressionDotProduct
   inputs [
    A {from=MaterialExpressionMaterialProxyReplace_5, output=""}
    B {from=MaterialExpressionVectorParameter_14, output=RGB}
   ]
   properties {}
  }
  MaterialExpressionEyeAdaptationInverse_0 {
   class=MaterialExpressionEyeAdaptationInverse
   inputs [
    LightValueInput {from=MaterialExpressionReroute_44, output=""}
    AlphaInput {from="", output=""}
   ]
   properties {}
  }
  MaterialExpressionFrac_11 {class=MaterialExpressionFrac, inputs [None {from=MaterialExpressionAdd_36, output=""}], properties {}}
  MaterialExpressionFrac_2 {class=MaterialExpressionFrac, inputs [None {from=MaterialExpressionSine_1, output=""}], properties {}}
  MaterialExpressionFrac_3 {class=MaterialExpressionFrac, inputs [None {from=MaterialExpressionSine_3, output=""}], properties {}}
  MaterialExpressionFunctionInput_0 {
   class=MaterialExpressionFunctionInput
   inputs [Preview {from=MaterialExpressionSetMaterialAttributes_2, output=""}]
   properties {inputName=Input MaterialAttributes, inputType=FunctionInput_MaterialAttributes}
  }
  MaterialExpressionFunctionInput_20 {
   class=MaterialExpressionFunctionInput
   inputs [Preview {from="", output=""}]
   properties {inputName=Baked AO, inputType=FunctionInput_Scalar, sortPriority=811}
  }
  MaterialExpressionFunctionInput_21 {
   class=MaterialExpressionFunctionInput
   inputs [Preview {from="", output=""}]
   properties {inputName=Painted Mask, inputType=FunctionInput_Scalar, sortPriority=812}
  }
  MaterialExpressionFunctionInput_6 {
   class=MaterialExpressionFunctionInput
   inputs [Preview {from="", output=""}]
   properties {inputName="\\ --- Dependencies --- /", inputType=FunctionInput_Scalar, sortPriority=800}
  }
  MaterialExpressionFunctionInput_7 {
   class=MaterialExpressionFunctionInput
   inputs [Preview {from="", output=""}]
   properties {inputName=Baked Curvature, inputType=FunctionInput_Scalar, sortPriority=810}
  }
  MaterialExpressionFunctionOutput_0 {
   class=MaterialExpressionFunctionOutput
   inputs [None {from=MaterialExpressionSetMaterialAttributes_1, output=""}]
   properties {outputName=MaterialAttributes, sortPriority=-10}
  }
  MaterialExpressionGetMaterialAttributes_0 {
   class=MaterialExpressionGetMaterialAttributes
   inputs [None {from=MaterialExpressionNamedRerouteUsage_103, output=""}]
   properties {attributeGetTypes [#1]}
  }
  MaterialExpressionGetMaterialAttributes_1 {
   class=MaterialExpressionGetMaterialAttributes
   inputs [None {from=MaterialExpressionNamedRerouteUsage_47, output=""}]
   properties {attributeGetTypes [#2]}
  }
  MaterialExpressionGetMaterialAttributes_2 {
   class=MaterialExpressionGetMaterialAttributes
   inputs [None {from=MaterialExpressionNamedRerouteUsage_39, output=""}]
   properties {attributeGetTypes [
     #3
     #4
     #5
     #6
     #2
    ]}
  }
  MaterialExpressionGetMaterialAttributes_3 {
   class=MaterialExpressionGetMaterialAttributes
   inputs [None {from=MaterialExpressionNamedRerouteUsage_70, output=""}]
   properties {attributeGetTypes [
     #3
     #7
     #4
     #5
     #8
     #9
    ]}
  }
  MaterialExpressionGetMaterialAttributes_4 {
   class=MaterialExpressionGetMaterialAttributes
   inputs [None {from=MaterialExpressionNamedRerouteUsage_55, output=""}]
   properties {attributeGetTypes [#6]}
  }
  MaterialExpressionGetMaterialAttributes_5 {
   class=MaterialExpressionGetMaterialAttributes
   inputs [None {from=MaterialExpressionNamedRerouteUsage_38, output=""}]
   properties {attributeGetTypes [#6]}
  }
  MaterialExpressionGetMaterialAttributes_6 {
   class=MaterialExpressionGetMaterialAttributes
   inputs [None {from=MaterialExpressionNamedRerouteUsage_40, output=""}]
   properties {attributeGetTypes [#4, #5]}
  }
  MaterialExpressionGetMaterialAttributes_7 {
   class=MaterialExpressionGetMaterialAttributes
   inputs [None {from=MaterialExpressionNamedRerouteUsage_82, output=""}]
   properties {attributeGetTypes [
     #3
     #5
     #9
     #10
    ]}
  }
  MaterialExpressionIf_1 {
   class=MaterialExpressionIf
   inputs [
    A {from=MaterialExpressionScalarParameter_89, output=""}
    B {from="", output=""}
    "A > B" {from=MaterialExpressionComponentMask_9, output=""}
    "A == B" {from="", output=""}
    "A < B" {from=MaterialExpressionConstant_9, output=""}
   ]
   properties {constB=1}
  }
  MaterialExpressionIf_3 {
   class=MaterialExpressionIf
   inputs [
    A {from=MaterialExpressionScalarParameter_47, output=""}
    B {from="", output=""}
    "A > B" {from=MaterialExpressionOneMinus_8, output=""}
    "A == B" {from="", output=""}
    "A < B" {from=MaterialExpressionDotProduct_9, output=""}
   ]
   properties {constB=0.5, desc="cheaper than lerp without \"==\" condition", equalsThreshold=0.001}
  }
  MaterialExpressionLinearInterpolate_0 {
   class=MaterialExpressionLinearInterpolate
   inputs [
    A {from=MaterialExpressionMultiply_56}
    B {from=MaterialExpressionMultiply_46}
    Alpha {from=MaterialExpressionScalarParameter_5}
   ]
  }
  MaterialExpressionLinearInterpolate_1 {
   class=MaterialExpressionLinearInterpolate
   inputs [
    A {from=MaterialExpressionVectorParameter_1, output=RGB}
    B {from=MaterialExpressionVectorParameter_2, output=RGB}
    Alpha {from=MaterialExpressionSaturate_0}
   ]
  }
  MaterialExpressionLinearInterpolate_10 {
   class=MaterialExpressionLinearInterpolate
   inputs [
    A {from=MaterialExpressionScalarParameter_25}
    B {from=MaterialExpressionReroute_23}
    Alpha {from=MaterialExpressionConstant_2}
   ]
  }
  MaterialExpressionLinearInterpolate_11 {
   class=MaterialExpressionLinearInterpolate
   inputs [
    A {from=MaterialExpressionReroute_20}
    B {from=MaterialExpressionMultiply_16}
    Alpha {from=MaterialExpressionNamedRerouteUsage_32}
   ]
  }
  MaterialExpressionLinearInterpolate_12 {
   class=MaterialExpressionLinearInterpolate
   inputs [
    A {from=MaterialExpressionScalarParameter_27}
    B {from=MaterialExpressionScalarParameter_28}
    Alpha {from=MaterialExpressionOneMinus_2}
   ]
  }
  MaterialExpressionLinearInterpolate_13 {
   class=MaterialExpressionLinearInterpolate
   inputs [
    A {from=MaterialExpressionReroute_57}
    B {from=MaterialExpressionAdd_47}
    Alpha {from=MaterialExpressionScalarParameter_100}
   ]
  }
  MaterialExpressionLinearInterpolate_15 {
   class=MaterialExpressionLinearInterpolate
   inputs [
    A {from=MaterialExpressionLinearInterpolate_18}
    B {from=MaterialExpressionMultiply_42}
    Alpha {from=MaterialExpressionScalarParameter_40}
   ]
  }
  MaterialExpressionLinearInterpolate_16 {
   class=MaterialExpressionLinearInterpolate
   inputs [
    A {from=MaterialExpressionScalarParameter_63}
    B {from=MaterialExpressionScalarParameter_99}
    Alpha {from=MaterialExpressionTextureSampleParameter2D_5, output=G}
   ]
   properties {constA=-4, constB=-20}
  }
  MaterialExpressionLinearInterpolate_17 {
   class=MaterialExpressionLinearInterpolate
   inputs [
    A {from=MaterialExpressionLinearInterpolate_5}
    B {from=MaterialExpressionMaterialFunctionCall_20, output=Result}
    Alpha {from=MaterialExpressionScalarParameter_23}
   ]
  }
  MaterialExpressionLinearInterpolate_18 {
   class=MaterialExpressionLinearInterpolate
   inputs [
    A {from=MaterialExpressionScalarParameter_43}
    B {from=MaterialExpressionReroute_18}
    Alpha {from=MaterialExpressionConstant_2}
   ]
  }
  MaterialExpressionLinearInterpolate_19 {
   class=MaterialExpressionLinearInterpolate
   inputs [
    A {from=MaterialExpressionVectorParameter_13, output=RGB}
    B {from=MaterialExpressionVectorParameter_15, output=RGB}
    Alpha {from=MaterialExpressionNamedRerouteUsage_49}
   ]
  }
  MaterialExpressionLinearInterpolate_20 {
   class=MaterialExpressionLinearInterpolate
   inputs [
    A {from=MaterialExpressionLinearInterpolate_10}
    B {from=MaterialExpressionMultiply_36}
    Alpha {from=MaterialExpressionScalarParameter_77}
   ]
  }
  MaterialExpressionLinearInterpolate_21 {
   class=MaterialExpressionLinearInterpolate
   inputs [
    A {from=MaterialExpressionScalarParameter_71}
    B {from=MaterialExpressionScalarParameter_72}
    Alpha {from=MaterialExpressionNamedRerouteUsage_81}
   ]
  }
  MaterialExpressionLinearInterpolate_22 {
   class=MaterialExpressionLinearInterpolate
   inputs [
    A {from=MaterialExpressionVectorParameter_19, output=RGB}
    B {from=MaterialExpressionVectorParameter_20, output=RGB}
    Alpha {from=MaterialExpressionPower_15}
   ]
  }
  MaterialExpressionLinearInterpolate_23 {
   class=MaterialExpressionLinearInterpolate
   inputs [
    A {from=MaterialExpressionMultiply_41}
    B {from=""}
    Alpha {from=MaterialExpressionSaturate_17}
   ]
  }
  MaterialExpressionLinearInterpolate_24 {
   class=MaterialExpressionLinearInterpolate
   inputs [
    A {from=MaterialExpressionSmoothStep_0}
    B {from=MaterialExpressionAdd_34}
    Alpha {from=MaterialExpressionPower_4}
   ]
  }
  MaterialExpressionLinearInterpolate_25 {
   class=MaterialExpressionLinearInterpolate
   inputs [
    A {from=MaterialExpressionComponentMask_25}
    B {from=MaterialExpressionComponentMask_26}
    Alpha {from=MaterialExpressionComponentMask_28}
   ]
  }
  MaterialExpressionLinearInterpolate_26 {
   class=MaterialExpressionLinearInterpolate
   inputs [
    A {from=MaterialExpressionVectorParameter_7, output=RGB}
    B {from=MaterialExpressionReroute_67}
    Alpha {from=MaterialExpressionMax_0}
   ]
  }
  MaterialExpressionLinearInterpolate_27 {
   class=MaterialExpressionLinearInterpolate
   inputs [
    A {from=MaterialExpressionReroute_67}
    B {from=MaterialExpressionLinearInterpolate_26}
    Alpha {from=MaterialExpressionNamedRerouteDeclaration_46}
   ]
  }
  MaterialExpressionLinearInterpolate_28 {
   class=MaterialExpressionLinearInterpolate
   inputs [
    A {from=MaterialExpressionNamedRerouteDeclaration_41}
    B {from=""}
    Alpha {from=MaterialExpressionNamedRerouteUsage_72}
   ]
  }
  MaterialExpressionLinearInterpolate_29 {
   class=MaterialExpressionLinearInterpolate
   inputs [
    A {from=MaterialExpressionLinearInterpolate_25}
    B {from=MaterialExpressionComponentMask_27}
    Alpha {from=MaterialExpressionComponentMask_29}
   ]
  }
  MaterialExpressionLinearInterpolate_3 {
   class=MaterialExpressionLinearInterpolate
   inputs [
    A {from=MaterialExpressionScalarParameter_6}
    B {from=MaterialExpressionScalarParameter_9}
    Alpha {from=MaterialExpressionMaterialFunctionCall_11, output=Result}
   ]
  }
  MaterialExpressionLinearInterpolate_31 {
   class=MaterialExpressionLinearInterpolate
   inputs [
    A {from=MaterialExpressionNamedRerouteDeclaration_43}
    B {from=MaterialExpressionMultiply_75}
    Alpha {from=MaterialExpressionComponentMask_0}
   ]
  }
  MaterialExpressionLinearInterpolate_32 {
   class=MaterialExpressionLinearInterpolate
   inputs [
    A {from=MaterialExpressionSubtract_8}
    B {from=MaterialExpressionAdd_42}
    Alpha {from=MaterialExpressionComponentMask_4}
   ]
  }
  MaterialExpressionLinearInterpolate_33 {
   class=MaterialExpressionLinearInterpolate
   inputs [
    A {from=MaterialExpressionScalarParameter_87}
    B {from=MaterialExpressionScalarParameter_90}
    Alpha {from=MaterialExpressionSaturate_12}
   ]
  }
  MaterialExpressionLinearInterpolate_35 {
   class=MaterialExpressionLinearInterpolate
   inputs [
    A {from=MaterialExpressionComponentMask_33}
    B {from=MaterialExpressionComponentMask_34}
    Alpha {from=MaterialExpressionComponentMask_36}
   ]
  }
  MaterialExpressionLinearInterpolate_36 {
   class=MaterialExpressionLinearInterpolate
   inputs [
    A {from=MaterialExpressionLinearInterpolate_35}
    B {from=MaterialExpressionComponentMask_35}
    Alpha {from=MaterialExpressionComponentMask_37}
   ]
  }
  MaterialExpressionLinearInterpolate_4 {
   class=MaterialExpressionLinearInterpolate
   inputs [
    A {from=MaterialExpressionScalarParameter_10}
    B {from=MaterialExpressionScalarParameter_11}
    Alpha {from=MaterialExpressionTextureSample_4, output=G}
   ]
  }
  MaterialExpressionLinearInterpolate_5 {
   class=MaterialExpressionLinearInterpolate
   inputs [
    A {from=MaterialExpressionMultiply_35}
    B {from=MaterialExpressionVectorParameter_10, output=RGB}
    Alpha {from=MaterialExpressionScalarParameter_70}
   ]
  }
  MaterialExpressionLinearInterpolate_6 {
   class=MaterialExpressionLinearInterpolate
   inputs [
    A {from=MaterialExpressionVertexNormalWS_7}
    B {from=MaterialExpressionVertexNormalWS_1}
    Alpha {from=MaterialExpressionScalarParameter_44}
   ]
  }
  MaterialExpressionLinearInterpolate_7 {
   class=MaterialExpressionLinearInterpolate
   inputs [
    A {from=MaterialExpressionMultiply_20}
    B {from=MaterialExpressionVectorParameter_9, output=RGB}
    Alpha {from=MaterialExpressionReroute_3}
   ]
  }
  MaterialExpressionLinearInterpolate_9 {
   class=MaterialExpressionLinearInterpolate
   inputs [
    A {from=MaterialExpressionLinearInterpolate_7}
    B {from=MaterialExpressionMultiply_62}
    Alpha {from=MaterialExpressionScalarParameter_56}
   ]
  }
  MaterialExpressionLocalPosition_0 {class=MaterialExpressionLocalPosition, inputs [], properties {}}
  MaterialExpressionLocalPosition_1 {class=MaterialExpressionLocalPosition, inputs [], properties {localOrigin=InstancePreSkinning}}
  MaterialExpressionLocalPosition_2 {class=MaterialExpressionLocalPosition, inputs [], properties {localOrigin=InstancePreSkinning}}
  MaterialExpressionLocalPosition_3 {class=MaterialExpressionLocalPosition, inputs [], properties {localOrigin=InstancePreSkinning}}
  MaterialExpressionMakeMaterialAttributes_0 {
   class=MaterialExpressionMakeMaterialAttributes
   inputs [
    BaseColor {from=MaterialExpressionTextureSampleParameter2D_3, output=RGB}
    Metallic {from=MaterialExpressionTextureSampleParameter2D_2, output=G}
    Specular {from=MaterialExpressionTextureSampleParameter2D_2, output=G}
    Roughness {from=MaterialExpressionTextureSampleParameter2D_2, output=G}
    Anisotropy {from="", output=""}
    EmissiveColor {from="", output=""}
    Opacity {from="", output=""}
    OpacityMask {from="", output=""}
    Normal {from=MaterialExpressionMultiply_58, output=""}
    Tangent {from="", output=""}
    WorldPositionOffset {from="", output=""}
    SubsurfaceColor {from="", output=""}
    ClearCoat {from="", output=""}
    ClearCoatRoughness {from="", output=""}
    AmbientOcclusion {from="", output=""}
    Refraction {from="", output=""}
    CustomizedUVs_0 {from="", output=""}
    CustomizedUVs_1 {from="", output=""}
    CustomizedUVs_2 {from="", output=""}
    CustomizedUVs_3 {from="", output=""}
    CustomizedUVs_4 {from="", output=""}
    CustomizedUVs_5 {from="", output=""}
    CustomizedUVs_6 {from="", output=""}
    CustomizedUVs_7 {from="", output=""}
    PixelDepthOffset {from="", output=""}
    ShadingModel {from="", output=""}
    Displacement {from="", output=""}
   ]
   properties {}
  }
  MaterialExpressionMaterialFunctionCall_1 {
   class=MaterialExpressionMaterialFunctionCall
   inputs [
    A {from=MaterialExpressionShaderStageSwitch_2}
    B {from=""}
    Bounds {from=MaterialExpressionVectorParameter_12, output=RGB}
    Edge Falloff {from=MaterialExpressionScalarParameter_19}
   ]
   properties {materialFunction=/Engine/Functions/Engine_MaterialFunctions01/Gradient/BoxMask-3D}
  }
  MaterialExpressionMaterialFunctionCall_11 {
   class=MaterialExpressionMaterialFunctionCall
   inputs [
    In {from=MaterialExpressionChannelMaskParameter_0, output=R}
    Contrast {from=MaterialExpressionScalarParameter_20}
   ]
   properties {materialFunction=&CheapContrast}
  }
  MaterialExpressionMaterialFunctionCall_12 {
   class=MaterialExpressionMaterialFunctionCall
   inputs [
    Tangent Normals {from=MaterialExpressionTextureSample_6, output=RGB}
    Vertex Normals {from=""}
    Triplanar Mask {from=MaterialExpressionReroute_2}
   ]
   properties {materialFunction=/Game/Athena/Environments/MaterialFunctions/MF_TriplanarNormalTangentToWorld}
  }
  MaterialExpressionMaterialFunctionCall_15 {
   class=MaterialExpressionMaterialFunctionCall
   inputs []
   properties {materialFunction=&ObjectLocalBounds}
  }
  MaterialExpressionMaterialFunctionCall_16 {
   class=MaterialExpressionMaterialFunctionCall
   inputs [
    A {from=""}
    B {from=""}
    Transition Phase {from=MaterialExpressionNamedRerouteUsage_48}
    Height Texture {from=MaterialExpressionMultiply_22}
    Contrast {from=""}
   ]
   properties {materialFunction=&HeightLerp}
  }
  MaterialExpressionMaterialFunctionCall_2 {
   class=MaterialExpressionMaterialFunctionCall
   inputs [
    Absolute World Position {from=MaterialExpressionWorldPosition_1, output=XYZ}
    "Normal (WS)" {from=""}
    Texture Scale {from=MaterialExpressionVectorParameter_6, output=RGB}
    Triplanar Fade {from=""}
    Derivatives Scale {from=""}
   ]
   properties {materialFunction=/BRCoreMaterials/Materials/MaterialFunctions/MF_TriplanarDithered_WavyDistortion}
  }
  MaterialExpressionMaterialFunctionCall_20 {
   class=MaterialExpressionMaterialFunctionCall
   inputs [
    In {from=MaterialExpressionGetMaterialAttributes_2, output=BaseColor}
    Contrast {from=MaterialExpressionScalarParameter_51}
   ]
   properties {materialFunction=/Engine/Functions/Engine_MaterialFunctions01/ImageAdjustment/CheapContrast_RGB}
  }
  MaterialExpressionMaterialFunctionCall_21 {
   class=MaterialExpressionMaterialFunctionCall
   inputs [
    Plane Normal Axis {from=MaterialExpressionReroute_8}
    Plane X Axis {from=MaterialExpressionNormalize_2}
    Plane Center {from=MaterialExpressionMaterialFunctionCall_23, output=Result}
    UV Scale {from=MaterialExpressionScalarParameter_58}
   ]
   properties {materialFunction=&VirtualPlaneCoordinates}
  }
  MaterialExpressionMaterialFunctionCall_23 {
   class=MaterialExpressionMaterialFunctionCall
   inputs []
   properties {materialFunction=&ComponentPivotLocation}
  }
  MaterialExpressionMaterialFunctionCall_25 {
   class=MaterialExpressionMaterialFunctionCall
   inputs [
    UV {from=MaterialExpressionAdd_23}
    Rotation {from=MaterialExpressionConstant_7}
    Speed {from=MaterialExpressionScalarParameter_65}
    Time {from=""}
    Rotate UVs {from=MaterialExpressionStaticBool_1}
    Fractional Part {from=MaterialExpressionStaticBool_1}
   ]
   properties {materialFunction=/Game/Accessories/FORT_Backpacks/Backpack_M_MED_PitGlass/FX/Materials/MF_3UVPanner}
  }
  MaterialExpressionMaterialFunctionCall_26 {
   class=MaterialExpressionMaterialFunctionCall
   inputs [
    In {from=MaterialExpressionMultiply_51}
    Contrast {from=MaterialExpressionScalarParameter_66}
   ]
   properties {materialFunction=&CheapContrast}
  }
  MaterialExpressionMaterialFunctionCall_3 {
   class=MaterialExpressionMaterialFunctionCall
   inputs [
    In {from=MaterialExpressionSphereMask_1}
    Contrast {from=MaterialExpressionScalarParameter_67}
   ]
   properties {materialFunction=&CheapContrast}
  }
  MaterialExpressionMaterialFunctionCall_35 {
   class=MaterialExpressionMaterialFunctionCall
   inputs [
    Normal {from=MaterialExpressionNamedRerouteUsage_78}
    Flatness {from=MaterialExpressionScalarParameter_88}
   ]
   properties {materialFunction=&FlattenNormal}
  }
  MaterialExpressionMaterialFunctionCall_4 {
   class=MaterialExpressionMaterialFunctionCall
   inputs [
    Absolute World Position {from=""}
    "Normal (WS)" {from=""}
    Texture Scale {from=MaterialExpressionScalarParameter_58}
    Triplanar Fade {from=""}
    Derivatives Scale {from=""}
   ]
   properties {materialFunction=&MF_TriplanarDithered_WavyDistortion}
  }
  MaterialExpressionMaterialFunctionCall_7 {
   class=MaterialExpressionMaterialFunctionCall
   inputs [
    In {from=MaterialExpressionAdd_30}
    Contrast {from=MaterialExpressionScalarParameter_86}
   ]
   properties {materialFunction=&CheapContrast}
  }
  MaterialExpressionMaterialProxyReplace_2 {
   class=MaterialExpressionMaterialProxyReplace
   inputs [
    Realtime {from=MaterialExpressionReroute_24, output=""}
    MaterialProxy {from=MaterialExpressionSaturate_4, output=""}
   ]
   properties {}
  }
  MaterialExpressionMaterialProxyReplace_5 {
   class=MaterialExpressionMaterialProxyReplace
   inputs [
    Realtime {from=MaterialExpressionConvert_0, output=""}
    MaterialProxy {from=MaterialExpressionConvert_1, output=""}
   ]
   properties {}
  }
  MaterialExpressionMax_0 {
   class=MaterialExpressionMax
   inputs [
    A {from=MaterialExpressionMaterialFunctionCall_7, output=Result}
    B {from=MaterialExpressionOneMinus_5}
   ]
  }
  MaterialExpressionMax_1 {
   class=MaterialExpressionMax
   inputs [
    A {from=MaterialExpressionMultiply_23}
    B {from=MaterialExpressionMultiply_27}
   ]
  }
  MaterialExpressionMax_2 {
   class=MaterialExpressionMax
   inputs [
    A {from=MaterialExpressionSaturate_3}
    B {from=MaterialExpressionScalarParameter_55}
   ]
  }
  MaterialExpressionMax_3 {
   class=MaterialExpressionMax
   inputs [
    A {from=MaterialExpressionMultiply_8}
    B {from=MaterialExpressionMultiply_7}
   ]
  }
  MaterialExpressionMax_4 {
   class=MaterialExpressionMax
   inputs [
    A {from=MaterialExpressionConvert_2}
    B {from=MaterialExpressionConvert_2}
   ]
  }
  MaterialExpressionMax_5 {
   class=MaterialExpressionMax
   inputs [
    A {from=MaterialExpressionConvert_2}
    B {from=MaterialExpressionConvert_2}
   ]
  }
  MaterialExpressionMax_6 {
   class=MaterialExpressionMax
   inputs [
    A {from=MaterialExpressionConvert_2}
    B {from=MaterialExpressionConvert_2}
   ]
  }
  MaterialExpressionMax_7 {
   class=MaterialExpressionMax
   inputs [
    A {from=MaterialExpressionComponentMask_23}
    B {from=MaterialExpressionComponentMask_24}
   ]
  }
  MaterialExpressionMax_8 {
   class=MaterialExpressionMax
   inputs [
    A {from=MaterialExpressionMultiply_38}
    B {from=MaterialExpressionMultiply_37}
   ]
  }
  MaterialExpressionMax_9 {
   class=MaterialExpressionMax
   inputs [
    A {from=MaterialExpressionMax_3}
    B {from=MaterialExpressionMax_8}
   ]
  }
  MaterialExpressionMin_3 {
   class=MaterialExpressionMin
   inputs [
    A {from=MaterialExpressionSmoothStep_8, output=""}
    B {from=MaterialExpressionSmoothStep_7, output=""}
   ]
   properties {}
  }
  MaterialExpressionMultiply_0 {
   class=MaterialExpressionMultiply
   inputs [
    A {from=MaterialExpressionScalarParameter_39}
    B {from=MaterialExpressionNamedRerouteUsage_33}
   ]
  }
  MaterialExpressionMultiply_11 {
   class=MaterialExpressionMultiply
   inputs [
    A {from=MaterialExpressionComponentMask_31}
    B {from=""}
   ]
   properties {constB=0.2}
  }
  MaterialExpressionMultiply_12 {
   class=MaterialExpressionMultiply
   inputs [
    A {from=MaterialExpressionPower_1}
    B {from=MaterialExpressionScalarParameter_17}
   ]
  }
  MaterialExpressionMultiply_13 {
   class=MaterialExpressionMultiply
   inputs [
    A {from=MaterialExpressionComponentMask_32}
    B {from=""}
   ]
   properties {constB=0.4}
  }
  MaterialExpressionMultiply_14 {
   class=MaterialExpressionMultiply
   inputs [
    A {from=MaterialExpressionScalarParameter_61}
    B {from=MaterialExpressionMultiply_49}
   ]
  }
  MaterialExpressionMultiply_15 {
   class=MaterialExpressionMultiply
   inputs [
    A {from=MaterialExpressionFrac_11}
    B {from=""}
   ]
   properties {constB=2}
  }
  MaterialExpressionMultiply_16 {
   class=MaterialExpressionMultiply
   inputs [
    A {from=MaterialExpressionAdd_17}
    B {from=MaterialExpressionScalarParameter_21}
   ]
  }
  MaterialExpressionMultiply_17 {
   class=MaterialExpressionMultiply
   inputs [
    A {from=MaterialExpressionReroute_22}
    B {from=MaterialExpressionReroute_22}
   ]
  }
  MaterialExpressionMultiply_18 {
   class=MaterialExpressionMultiply
   inputs [
    A {from=MaterialExpressionReroute_56}
    B {from=MaterialExpressionScalarParameter_30}
   ]
  }
  MaterialExpressionMultiply_19 {
   class=MaterialExpressionMultiply
   inputs [
    A {from=MaterialExpressionMultiply_25}
    B {from=MaterialExpressionPower_2}
   ]
  }
  MaterialExpressionMultiply_2 {
   class=MaterialExpressionMultiply
   inputs [
    A {from=MaterialExpressionDivide_10}
    B {from=MaterialExpressionTextureCoordinate_2}
   ]
  }
  MaterialExpressionMultiply_20 {
   class=MaterialExpressionMultiply
   inputs [
    A {from=MaterialExpressionVectorParameter_8, output=RGB}
    B {from=MaterialExpressionReroute_46}
   ]
  }
  MaterialExpressionMultiply_21 {
   class=MaterialExpressionMultiply
   inputs [
    A {from=MaterialExpressionMultiply_3}
    B {from=MaterialExpressionReroute_48}
   ]
  }
  MaterialExpressionMultiply_22 {
   class=MaterialExpressionMultiply
   inputs [
    A {from=MaterialExpressionNamedRerouteUsage_1}
    B {from=MaterialExpressionGetMaterialAttributes_1, output=Opacity}
   ]
  }
  MaterialExpressionMultiply_23 {
   class=MaterialExpressionMultiply
   inputs [
    A {from=MaterialExpressionScalarParameter_36}
    B {from=MaterialExpressionMultiply_39}
   ]
  }
  MaterialExpressionMultiply_24 {
   class=MaterialExpressionMultiply
   inputs [
    A {from=MaterialExpressionSine_2}
    B {from=MaterialExpressionScalarParameter_34}
   ]
   properties {constB=0.02}
  }
  MaterialExpressionMultiply_25 {
   class=MaterialExpressionMultiply
   inputs [
    A {from=MaterialExpressionFrac_3}
    B {from=MaterialExpressionFrac_2}
   ]
  }
  MaterialExpressionMultiply_26 {
   class=MaterialExpressionMultiply
   inputs [
    A {from=MaterialExpressionMultiply_44}
    B {from=MaterialExpressionNamedRerouteUsage_92}
   ]
  }
  MaterialExpressionMultiply_27 {
   class=MaterialExpressionMultiply
   inputs [
    A {from=MaterialExpressionScalarParameter_24}
    B {from=MaterialExpressionMaterialFunctionCall_16, output=Alpha}
   ]
  }
  MaterialExpressionMultiply_28 {
   class=MaterialExpressionMultiply
   inputs [
    A {from=MaterialExpressionMultiply_17}
    B {from=MaterialExpressionScalarParameter_45}
   ]
  }
  MaterialExpressionMultiply_29 {
   class=MaterialExpressionMultiply
   inputs [
    A {from=MaterialExpressionSign_0}
    B {from=""}
   ]
   properties {constB=-1}
  }
  MaterialExpressionMultiply_3 {
   class=MaterialExpressionMultiply
   inputs [
    A {from=MaterialExpressionNamedRerouteUsage_26}
    B {from=MaterialExpressionScalarParameter_3}
   ]
   properties {desc=Offset BumpOffset with height of internal layers}
  }
  MaterialExpressionMultiply_31 {
   class=MaterialExpressionMultiply
   inputs [
    A {from=MaterialExpressionReroute_4}
    B {from=MaterialExpressionNamedRerouteUsage_63}
   ]
  }
  MaterialExpressionMultiply_32 {
   class=MaterialExpressionMultiply
   inputs [
    A {from=MaterialExpressionMultiply_34}
    B {from=MaterialExpressionScalarParameter_35}
   ]
  }
  MaterialExpressionMultiply_33 {
   class=MaterialExpressionMultiply
   inputs [
    A {from=MaterialExpressionDistance_1}
    B {from=""}
   ]
   properties {constB=2}
  }
  MaterialExpressionMultiply_34 {
   class=MaterialExpressionMultiply
   inputs [
    A {from=MaterialExpressionAdd_14}
    B {from=MaterialExpressionScalarParameter_57}
   ]
  }
  MaterialExpressionMultiply_35 {
   class=MaterialExpressionMultiply
   inputs [
    A {from=MaterialExpressionVectorParameter_10, output=RGB}
    B {from=MaterialExpressionGetMaterialAttributes_2, output=BaseColor}
   ]
  }
  MaterialExpressionMultiply_36 {
   class=MaterialExpressionMultiply
   inputs [
    A {from=MaterialExpressionOneMinus_6}
    B {from=MaterialExpressionLinearInterpolate_10}
   ]
  }
  MaterialExpressionMultiply_37 {
   class=MaterialExpressionMultiply
   inputs [
    A {from=MaterialExpressionAppendVector_5}
    B {from=MaterialExpressionConvert_10}
   ]
  }
  MaterialExpressionMultiply_38 {
   class=MaterialExpressionMultiply
   inputs [
    A {from=MaterialExpressionAppendVector_6}
    B {from=MaterialExpressionConvert_10}
   ]
  }
  MaterialExpressionMultiply_39 {
   class=MaterialExpressionMultiply
   inputs [
    A {from=MaterialExpressionNamedRerouteUsage_44}
    B {from=MaterialExpressionMultiply_22}
   ]
  }
  MaterialExpressionMultiply_4 {
   class=MaterialExpressionMultiply
   inputs [
    A {from=MaterialExpressionReroute_41}
    B {from=MaterialExpressionScalarParameter_4}
   ]
  }
  MaterialExpressionMultiply_40 {
   class=MaterialExpressionMultiply
   inputs [
    A {from=MaterialExpressionStaticSwitch_3}
    B {from=MaterialExpressionLinearInterpolate_19}
   ]
  }
  MaterialExpressionMultiply_41 {
   class=MaterialExpressionMultiply
   inputs [
    A {from=MaterialExpressionReroute_64}
    B {from=MaterialExpressionReroute_64}
   ]
  }
  MaterialExpressionMultiply_42 {
   class=MaterialExpressionMultiply
   inputs [
    A {from=MaterialExpressionMultiply_53}
    B {from=MaterialExpressionLinearInterpolate_18}
   ]
  }
  MaterialExpressionMultiply_43 {
   class=MaterialExpressionMultiply
   inputs [
    A {from=MaterialExpressionPower_7}
    B {from=MaterialExpressionScalarParameter_37}
   ]
  }
  MaterialExpressionMultiply_44 {
   class=MaterialExpressionMultiply
   inputs [
    A {from=MaterialExpressionAdd_15}
    B {from=MaterialExpressionScalarParameter_57}
   ]
  }
  MaterialExpressionMultiply_45 {
   class=MaterialExpressionMultiply
   inputs [
    A {from=MaterialExpressionStaticSwitchParameter_1}
    B {from=MaterialExpressionLinearInterpolate_1}
   ]
  }
  MaterialExpressionMultiply_46 {
   class=MaterialExpressionMultiply
   inputs [
    A {from=MaterialExpressionMultiply_56}
    B {from=MaterialExpressionNamedRerouteUsage_30}
   ]
  }
  MaterialExpressionMultiply_47 {
   class=MaterialExpressionMultiply
   inputs [
    A {from=MaterialExpressionReroute_39}
    B {from=MaterialExpressionScalarParameter_60}
   ]
  }
  MaterialExpressionMultiply_48 {
   class=MaterialExpressionMultiply
   inputs [
    A {from=MaterialExpressionScalarParameter_59}
    B {from=MaterialExpressionTwoSidedSign_3}
   ]
  }
  MaterialExpressionMultiply_49 {
   class=MaterialExpressionMultiply
   inputs [
    A {from=MaterialExpressionConstantBiasScale_0}
    B {from=MaterialExpressionSquareRoot_0}
   ]
  }
  MaterialExpressionMultiply_50 {
   class=MaterialExpressionMultiply
   inputs [
    A {from=MaterialExpressionNamedRerouteUsage_32}
    B {from=MaterialExpressionNamedRerouteUsage_4}
   ]
  }
  MaterialExpressionMultiply_51 {
   class=MaterialExpressionMultiply
   inputs [
    A {from=MaterialExpressionDivide_9}
    B {from=MaterialExpressionNamedRerouteUsage_73}
   ]
  }
  MaterialExpressionMultiply_52 {
   class=MaterialExpressionMultiply
   inputs [
    A {from=MaterialExpressionReroute_9}
    B {from=MaterialExpressionVectorParameter_22, output=BaseColor Multiply}
   ]
  }
  MaterialExpressionMultiply_53 {
   class=MaterialExpressionMultiply
   inputs [
    A {from=MaterialExpressionTextureSample_3, output=R}
    B {from=""}
   ]
   properties {constB=2}
  }
  MaterialExpressionMultiply_54 {
   class=MaterialExpressionMultiply
   inputs [
    A {from=MaterialExpressionReroute_9}
    B {from=MaterialExpressionVectorParameter_23, output=BaseColor Multiply}
   ]
  }
  MaterialExpressionMultiply_56 {
   class=MaterialExpressionMultiply
   inputs [
    A {from=MaterialExpressionMultiply_45}
    B {from=MaterialExpressionScalarParameter_76}
   ]
   properties {constB=5}
  }
  MaterialExpressionMultiply_57 {
   class=MaterialExpressionMultiply
   inputs [
    A {from=MaterialExpressionNamedRerouteDeclaration_42}
    B {from=MaterialExpressionLinearInterpolate_22}
   ]
  }
  MaterialExpressionMultiply_58 {
   class=MaterialExpressionMultiply
   inputs [
    A {from=MaterialExpressionTextureSampleParameter2D_0, output=RGB}
    B {from=MaterialExpressionConvert_3}
   ]
  }
  MaterialExpressionMultiply_59 {
   class=MaterialExpressionMultiply
   inputs [
    A {from=""}
    B {from=MaterialExpressionScalarParameter_38}
   ]
  }
  MaterialExpressionMultiply_6 {
   class=MaterialExpressionMultiply
   inputs [
    A {from=MaterialExpressionAdd_33}
    B {from=""}
   ]
   properties {constB=0.3333}
  }
  MaterialExpressionMultiply_60 {
   class=MaterialExpressionMultiply
   inputs [
    A {from=MaterialExpressionSaturate_0}
    B {from=MaterialExpressionNamedRerouteUsage_56}
   ]
  }
  MaterialExpressionMultiply_61 {
   class=MaterialExpressionMultiply
   inputs [
    A {from=MaterialExpressionNamedRerouteUsage_65}
    B {from=MaterialExpressionNamedRerouteUsage_93}
   ]
  }
  MaterialExpressionMultiply_62 {
   class=MaterialExpressionMultiply
   inputs [
    A {from=MaterialExpressionLinearInterpolate_7}
    B {from=MaterialExpressionNamedRerouteUsage_35}
   ]
  }
  MaterialExpressionMultiply_63 {
   class=MaterialExpressionMultiply
   inputs [
    A {from=MaterialExpressionCameraVectorWS_6}
    B {from=MaterialExpressionConstant3Vector_0}
   ]
  }
  MaterialExpressionMultiply_64 {
   class=MaterialExpressionMultiply
   inputs [
    A {from=MaterialExpressionCameraVectorWS_2}
    B {from=""}
   ]
   properties {constB=-1}
  }
  MaterialExpressionMultiply_65 {
   class=MaterialExpressionMultiply
   inputs [
    A {from=MaterialExpressionMultiply_64}
    B {from=MaterialExpressionNamedRerouteUsage_107}
   ]
  }
  MaterialExpressionMultiply_66 {
   class=MaterialExpressionMultiply
   inputs [
    A {from=MaterialExpressionDivide_12}
    B {from=MaterialExpressionTextureCoordinate_3}
   ]
  }
  MaterialExpressionMultiply_67 {
   class=MaterialExpressionMultiply
   inputs [
    A {from=MaterialExpressionAdd_41}
    B {from=MaterialExpressionSphereMask_0}
   ]
  }
  MaterialExpressionMultiply_68 {
   class=MaterialExpressionMultiply
   inputs [
    A {from=MaterialExpressionComponentMask_10}
    B {from=MaterialExpressionComponentMask_15}
   ]
  }
  MaterialExpressionMultiply_69 {
   class=MaterialExpressionMultiply
   inputs [
    A {from=MaterialExpressionAdd_43}
    B {from=MaterialExpressionScalarParameter_75}
   ]
  }
  MaterialExpressionMultiply_7 {
   class=MaterialExpressionMultiply
   inputs [
    A {from=MaterialExpressionAppendVector_10}
    B {from=MaterialExpressionConvert_9}
   ]
  }
  MaterialExpressionMultiply_70 {
   class=MaterialExpressionMultiply
   inputs [
    A {from=MaterialExpressionComponentMask_16}
    B {from=MaterialExpressionComponentMask_13}
   ]
  }
  MaterialExpressionMultiply_71 {
   class=MaterialExpressionMultiply
   inputs [
    A {from=MaterialExpressionComponentMask_20}
    B {from=MaterialExpressionComponentMask_11}
   ]
  }
  MaterialExpressionMultiply_72 {
   class=MaterialExpressionMultiply
   inputs [
    A {from=MaterialExpressionNamedRerouteUsage_37}
    B {from=MaterialExpressionScalarParameter_106}
   ]
   properties {constB=0.1}
  }
  MaterialExpressionMultiply_73 {
   class=MaterialExpressionMultiply
   inputs [
    A {from=MaterialExpressionReroute_9}
    B {from=MaterialExpressionVectorParameter_25, output=BaseColor Multiply}
   ]
  }
  MaterialExpressionMultiply_75 {
   class=MaterialExpressionMultiply
   inputs [
    A {from=MaterialExpressionNamedRerouteDeclaration_43}
    B {from=MaterialExpressionNamedRerouteDeclaration_43}
   ]
  }
  MaterialExpressionMultiply_76 {
   class=MaterialExpressionMultiply
   inputs [
    A {from=MaterialExpressionNamedRerouteUsage_27}
    B {from=MaterialExpressionReroute_72}
   ]
  }
  MaterialExpressionMultiply_77 {
   class=MaterialExpressionMultiply
   inputs [
    A {from=MaterialExpressionScalarParameter_107}
    B {from=MaterialExpressionMultiply_76}
   ]
  }
  MaterialExpressionMultiply_8 {
   class=MaterialExpressionMultiply
   inputs [
    A {from=MaterialExpressionAppendVector_2}
    B {from=MaterialExpressionConvert_9}
   ]
  }
  MaterialExpressionMultiply_81 {
   class=MaterialExpressionMultiply
   inputs [
    A {from=MaterialExpressionReroute_10}
    B {from=MaterialExpressionVectorParameter_22, output=Roughness Multiply}
   ]
  }
  MaterialExpressionMultiply_82 {
   class=MaterialExpressionMultiply
   inputs [
    A {from=MaterialExpressionReroute_10}
    B {from=MaterialExpressionVectorParameter_23, output=Roughness Multiply}
   ]
  }
  MaterialExpressionMultiply_83 {
   class=MaterialExpressionMultiply
   inputs [
    A {from=MaterialExpressionReroute_10}
    B {from=MaterialExpressionVectorParameter_25, output=Roughness Multiply}
   ]
  }
  MaterialExpressionMultiply_84 {
   class=MaterialExpressionMultiply
   inputs [
    A {from=MaterialExpressionReroute_14}
    B {from=MaterialExpressionVectorParameter_22, output=Emissive Multiply}
   ]
  }
  MaterialExpressionMultiply_85 {
   class=MaterialExpressionMultiply
   inputs [
    A {from=MaterialExpressionSaturate_5}
    B {from=MaterialExpressionScalarParameter_101}
   ]
  }
  MaterialExpressionMultiply_86 {
   class=MaterialExpressionMultiply
   inputs [
    A {from=MaterialExpressionReroute_14}
    B {from=MaterialExpressionVectorParameter_23, output=Emissive Multiply}
   ]
  }
  MaterialExpressionMultiply_87 {
   class=MaterialExpressionMultiply
   inputs [
    A {from=MaterialExpressionReroute_14}
    B {from=MaterialExpressionVectorParameter_25, output=Emissive Multiply}
   ]
  }
  MaterialExpressionMultiply_88 {
   class=MaterialExpressionMultiply
   inputs [
    A {from=MaterialExpressionReroute_15}
    B {from=MaterialExpressionVectorParameter_22, output=SSS Multiply}
   ]
  }
  MaterialExpressionMultiply_89 {
   class=MaterialExpressionMultiply
   inputs [
    A {from=MaterialExpressionReroute_15}
    B {from=MaterialExpressionVectorParameter_23, output=SSS Multiply}
   ]
  }
  MaterialExpressionMultiply_9 {
   class=MaterialExpressionMultiply
   inputs [
    A {from=MaterialExpressionComponentMask_30}
    B {from=""}
   ]
   properties {constB=-1}
  }
  MaterialExpressionMultiply_90 {
   class=MaterialExpressionMultiply
   inputs [
    A {from=MaterialExpressionReroute_15}
    B {from=MaterialExpressionVectorParameter_25, output=SSS Multiply}
   ]
  }
  MaterialExpressionNamedRerouteDeclaration_0 {
   class=MaterialExpressionNamedRerouteDeclaration
   inputs [None {from=MaterialExpressionMultiply_56}]
   properties {name=Ice - Internal Aurora, nodeColor=<rgba 0 0.9766 1 1>}
  }
  MaterialExpressionNamedRerouteDeclaration_1 {
   class=MaterialExpressionNamedRerouteDeclaration
   inputs [None {from=MaterialExpressionOneMinus_0}]
   properties {name=Fresnel_Inverted, nodeColor=<rgba 0.9922 1 0 1>}
  }
  MaterialExpressionNamedRerouteDeclaration_14 {
   class=MaterialExpressionNamedRerouteDeclaration
   inputs [None {from=MaterialExpressionTextureSample_8, output=R}]
   properties {name=Internal Layers - Distortion - Height, nodeColor=<rgba 0.3828 1 0 1>}
  }
  MaterialExpressionNamedRerouteDeclaration_15 {
   class=MaterialExpressionNamedRerouteDeclaration
   inputs [None {from=MaterialExpressionTextureCoordinate_1}]
   properties {name=Internal Layers - UVs, nodeColor=<rgba 1 0 0.7031 1>}
  }
  MaterialExpressionNamedRerouteDeclaration_16 {
   class=MaterialExpressionNamedRerouteDeclaration
   inputs [None {from=MaterialExpressionNamedRerouteDeclaration_21}]
   properties {name=_Z Gradient, nodeColor=<rgba 1 0 0.0525 1>}
  }
  MaterialExpressionNamedRerouteDeclaration_17 {
   class=MaterialExpressionNamedRerouteDeclaration
   inputs [None {from=MaterialExpressionMultiply_12}]
   properties {name=Ice - Light - Skylight, nodeColor=<rgba 0 1 0.0156 1>}
  }
  MaterialExpressionNamedRerouteDeclaration_18 {
   class=MaterialExpressionNamedRerouteDeclaration
   inputs [None {from=MaterialExpressionMakeMaterialAttributes_0}]
   properties {name=_Input MaterialAttributes, nodeColor=<rgba 0.1016 0 1 1>}
  }
  MaterialExpressionNamedRerouteDeclaration_2 {
   class=MaterialExpressionNamedRerouteDeclaration
   inputs [None {from=MaterialExpressionOneMinus_1}]
   properties {name=Fresnel, nodeColor=<rgba 0.9922 1 0 1>}
  }
  MaterialExpressionNamedRerouteDeclaration_20 {
   class=MaterialExpressionNamedRerouteDeclaration
   inputs [None {from=MaterialExpressionSaturate_18}]
   properties {name=Thickness, nodeColor=<rgba 1 0.0469 0 1>}
  }
  MaterialExpressionNamedRerouteDeclaration_21 {
   class=MaterialExpressionNamedRerouteDeclaration
   inputs [None {from=MaterialExpressionComponentMask_1}]
   properties {name=_Z Gradient, nodeColor=<rgba 0 0 0.5469 1>}
  }
  MaterialExpressionNamedRerouteDeclaration_22 {
   class=MaterialExpressionNamedRerouteDeclaration
   inputs [None {from=MaterialExpressionMaterialFunctionCall_16, output=Alpha}]
   properties {name=InternalLight - Mask, nodeColor=<rgba 1 1 1 1>}
  }
  MaterialExpressionNamedRerouteDeclaration_23 {
   class=MaterialExpressionNamedRerouteDeclaration
   inputs [None {from=MaterialExpressionSaturate_10}]
   properties {name=Internal Light - Sun - BackScatter, nodeColor=<rgba 1 0.8438 0 1>}
  }
  MaterialExpressionNamedRerouteDeclaration_24 {
   class=MaterialExpressionNamedRerouteDeclaration
   inputs [None {from=MaterialExpressionMax_1}]
   properties {name=SSS - Opacity, nodeColor=<rgba 0 1 0.0391 1>}
  }
  MaterialExpressionNamedRerouteDeclaration_25 {
   class=MaterialExpressionNamedRerouteDeclaration
   inputs [None {from=MaterialExpressionMultiply_40}]
   properties {name=SSS - Color, nodeColor=<rgba 0 1 0.0391 1>}
  }
  MaterialExpressionNamedRerouteDeclaration_26 {
   class=MaterialExpressionNamedRerouteDeclaration
   inputs [None {from=MaterialExpressionSetMaterialAttributes_4}]
   properties {name=Ice Elements Combined, nodeColor=<rgba 0.9922 1 0 1>}
  }
  MaterialExpressionNamedRerouteDeclaration_28 {
   class=MaterialExpressionNamedRerouteDeclaration
   inputs [None {from=MaterialExpressionMax_2}]
   properties {name=ShardCoreDarkenMask, nodeColor=<rgba 1 0 0.0469 1>}
  }
  MaterialExpressionNamedRerouteDeclaration_29 {
   class=MaterialExpressionNamedRerouteDeclaration
   inputs [None {from=MaterialExpressionLinearInterpolate_9}]
   properties {name=Internal Layers, nodeColor=<rgba 0.125 0 1 1>}
  }
  MaterialExpressionNamedRerouteDeclaration_3 {
   class=MaterialExpressionNamedRerouteDeclaration
   inputs [None {from=MaterialExpressionSetMaterialAttributes_5}]
   properties {name=Ice, nodeColor=<rgba 1 0.4453 0 1>}
  }
  MaterialExpressionNamedRerouteDeclaration_30 {
   class=MaterialExpressionNamedRerouteDeclaration
   inputs [None {from=MaterialExpressionFunctionInput_20}]
   properties {name=_Baked AO, nodeColor=<rgba 1 0 0.0525 1>}
  }
  MaterialExpressionNamedRerouteDeclaration_31 {
   class=MaterialExpressionNamedRerouteDeclaration
   inputs [None {from=MaterialExpressionMultiply_57}]
   properties {name=DisplayEmsissiveColor, nodeColor=<rgba 0.3125 1 0 1>}
  }
  MaterialExpressionNamedRerouteDeclaration_32 {
   class=MaterialExpressionNamedRerouteDeclaration
   inputs [None {from=MaterialExpressionFunctionInput_7}]
   properties {name=_Baked Curvature, nodeColor=<rgba 1 0 0.0525 1>}
  }
  MaterialExpressionNamedRerouteDeclaration_33 {
   class=MaterialExpressionNamedRerouteDeclaration
   inputs [None {from=MaterialExpressionFunctionInput_21}]
   properties {name=_Painted Mask, nodeColor=<rgba 1 0 0.0525 1>}
  }
  MaterialExpressionNamedRerouteDeclaration_34 {
   class=MaterialExpressionNamedRerouteDeclaration
   inputs [None {from=MaterialExpressionMultiply_67}]
   properties {name=Parallax Layers, nodeColor=<rgba 0.5234 1 0 1>}
  }
  MaterialExpressionNamedRerouteDeclaration_35 {
   class=MaterialExpressionNamedRerouteDeclaration
   inputs [None {from=MaterialExpressionReroute_3}]
   properties {name=Internal Layers - Cracks, nodeColor=<rgba 0.125 0 1 1>}
  }
  MaterialExpressionNamedRerouteDeclaration_36 {
   class=MaterialExpressionNamedRerouteDeclaration
   inputs [None {from=MaterialExpressionReroute_46}]
   properties {name=Internal Layers - Air Bubbles, nodeColor=<rgba 0.125 0 1 1>}
  }
  MaterialExpressionNamedRerouteDeclaration_38 {
   class=MaterialExpressionNamedRerouteDeclaration
   inputs [None {from=MaterialExpressionSaturate_1}]
   properties {name=IceDisplay_Opacity, nodeColor=<rgba 0.3828 0 1 1>}
  }
  MaterialExpressionNamedRerouteDeclaration_39 {
   class=MaterialExpressionNamedRerouteDeclaration
   inputs [None {from=MaterialExpressionTextureSample_4, output=B}]
   properties {name=Ice - Soft Crack Shapes, nodeColor=<rgba 1 0.7266 0 1>}
  }
  MaterialExpressionNamedRerouteDeclaration_4 {
   class=MaterialExpressionNamedRerouteDeclaration
   inputs [None {from=MaterialExpressionSaturate_9}]
   properties {name=Internal Light - Shard - Towards, nodeColor=<rgba 1 0.8438 0 1>}
  }
  MaterialExpressionNamedRerouteDeclaration_40 {
   class=MaterialExpressionNamedRerouteDeclaration
   inputs [None {from=MaterialExpressionSaturate_32}]
   properties {name=Internal Light - Z up Normal Light, nodeColor=<rgba 1 0.8672 0 1>}
  }
  MaterialExpressionNamedRerouteDeclaration_41 {
   class=MaterialExpressionNamedRerouteDeclaration
   inputs [None {from=MaterialExpressionSaturate_20}]
   properties {name=TrophyDepth, nodeColor=<rgba 1 0.1172 0 1>}
  }
  MaterialExpressionNamedRerouteDeclaration_42 {
   class=MaterialExpressionNamedRerouteDeclaration
   inputs [None {from=MaterialExpressionSwitch_1}]
   properties {name=DisplayArea, nodeColor=<rgba 0 0.7891 1 1>}
  }
  MaterialExpressionNamedRerouteDeclaration_43 {
   class=MaterialExpressionNamedRerouteDeclaration
   inputs [None {from=MaterialExpressionStaticSwitchParameter_7}]
   properties {name=IceDisplay_InnerLight, nodeColor=<rgba 1 0 0.75 1>}
  }
  MaterialExpressionNamedRerouteDeclaration_45 {
   class=MaterialExpressionNamedRerouteDeclaration
   inputs [None {from=MaterialExpressionSaturate_0}]
   properties {name=WS Gradient and Thickness Mix, nodeColor=<rgba 0.4297 0 1 1>}
  }
  MaterialExpressionNamedRerouteDeclaration_46 {
   class=MaterialExpressionNamedRerouteDeclaration
   inputs [None {from=MaterialExpressionLinearInterpolate_33}]
   properties {name=IceDisplay_TrophyFlatColor, nodeColor=<rgba 1 0.7734 0 1>}
  }
  MaterialExpressionNamedRerouteDeclaration_5 {
   class=MaterialExpressionNamedRerouteDeclaration
   inputs [None {from=MaterialExpressionSaturate_8}]
   properties {name=Internal Light - Sun - Towards, nodeColor=<rgba 1 0.8438 0 1>}
  }
  MaterialExpressionNamedRerouteDeclaration_50 {
   class=MaterialExpressionNamedRerouteDeclaration
   inputs [None {from=MaterialExpressionLinearInterpolate_36}]
   properties {name=Offset UVs, nodeColor=<rgba 0.9453 0 1 1>}
  }
  MaterialExpressionNamedRerouteDeclaration_53 {
   class=MaterialExpressionNamedRerouteDeclaration
   inputs [None {from=MaterialExpressionScalarParameter_103}]
   properties {name=UV Scale, nodeColor=<rgba 0.9688 1 0 1>}
  }
  MaterialExpressionNamedRerouteDeclaration_54 {
   class=MaterialExpressionNamedRerouteDeclaration
   inputs [None {from=MaterialExpressionAppendVector_9}]
   properties {name=Dithered Mask, nodeColor=<rgba 1 0 0 1>}
  }
  MaterialExpressionNamedRerouteDeclaration_55 {
   class=MaterialExpressionNamedRerouteDeclaration
   inputs [None {from=MaterialExpressionLinearInterpolate_16}]
   properties {name=Depth Offset, nodeColor=<rgba 0.4531 1 0 1>}
  }
  MaterialExpressionNamedRerouteDeclaration_56 {
   class=MaterialExpressionNamedRerouteDeclaration
   inputs [None {from=MaterialExpressionScalarParameter_104}]
   properties {name=Offset UV Scale, nodeColor=<rgba 0.9688 1 0 1>}
  }
  MaterialExpressionNamedRerouteDeclaration_57 {
   class=MaterialExpressionNamedRerouteDeclaration
   inputs [None {from=MaterialExpressionMultiply_66}]
   properties {name=Offset DDX DDY Coords, nodeColor=<rgba 1 0.2344 0 1>}
  }
  MaterialExpressionNamedRerouteDeclaration_6 {
   class=MaterialExpressionNamedRerouteDeclaration
   inputs [None {from=MaterialExpressionTransform_6}]
   properties {name=Internal Layers - Distortion - Normal, nodeColor=<rgba 0 1 0.3672 1>}
  }
  MaterialExpressionNamedRerouteDeclaration_7 {
   class=MaterialExpressionNamedRerouteDeclaration
   inputs [None {from=MaterialExpressionMax_9}]
   properties {name=UV_Box, nodeColor=<rgba 0.875 0 1 1>}
  }
  MaterialExpressionNamedRerouteUsage_0 {class=MaterialExpressionNamedRerouteUsage}
  MaterialExpressionNamedRerouteUsage_1 {class=MaterialExpressionNamedRerouteUsage}
  MaterialExpressionNamedRerouteUsage_103 {class=MaterialExpressionNamedRerouteUsage}
  MaterialExpressionNamedRerouteUsage_104 {class=MaterialExpressionNamedRerouteUsage}
  MaterialExpressionNamedRerouteUsage_105 {class=MaterialExpressionNamedRerouteUsage}
  MaterialExpressionNamedRerouteUsage_106 {class=MaterialExpressionNamedRerouteUsage}
  MaterialExpressionNamedRerouteUsage_107 {class=MaterialExpressionNamedRerouteUsage}
  MaterialExpressionNamedRerouteUsage_108 {class=MaterialExpressionNamedRerouteUsage}
  MaterialExpressionNamedRerouteUsage_109 {class=MaterialExpressionNamedRerouteUsage}
  MaterialExpressionNamedRerouteUsage_26 {class=MaterialExpressionNamedRerouteUsage}
  MaterialExpressionNamedRerouteUsage_27 {class=MaterialExpressionNamedRerouteUsage}
  MaterialExpressionNamedRerouteUsage_28 {class=MaterialExpressionNamedRerouteUsage}
  MaterialExpressionNamedRerouteUsage_3 {class=MaterialExpressionNamedRerouteUsage}
  MaterialExpressionNamedRerouteUsage_30 {class=MaterialExpressionNamedRerouteUsage}
  MaterialExpressionNamedRerouteUsage_31 {class=MaterialExpressionNamedRerouteUsage}
  MaterialExpressionNamedRerouteUsage_32 {class=MaterialExpressionNamedRerouteUsage}
  MaterialExpressionNamedRerouteUsage_33 {class=MaterialExpressionNamedRerouteUsage}
  MaterialExpressionNamedRerouteUsage_34 {class=MaterialExpressionNamedRerouteUsage}
  MaterialExpressionNamedRerouteUsage_35 {class=MaterialExpressionNamedRerouteUsage}
  MaterialExpressionNamedRerouteUsage_36 {class=MaterialExpressionNamedRerouteUsage}
  MaterialExpressionNamedRerouteUsage_37 {class=MaterialExpressionNamedRerouteUsage}
  MaterialExpressionNamedRerouteUsage_38 {class=MaterialExpressionNamedRerouteUsage}
  MaterialExpressionNamedRerouteUsage_39 {class=MaterialExpressionNamedRerouteUsage}
  MaterialExpressionNamedRerouteUsage_4 {class=MaterialExpressionNamedRerouteUsage}
  MaterialExpressionNamedRerouteUsage_40 {class=MaterialExpressionNamedRerouteUsage}
  MaterialExpressionNamedRerouteUsage_41 {class=MaterialExpressionNamedRerouteUsage}
  MaterialExpressionNamedRerouteUsage_42 {class=MaterialExpressionNamedRerouteUsage}
  MaterialExpressionNamedRerouteUsage_43 {class=MaterialExpressionNamedRerouteUsage}
  MaterialExpressionNamedRerouteUsage_44 {class=MaterialExpressionNamedRerouteUsage}
  MaterialExpressionNamedRerouteUsage_46 {class=MaterialExpressionNamedRerouteUsage}
  MaterialExpressionNamedRerouteUsage_47 {class=MaterialExpressionNamedRerouteUsage}
  MaterialExpressionNamedRerouteUsage_48 {class=MaterialExpressionNamedRerouteUsage}
  MaterialExpressionNamedRerouteUsage_49 {class=MaterialExpressionNamedRerouteUsage}
  MaterialExpressionNamedRerouteUsage_5 {class=MaterialExpressionNamedRerouteUsage}
  MaterialExpressionNamedRerouteUsage_50 {class=MaterialExpressionNamedRerouteUsage}
  MaterialExpressionNamedRerouteUsage_51 {class=MaterialExpressionNamedRerouteUsage}
  MaterialExpressionNamedRerouteUsage_54 {class=MaterialExpressionNamedRerouteUsage}
  MaterialExpressionNamedRerouteUsage_55 {class=MaterialExpressionNamedRerouteUsage}
  MaterialExpressionNamedRerouteUsage_56 {class=MaterialExpressionNamedRerouteUsage}
  MaterialExpressionNamedRerouteUsage_57 {class=MaterialExpressionNamedRerouteUsage}
  MaterialExpressionNamedRerouteUsage_62 {class=MaterialExpressionNamedRerouteUsage}
  MaterialExpressionNamedRerouteUsage_63 {class=MaterialExpressionNamedRerouteUsage}
  MaterialExpressionNamedRerouteUsage_65 {class=MaterialExpressionNamedRerouteUsage}
  MaterialExpressionNamedRerouteUsage_67 {class=MaterialExpressionNamedRerouteUsage}
  MaterialExpressionNamedRerouteUsage_68 {class=MaterialExpressionNamedRerouteUsage}
  MaterialExpressionNamedRerouteUsage_69 {class=MaterialExpressionNamedRerouteUsage}
  MaterialExpressionNamedRerouteUsage_70 {class=MaterialExpressionNamedRerouteUsage}
  MaterialExpressionNamedRerouteUsage_71 {class=MaterialExpressionNamedRerouteUsage}
  MaterialExpressionNamedRerouteUsage_72 {class=MaterialExpressionNamedRerouteUsage}
  MaterialExpressionNamedRerouteUsage_73 {class=MaterialExpressionNamedRerouteUsage}
  MaterialExpressionNamedRerouteUsage_74 {class=MaterialExpressionNamedRerouteUsage}
  MaterialExpressionNamedRerouteUsage_75 {class=MaterialExpressionNamedRerouteUsage}
  MaterialExpressionNamedRerouteUsage_76 {class=MaterialExpressionNamedRerouteUsage}
  MaterialExpressionNamedRerouteUsage_78 {class=MaterialExpressionNamedRerouteUsage}
  MaterialExpressionNamedRerouteUsage_79 {class=MaterialExpressionNamedRerouteUsage}
  MaterialExpressionNamedRerouteUsage_80 {class=MaterialExpressionNamedRerouteUsage}
  MaterialExpressionNamedRerouteUsage_81 {class=MaterialExpressionNamedRerouteUsage}
  MaterialExpressionNamedRerouteUsage_82 {class=MaterialExpressionNamedRerouteUsage}
  MaterialExpressionNamedRerouteUsage_83 {class=MaterialExpressionNamedRerouteUsage}
  MaterialExpressionNamedRerouteUsage_85 {class=MaterialExpressionNamedRerouteUsage}
  MaterialExpressionNamedRerouteUsage_92 {class=MaterialExpressionNamedRerouteUsage}
  MaterialExpressionNamedRerouteUsage_93 {class=MaterialExpressionNamedRerouteUsage}
  MaterialExpressionNamedRerouteUsage_95 {class=MaterialExpressionNamedRerouteUsage}
  MaterialExpressionNormalize_0 {class=MaterialExpressionNormalize, inputs [VectorInput {from=MaterialExpressionSubtract_1, output=""}]}
  MaterialExpressionNormalize_1 {class=MaterialExpressionNormalize, inputs [VectorInput {from=MaterialExpressionAdd_0, output=""}]}
  MaterialExpressionNormalize_2 {class=MaterialExpressionNormalize, inputs [VectorInput {from=MaterialExpressionCrossProduct_0, output=""}]}
  MaterialExpressionObjectPositionWS_0 {class=MaterialExpressionObjectPositionWS, inputs [], properties {}}
  MaterialExpressionOneMinus_0 {class=MaterialExpressionOneMinus, inputs [None {from=MaterialExpressionNamedRerouteDeclaration_2}]}
  MaterialExpressionOneMinus_1 {class=MaterialExpressionOneMinus, inputs [None {from=MaterialExpressionSaturate_7}]}
  MaterialExpressionOneMinus_10 {class=MaterialExpressionOneMinus, inputs [None {from=MaterialExpressionMax_7}]}
  MaterialExpressionOneMinus_11 {class=MaterialExpressionOneMinus, inputs [None {from=MaterialExpressionConvert_6}]}
  MaterialExpressionOneMinus_2 {class=MaterialExpressionOneMinus, inputs [None {from=MaterialExpressionNamedRerouteUsage_43}]}
  MaterialExpressionOneMinus_3 {class=MaterialExpressionOneMinus, inputs [None {from=MaterialExpressionConvert_6}]}
  MaterialExpressionOneMinus_4 {class=MaterialExpressionOneMinus, inputs [None {from=MaterialExpressionDotProduct_6}]}
  MaterialExpressionOneMinus_5 {class=MaterialExpressionOneMinus, inputs [None {from=MaterialExpressionNamedRerouteUsage_85}]}
  MaterialExpressionOneMinus_6 {class=MaterialExpressionOneMinus, inputs [None {from=MaterialExpressionTextureSample_3, output=R}]}
  MaterialExpressionOneMinus_8 {class=MaterialExpressionOneMinus, inputs [None {from=MaterialExpressionDotProduct_9}]}
  MaterialExpressionOneMinus_9 {class=MaterialExpressionOneMinus, inputs [None {from=MaterialExpressionConvert_6}]}
  MaterialExpressionPanner_0 {
   class=MaterialExpressionPanner
   inputs [
    Coordinate {from=MaterialExpressionMultiply_47}
    Time {from=""}
    Speed {from=MaterialExpressionAppendVector_1}
   ]
   properties {bFractionalPart=true}
  }
  MaterialExpressionPixelDepth_0 {class=MaterialExpressionPixelDepth, inputs [], properties {}}
  MaterialExpressionPixelDepth_1 {class=MaterialExpressionPixelDepth, inputs [], properties {}}
  MaterialExpressionPixelDepth_3 {class=MaterialExpressionPixelDepth, inputs [], properties {}}
  MaterialExpressionPixelNormalWS_0 {class=MaterialExpressionPixelNormalWS, inputs [], properties {}}
  MaterialExpressionPixelNormalWS_1 {class=MaterialExpressionPixelNormalWS, inputs [], properties {}}
  MaterialExpressionPixelNormalWS_2 {class=MaterialExpressionPixelNormalWS, inputs [], properties {}}
  MaterialExpressionPower_0 {
   class=MaterialExpressionPower
   inputs [
    Base {from=MaterialExpressionSaturate_2}
    Exp {from=MaterialExpressionScalarParameter_2}
   ]
  }
  MaterialExpressionPower_1 {
   class=MaterialExpressionPower
   inputs [
    Base {from=MaterialExpressionSkyLightEnvMapSample_0}
    Exp {from=MaterialExpressionScalarParameter_8}
   ]
  }
  MaterialExpressionPower_10 {
   class=MaterialExpressionPower
   inputs [
    Base {from=""}
    Exp {from=""}
   ]
  }
  MaterialExpressionPower_12 {
   class=MaterialExpressionPower
   inputs [
    Base {from=MaterialExpressionSaturate_23}
    Exp {from=MaterialExpressionScalarParameter_54}
   ]
  }
  MaterialExpressionPower_14 {
   class=MaterialExpressionPower
   inputs [
    Base {from=MaterialExpressionMin_3}
    Exp {from=""}
   ]
   properties {constExponent=4}
  }
  MaterialExpressionPower_15 {
   class=MaterialExpressionPower
   inputs [
    Base {from=MaterialExpressionNamedRerouteDeclaration_42}
    Exp {from=""}
   ]
   properties {constExponent=4}
  }
  MaterialExpressionPower_2 {
   class=MaterialExpressionPower
   inputs [
    Base {from=MaterialExpressionLinearInterpolate_12}
    Exp {from=MaterialExpressionScalarParameter_31}
   ]
  }
  MaterialExpressionPower_3 {
   class=MaterialExpressionPower
   inputs [
    Base {from=MaterialExpressionOneMinus_4}
    Exp {from=MaterialExpressionConstant_1}
   ]
  }
  MaterialExpressionPower_4 {
   class=MaterialExpressionPower
   inputs [
    Base {from=MaterialExpressionNamedRerouteUsage_80}
    Exp {from=MaterialExpressionScalarParameter_82}
   ]
  }
  MaterialExpressionPower_5 {
   class=MaterialExpressionPower
   inputs [
    Base {from=MaterialExpressionDivide_3}
    Exp {from=MaterialExpressionScalarParameter_53}
   ]
  }
  MaterialExpressionPower_6 {
   class=MaterialExpressionPower
   inputs [
    Base {from=MaterialExpressionDotProduct_2}
    Exp {from=MaterialExpressionScalarParameter_14}
   ]
  }
  MaterialExpressionPower_7 {
   class=MaterialExpressionPower
   inputs [
    Base {from=MaterialExpressionSaturate_16}
    Exp {from=MaterialExpressionScalarParameter_22}
   ]
  }
  MaterialExpressionPower_9 {
   class=MaterialExpressionPower
   inputs [
    Base {from=MaterialExpressionDesaturation_0}
    Exp {from=MaterialExpressionScalarParameter_105}
   ]
  }
  MaterialExpressionQualitySwitch_0 {
   class=MaterialExpressionQualitySwitch
   inputs [
    Default {from="", output=""}
    Low {from="", output=""}
    High {from="", output=""}
    Medium {from="", output=""}
    Epic {from="", output=""}
   ]
   properties {}
  }
  MaterialExpressionQualitySwitch_1 {
   class=MaterialExpressionQualitySwitch
   inputs [
    Default {from=MaterialExpressionAdd_18, output=""}
    Low {from=MaterialExpressionReroute_38, output=""}
    High {from="", output=""}
    Medium {from=MaterialExpressionReroute_38, output=""}
    Epic {from="", output=""}
   ]
   properties {}
  }
  MaterialExpressionQualitySwitch_11 {
   class=MaterialExpressionQualitySwitch
   inputs [
    Default {from=MaterialExpressionReroute_10, output=""}
    Low {from=MaterialExpressionMultiply_82, output=""}
    High {from=MaterialExpressionMultiply_83, output=""}
    Medium {from=MaterialExpressionMultiply_81, output=""}
    Epic {from=MaterialExpressionReroute_10, output=""}
   ]
   properties {}
  }
  MaterialExpressionQualitySwitch_12 {
   class=MaterialExpressionQualitySwitch
   inputs [
    Default {from=MaterialExpressionReroute_14, output=""}
    Low {from=MaterialExpressionMultiply_86, output=""}
    High {from=MaterialExpressionMultiply_87, output=""}
    Medium {from=MaterialExpressionMultiply_84, output=""}
    Epic {from=MaterialExpressionReroute_14, output=""}
   ]
   properties {}
  }
  MaterialExpressionQualitySwitch_13 {
   class=MaterialExpressionQualitySwitch
   inputs [
    Default {from=MaterialExpressionReroute_15, output=""}
    Low {from=MaterialExpressionMultiply_89, output=""}
    High {from=MaterialExpressionMultiply_90, output=""}
    Medium {from=MaterialExpressionMultiply_88, output=""}
    Epic {from=MaterialExpressionReroute_15, output=""}
   ]
   properties {}
  }
  MaterialExpressionQualitySwitch_2 {
   class=MaterialExpressionQualitySwitch
   inputs [
    Default {from=MaterialExpressionAdd_45, output=""}
    Low {from=MaterialExpressionReroute_72, output=""}
    High {from=MaterialExpressionAdd_45, output=""}
    Medium {from=MaterialExpressionReroute_72, output=""}
    Epic {from=MaterialExpressionAdd_45, output=""}
   ]
   properties {}
  }
  MaterialExpressionQualitySwitch_3 {
   class=MaterialExpressionQualitySwitch
   inputs [
    Default {from=MaterialExpressionAdd_40, output=""}
    Low {from="", output=""}
    High {from="", output=""}
    Medium {from="", output=""}
    Epic {from="", output=""}
   ]
   properties {}
  }
  MaterialExpressionQualitySwitch_6 {
   class=MaterialExpressionQualitySwitch
   inputs [
    Default {from=MaterialExpressionReroute_9, output=""}
    Low {from=MaterialExpressionMultiply_54, output=""}
    High {from=MaterialExpressionMultiply_73, output=""}
    Medium {from=MaterialExpressionMultiply_52, output=""}
    Epic {from=MaterialExpressionReroute_9, output=""}
   ]
   properties {}
  }
  MaterialExpressionRayTracingQualitySwitch_0 {
   class=MaterialExpressionRayTracingQualitySwitch
   inputs [
    Normal {from=MaterialExpressionReroute_29, output=""}
    RayTraced {from=MaterialExpressionMultiply_32, output=""}
   ]
   properties {}
  }
  MaterialExpressionReflectionVectorWS_0 {class=MaterialExpressionReflectionVectorWS, inputs [CustomWorldNormal {from="", output=""}], properties {}}
  MaterialExpressionReroute_0 {class=MaterialExpressionReroute, inputs [None {from=MaterialExpressionDivide_0}]}
  MaterialExpressionReroute_1 {class=MaterialExpressionReroute, inputs [None {from=MaterialExpressionGetMaterialAttributes_3, output=BaseColor}]}
  MaterialExpressionReroute_10 {class=MaterialExpressionReroute, inputs [None {from=MaterialExpressionGetMaterialAttributes_7, output=Roughness}]}
  MaterialExpressionReroute_11 {class=MaterialExpressionReroute, inputs [None {from=MaterialExpressionReroute_1}]}
  MaterialExpressionReroute_12 {class=MaterialExpressionReroute, inputs [None {from=MaterialExpressionLocalPosition_2, output=XYZ}]}
  MaterialExpressionReroute_13 {class=MaterialExpressionReroute, inputs [None {from=MaterialExpressionDivide_13}]}
  MaterialExpressionReroute_14 {class=MaterialExpressionReroute, inputs [None {from=MaterialExpressionGetMaterialAttributes_7, output=EmissiveColor}]}
  MaterialExpressionReroute_15 {class=MaterialExpressionReroute, inputs [None {from=MaterialExpressionGetMaterialAttributes_7, output=SubsurfaceColor}]}
  MaterialExpressionReroute_17 {class=MaterialExpressionReroute, inputs [None {from=MaterialExpressionReroute_21}]}
  MaterialExpressionReroute_18 {class=MaterialExpressionReroute, inputs [None {from=MaterialExpressionGetMaterialAttributes_6, output=Roughness}]}
  MaterialExpressionReroute_19 {class=MaterialExpressionReroute, inputs [None {from=MaterialExpressionReroute_57}]}
  MaterialExpressionReroute_2 {class=MaterialExpressionReroute, inputs [None {from=MaterialExpressionReroute_6}]}
  MaterialExpressionReroute_20 {class=MaterialExpressionReroute, inputs [None {from=MaterialExpressionLinearInterpolate_17}]}
  MaterialExpressionReroute_21 {class=MaterialExpressionReroute, inputs [None {from=MaterialExpressionGetMaterialAttributes_3, output=Roughness}]}
  MaterialExpressionReroute_22 {class=MaterialExpressionReroute, inputs [None {from=MaterialExpressionReroute_1}]}
  MaterialExpressionReroute_23 {class=MaterialExpressionReroute, inputs [None {from=MaterialExpressionGetMaterialAttributes_6, output=Specular}]}
  MaterialExpressionReroute_24 {class=MaterialExpressionReroute, inputs [None {from=MaterialExpressionReroute_45}]}
  MaterialExpressionReroute_25 {class=MaterialExpressionReroute, inputs [None {from=MaterialExpressionMultiply_25}]}
  MaterialExpressionReroute_26 {class=MaterialExpressionReroute, inputs [None {from=MaterialExpressionRayTracingQualitySwitch_0}]}
  MaterialExpressionReroute_27 {class=MaterialExpressionReroute, inputs [None {from=MaterialExpressionReroute_47}]}
  MaterialExpressionReroute_28 {class=MaterialExpressionReroute, inputs [None {from=MaterialExpressionReroute_43}]}
  MaterialExpressionReroute_29 {class=MaterialExpressionReroute, inputs [None {from=MaterialExpressionEyeAdaptationInverse_0, output=EyeAdaptationInverse}]}
  MaterialExpressionReroute_3 {class=MaterialExpressionReroute, inputs [None {from=MaterialExpressionLinearInterpolate_4}]}
  MaterialExpressionReroute_30 {class=MaterialExpressionReroute, inputs [None {from=MaterialExpressionLinearInterpolate_20}]}
  MaterialExpressionReroute_31 {class=MaterialExpressionReroute, inputs [None {from=MaterialExpressionLinearInterpolate_15}]}
  MaterialExpressionReroute_32 {class=MaterialExpressionReroute, inputs [None {from=MaterialExpressionReroute_33}]}
  MaterialExpressionReroute_33 {class=MaterialExpressionReroute, inputs [None {from=MaterialExpressionGetMaterialAttributes_3, output=Specular}]}
  MaterialExpressionReroute_34 {class=MaterialExpressionReroute, inputs [None {from=MaterialExpressionReroute_20}]}
  MaterialExpressionReroute_35 {class=MaterialExpressionReroute, inputs [None {from=MaterialExpressionReroute_65}]}
  MaterialExpressionReroute_36 {class=MaterialExpressionReroute, inputs [None {from=MaterialExpressionStaticSwitchParameter_3}]}
  MaterialExpressionReroute_37 {class=MaterialExpressionReroute, inputs [None {from=MaterialExpressionGetMaterialAttributes_5, output=Normal}]}
  MaterialExpressionReroute_38 {class=MaterialExpressionReroute, inputs [None {from=MaterialExpressionVertexInterpolator_0, output=PS}]}
  MaterialExpressionReroute_39 {class=MaterialExpressionReroute, inputs [None {from=MaterialExpressionMaterialFunctionCall_21, output=UVs}]}
  MaterialExpressionReroute_4 {class=MaterialExpressionReroute, inputs [None {from=MaterialExpressionLinearInterpolate_3}]}
  MaterialExpressionReroute_40 {class=MaterialExpressionReroute, inputs [None {from=MaterialExpressionMultiply_26}]}
  MaterialExpressionReroute_41 {class=MaterialExpressionReroute, inputs [None {from=MaterialExpressionReroute_53}]}
  MaterialExpressionReroute_42 {class=MaterialExpressionReroute, inputs [None {from=MaterialExpressionNamedRerouteUsage_50}]}
  MaterialExpressionReroute_43 {class=MaterialExpressionReroute, inputs [None {from=MaterialExpressionNamedRerouteUsage_50}]}
  MaterialExpressionReroute_44 {class=MaterialExpressionReroute, inputs [None {from=MaterialExpressionAdd_40}]}
  MaterialExpressionReroute_45 {class=MaterialExpressionReroute, inputs [None {from=MaterialExpressionReroute_73}]}
  MaterialExpressionReroute_46 {class=MaterialExpressionReroute, inputs [None {from=MaterialExpressionMultiply_31}]}
  MaterialExpressionReroute_47 {class=MaterialExpressionReroute, inputs [None {from=MaterialExpressionMultiply_28}]}
  MaterialExpressionReroute_48 {class=MaterialExpressionReroute, inputs [None {from=MaterialExpressionDivide_16}]}
  MaterialExpressionReroute_49 {class=MaterialExpressionReroute, inputs [None {from=MaterialExpressionMaterialFunctionCall_2, output=Mask}]}
  MaterialExpressionReroute_5 {class=MaterialExpressionReroute, inputs [None {from=MaterialExpressionMultiply_85}]}
  MaterialExpressionReroute_50 {class=MaterialExpressionReroute, inputs [None {from=MaterialExpressionMultiply_63}]}
  MaterialExpressionReroute_51 {class=MaterialExpressionReroute, inputs [None {from=MaterialExpressionMultiply_28}]}
  MaterialExpressionReroute_52 {class=MaterialExpressionReroute, inputs [None {from=MaterialExpressionReroute_51}]}
  MaterialExpressionReroute_53 {class=MaterialExpressionReroute, inputs [None {from=MaterialExpressionReroute_35}]}
  MaterialExpressionReroute_55 {class=MaterialExpressionReroute, inputs [None {from=MaterialExpressionReroute_35}]}
  MaterialExpressionReroute_56 {class=MaterialExpressionReroute, inputs [None {from=MaterialExpressionReroute_55}]}
  MaterialExpressionReroute_57 {class=MaterialExpressionReroute, inputs [None {from=MaterialExpressionReroute_1}]}
  MaterialExpressionReroute_58 {class=MaterialExpressionReroute, inputs [None {from=MaterialExpressionShadingPathSwitch_0}]}
  MaterialExpressionReroute_59 {class=MaterialExpressionReroute, inputs [None {from=MaterialExpressionMultiply_18}]}
  MaterialExpressionReroute_6 {class=MaterialExpressionReroute, inputs [None {from=MaterialExpressionMaterialFunctionCall_2, output=Mask}]}
  MaterialExpressionReroute_60 {class=MaterialExpressionReroute, inputs [None {from=MaterialExpressionNamedRerouteUsage_105}]}
  MaterialExpressionReroute_61 {class=MaterialExpressionReroute, inputs [None {from=MaterialExpressionShadingPathSwitch_1}]}
  MaterialExpressionReroute_62 {class=MaterialExpressionReroute, inputs [None {from=MaterialExpressionReroute_63}]}
  MaterialExpressionReroute_63 {class=MaterialExpressionReroute, inputs [None {from=MaterialExpressionMaterialFunctionCall_2, output=UVs}]}
  MaterialExpressionReroute_64 {class=MaterialExpressionReroute, inputs [None {from=MaterialExpressionReroute_74}]}
  MaterialExpressionReroute_65 {class=MaterialExpressionReroute, inputs [None {from=MaterialExpressionMaterialFunctionCall_2, output=UVs}]}
  MaterialExpressionReroute_66 {class=MaterialExpressionReroute, inputs [None {from=MaterialExpressionReroute_31}]}
  MaterialExpressionReroute_67 {class=MaterialExpressionReroute, inputs [None {from=MaterialExpressionNamedRerouteDeclaration_31}]}
  MaterialExpressionReroute_68 {class=MaterialExpressionReroute, inputs [None {from=MaterialExpressionReroute_67}]}
  MaterialExpressionReroute_69 {class=MaterialExpressionReroute, inputs [None {from=MaterialExpressionNamedRerouteUsage_106}]}
  MaterialExpressionReroute_7 {class=MaterialExpressionReroute, inputs [None {from=MaterialExpressionDivide_1}]}
  MaterialExpressionReroute_70 {class=MaterialExpressionReroute, inputs [None {from=MaterialExpressionSubtract_0}]}
  MaterialExpressionReroute_71 {class=MaterialExpressionReroute, inputs [None {from=MaterialExpressionNamedRerouteUsage_51}]}
  MaterialExpressionReroute_72 {class=MaterialExpressionReroute, inputs [None {from=MaterialExpressionReroute_11}]}
  MaterialExpressionReroute_73 {class=MaterialExpressionReroute, inputs [None {from=MaterialExpressionShadingPathSwitch_2}]}
  MaterialExpressionReroute_74 {class=MaterialExpressionReroute, inputs [None {from=MaterialExpressionLinearInterpolate_28}]}
  MaterialExpressionReroute_8 {class=MaterialExpressionReroute, inputs [None {from=MaterialExpressionCameraVectorWS_1}]}
  MaterialExpressionReroute_9 {class=MaterialExpressionReroute, inputs [None {from=MaterialExpressionGetMaterialAttributes_7, output=BaseColor}]}
  MaterialExpressionRound_1 {class=MaterialExpressionRound, inputs [None {from=MaterialExpressionAdd_6, output=""}], properties {}}
  MaterialExpressionSaturate_0 {class=MaterialExpressionSaturate, inputs [None {from=MaterialExpressionReroute_0}]}
  MaterialExpressionSaturate_1 {class=MaterialExpressionSaturate, inputs [None {from=MaterialExpressionLinearInterpolate_23}]}
  MaterialExpressionSaturate_10 {class=MaterialExpressionSaturate, inputs [None {from=MaterialExpressionPower_3}]}
  MaterialExpressionSaturate_11 {class=MaterialExpressionSaturate, inputs [None {from=MaterialExpressionSign_0}]}
  MaterialExpressionSaturate_12 {class=MaterialExpressionSaturate, inputs [None {from=MaterialExpressionDivide_14}]}
  MaterialExpressionSaturate_14 {class=MaterialExpressionSaturate, inputs [None {from=MaterialExpressionComponentMask_12}]}
  MaterialExpressionSaturate_16 {class=MaterialExpressionSaturate, inputs [None {from=MaterialExpressionIf_3}]}
  MaterialExpressionSaturate_17 {class=MaterialExpressionSaturate, inputs [None {from=MaterialExpressionLinearInterpolate_21}]}
  MaterialExpressionSaturate_18 {class=MaterialExpressionSaturate, inputs [None {from=MaterialExpressionMultiply_43}]}
  MaterialExpressionSaturate_2 {class=MaterialExpressionSaturate, inputs [None {from=MaterialExpressionDotProduct_1}]}
  MaterialExpressionSaturate_20 {class=MaterialExpressionSaturate, inputs [None {from=MaterialExpressionLinearInterpolate_24}]}
  MaterialExpressionSaturate_21 {class=MaterialExpressionSaturate, inputs [None {from=MaterialExpressionAdd_8}]}
  MaterialExpressionSaturate_23 {class=MaterialExpressionSaturate, inputs [None {from=MaterialExpressionDotProduct_0}]}
  MaterialExpressionSaturate_3 {class=MaterialExpressionSaturate, inputs [None {from=MaterialExpressionPower_5}]}
  MaterialExpressionSaturate_32 {class=MaterialExpressionSaturate, inputs [None {from=MaterialExpressionMultiply_69}]}
  MaterialExpressionSaturate_4 {class=MaterialExpressionSaturate, inputs [None {from=MaterialExpressionAdd_28}]}
  MaterialExpressionSaturate_5 {class=MaterialExpressionSaturate, inputs [None {from=MaterialExpressionLinearInterpolate_11}]}
  MaterialExpressionSaturate_6 {class=MaterialExpressionSaturate, inputs [None {from=MaterialExpressionMultiply_29}]}
  MaterialExpressionSaturate_7 {class=MaterialExpressionSaturate, inputs [None {from=MaterialExpressionDotProduct_4}]}
  MaterialExpressionSaturate_8 {class=MaterialExpressionSaturate, inputs [None {from=MaterialExpressionPower_12}]}
  MaterialExpressionSaturate_9 {class=MaterialExpressionSaturate, inputs [None {from=MaterialExpressionPower_0}]}
  MaterialExpressionScalarParameter_0 {
   class=MaterialExpressionScalarParameter
   properties {defaultValue=2000, group="-       EchoFrost Ice", parameterName=* Internal Aurora - Gradient - Scale, sortPriority=310}
  }
  MaterialExpressionScalarParameter_1 {
   class=MaterialExpressionScalarParameter
   properties {
    defaultValue=-1.2
    group="-       EchoFrost Ice"
    parameterName=* Internal Aurora - Gradient - Offset
    primitiveDataIndex=3
    sortPriority=310
   }
  }
  MaterialExpressionScalarParameter_10 {
   class=MaterialExpressionScalarParameter
   properties {group="-       EchoFrost Ice", parameterName=* Internal Layers - Cracks - Min, sortPriority=200}
  }
  MaterialExpressionScalarParameter_100 {
   class=MaterialExpressionScalarParameter
   properties {defaultValue=0.2, group="-       EchoFrost Ice", parameterName=* Internal Layers - BaseColor Contribution, sortPriority=200}
  }
  MaterialExpressionScalarParameter_101 {
   class=MaterialExpressionScalarParameter
   properties {defaultValue=0.5, group="-       EchoFrost Ice", parameterName=* Ice - Basecolor - Overall Brightness, sortPriority=50}
  }
  MaterialExpressionScalarParameter_102 {
   class=MaterialExpressionScalarParameter
   properties {defaultValue=0.1, group="-       EchoFrost Ice", parameterName=* Internal Layers - Parallax - Triplanar Blend, sortPriority=311}
  }
  MaterialExpressionScalarParameter_103 {
   class=MaterialExpressionScalarParameter
   properties {defaultValue=200, group="-       EchoFrost Ice", parameterName=* Internal Layers - Parallax - Scale 1, sortPriority=302}
  }
  MaterialExpressionScalarParameter_104 {
   class=MaterialExpressionScalarParameter
   properties {defaultValue=500, group="-       EchoFrost Ice", parameterName=* Internal Layers - Parallax - Scale 2, sortPriority=303}
  }
  MaterialExpressionScalarParameter_105 {
   class=MaterialExpressionScalarParameter
   properties {defaultValue=1, group="-       EchoFrost Ice", parameterName=* Internal Layers - Parallax - Power, sortPriority=308}
  }
  MaterialExpressionScalarParameter_106 {
   class=MaterialExpressionScalarParameter
   properties {defaultValue=0.1, group="-       EchoFrost Ice", parameterName=* Internal Layers - Parallax - Emissive Intensity, sortPriority=307}
  }
  MaterialExpressionScalarParameter_107 {
   class=MaterialExpressionScalarParameter
   properties {defaultValue=2, group="-       EchoFrost Ice", parameterName=* Internal Layers - Parallax - BaseColor Intensity, sortPriority=306}
  }
  MaterialExpressionScalarParameter_11 {
   class=MaterialExpressionScalarParameter
   properties {defaultValue=0.3, group="-       EchoFrost Ice", parameterName=* Internal Layers - Cracks - Max, sortPriority=200}
  }
  MaterialExpressionScalarParameter_12 {
   class=MaterialExpressionScalarParameter
   properties {bUseCustomPrimitiveData=true, group="-       EchoFrost Ice_Display", parameterName=CPD_IceDisplay_OffsetZ, primitiveDataIndex=3}
  }
  MaterialExpressionScalarParameter_13 {
   class=MaterialExpressionScalarParameter
   properties {defaultValue=-0.2, group="-       EchoFrost Ice", parameterName=* Internal Layers - HeightRatio, sortPriority=200}
  }
  MaterialExpressionScalarParameter_14 {class=MaterialExpressionScalarParameter, properties {defaultValue=0.5, parameterName=testing}}
  MaterialExpressionScalarParameter_15 {
   class=MaterialExpressionScalarParameter
   properties {
    controlType=EnumerationIndex
    desc="0=Column\r\n1=Window\r\n2=Box"
    group="-       EchoFrost Ice_Display"
    parameterName=IceDisplay_Column/Window/Box
    sliderMax=1
    sortPriority=40
   }
  }
  MaterialExpressionScalarParameter_16 {
   class=MaterialExpressionScalarParameter
   properties {bUseCustomPrimitiveData=true, group="-       EchoFrost Ice", parameterName=* Internal Aurora - Gradient - CPD Offset, primitiveDataIndex=2}
  }
  MaterialExpressionScalarParameter_17 {
   class=MaterialExpressionScalarParameter
   properties {defaultValue=0.1, group="-       EchoFrost Ice", parameterName=* Ice - Internal Light - Sky Sample - Brightness}
  }
  MaterialExpressionScalarParameter_19 {class=MaterialExpressionScalarParameter, properties {defaultValue=0.5, parameterName=IceDisplay_Box_BoxMask_Falloff}}
  MaterialExpressionScalarParameter_2 {
   class=MaterialExpressionScalarParameter
   properties {defaultValue=5, group="-       EchoFrost Ice", parameterName=* Ice - Internal Light - Shard - Power}
  }
  MaterialExpressionScalarParameter_20 {
   class=MaterialExpressionScalarParameter
   properties {group="-       EchoFrost Ice", parameterName=* Internal Layers - Air Bubbles - Contrast, sortPriority=200}
  }
  MaterialExpressionScalarParameter_21 {
   class=MaterialExpressionScalarParameter
   properties {defaultValue=2, group="-       EchoFrost Ice", parameterName=* Ice - BaseColor - Fresnel Intensity, sortPriority=50}
  }
  MaterialExpressionScalarParameter_22 {
   class=MaterialExpressionScalarParameter
   properties {defaultValue=3, group="-       EchoFrost Ice", parameterName=* Ice - Thickness - Power, sortPriority=85}
  }
  MaterialExpressionScalarParameter_23 {
   class=MaterialExpressionScalarParameter
   properties {
    defaultValue=0.5
    desc="1 = Fully use the Layers BaseColor, 0 = Fully use the Tint color"
    group="-       EchoFrost Ice"
    parameterName=* Ice - BaseColor - Layers - Blend
    sortPriority=50
   }
  }
  MaterialExpressionScalarParameter_24 {
   class=MaterialExpressionScalarParameter
   properties {defaultValue=0.1, group="-       EchoFrost Ice", parameterName=* Ice - SSS - Opacity - Towards, sortPriority=60}
  }
  MaterialExpressionScalarParameter_25 {class=MaterialExpressionScalarParameter, properties {defaultValue=0.5, group="-       EchoFrost Ice", parameterName=* Ice - Specular}}
  MaterialExpressionScalarParameter_26 {class=MaterialExpressionScalarParameter, properties {defaultValue=2000, parameterName=InternalLight - Simple WS Noise - Scale}}
  MaterialExpressionScalarParameter_27 {class=MaterialExpressionScalarParameter, properties {parameterName=InternalLight - Bounds - Min}}
  MaterialExpressionScalarParameter_28 {class=MaterialExpressionScalarParameter, properties {defaultValue=1, parameterName=InternalLight - Bounds - Max}}
  MaterialExpressionScalarParameter_29 {
   class=MaterialExpressionScalarParameter
   properties {defaultValue=0.8, group="-       EchoFrost Ice", parameterName=* Internal Layers - Parallax - Color Desaturation, sortPriority=310}
  }
  MaterialExpressionScalarParameter_3 {
   class=MaterialExpressionScalarParameter
   properties {defaultValue=-0.1, group="-       EchoFrost Ice", parameterName=* Internal Layers - Air Bubbles - Depth, sortPriority=200}
  }
  MaterialExpressionScalarParameter_30 {
   class=MaterialExpressionScalarParameter
   properties {defaultValue=3, group="-       EchoFrost Ice", parameterName=* Internal Layers - Air Bubbles - Tiling, sortPriority=200}
  }
  MaterialExpressionScalarParameter_31 {class=MaterialExpressionScalarParameter, properties {defaultValue=10, parameterName=InternalLight - Bounds - Power}}
  MaterialExpressionScalarParameter_33 {class=MaterialExpressionScalarParameter, properties {defaultValue=1000, parameterName=InternalLight - Simple WS Noise - Scale2}}
  MaterialExpressionScalarParameter_34 {class=MaterialExpressionScalarParameter, properties {defaultValue=0.001, parameterName=Speed}}
  MaterialExpressionScalarParameter_35 {
   class=MaterialExpressionScalarParameter
   properties {
    defaultValue=1
    desc=Happens Post Eye Adaption Inverse adjustments
    group="-       EchoFrost Ice"
    parameterName="* Internal Light - Global Emissive (Lumen)"
    sortPriority=100
   }
  }
  MaterialExpressionScalarParameter_36 {
   class=MaterialExpressionScalarParameter
   properties {defaultValue=0.5, group="-       EchoFrost Ice", parameterName=* Ice - SSS - Opacity - BackScatter, sortPriority=60}
  }
  MaterialExpressionScalarParameter_37 {
   class=MaterialExpressionScalarParameter
   properties {defaultValue=9, group="-       EchoFrost Ice", parameterName=* Ice - Thickness - Intensity, sortPriority=85}
  }
  MaterialExpressionScalarParameter_38 {class=MaterialExpressionScalarParameter, properties {defaultValue=1, parameterName=normal boos}}
  MaterialExpressionScalarParameter_39 {
   class=MaterialExpressionScalarParameter
   properties {defaultValue=0.2, group="-       EchoFrost Ice", parameterName=* Ice - SSS - Strength, sortPriority=60}
  }
  MaterialExpressionScalarParameter_4 {
   class=MaterialExpressionScalarParameter
   properties {defaultValue=2, group="-       EchoFrost Ice", parameterName=* Internal Layers - Cracks - Tiling, sortPriority=200}
  }
  MaterialExpressionScalarParameter_40 {
   class=MaterialExpressionScalarParameter
   properties {defaultValue=0.5, group="-       EchoFrost Ice", parameterName=* Ice - Roughness - Surface Noise}
  }
  MaterialExpressionScalarParameter_41 {
   class=MaterialExpressionScalarParameter
   properties {group="-       EchoFrost Ice_Display", parameterName="IceDisplay_Window_Min(Trasnp)", sortPriority=50}
  }
  MaterialExpressionScalarParameter_42 {
   class=MaterialExpressionScalarParameter
   properties {
    defaultValue=-80
    desc="0 = vertex pos will be full transaprent\r\nnegative value = makes surface more opaque\r\npositive value = makes surface more transparent"
    group="-       EchoFrost Ice_Display"
    parameterName=IceDisplay_Depht_SurfaceOpacity
   }
  }
  MaterialExpressionScalarParameter_43 {class=MaterialExpressionScalarParameter, properties {defaultValue=0.25, group="-       EchoFrost Ice", parameterName=* Ice - Roughness}}
  MaterialExpressionScalarParameter_44 {
   class=MaterialExpressionScalarParameter
   properties {desc="0=PixelNormal, 1=VertexNormal", group="-       EchoFrost Ice", parameterName=* Ice - Internal Light - Normal Selection}
  }
  MaterialExpressionScalarParameter_45 {
   class=MaterialExpressionScalarParameter
   properties {defaultValue=2, group="-       EchoFrost Ice", parameterName=* Internal Light - Ice BaseColor Contribution, sortPriority=100}
  }
  MaterialExpressionScalarParameter_46 {
   class=MaterialExpressionScalarParameter
   properties {
    defaultValue=80
    desc=How deep into the ice you can see
    group="-       EchoFrost Ice_Display"
    parameterName=IceDisplay_Depht_LightPenetration
   }
  }
  MaterialExpressionScalarParameter_47 {
   class=MaterialExpressionScalarParameter
   properties {
    desc=(Green Channel from SpecularMSKs and VTX Color)
    group="-       EchoFrost Ice"
    parameterName=* Ice - Thickness - Invert
    sortPriority=82
   }
  }
  MaterialExpressionScalarParameter_48 {
   class=MaterialExpressionScalarParameter
   properties {defaultValue=1, group="-       EchoFrost Ice_Display", parameterName="IceDisplay_Window_Max(Opaque)", sortPriority=50}
  }
  MaterialExpressionScalarParameter_49 {
   class=MaterialExpressionScalarParameter
   properties {defaultValue=70, group="-       EchoFrost Ice_Display", parameterName=IceDisplay_Box_SphereMask_Radius}
  }
  MaterialExpressionScalarParameter_5 {
   class=MaterialExpressionScalarParameter
   properties {defaultValue=1, group="-       EchoFrost Ice", parameterName=* Internal Aurora - Thickness Contribution, sortPriority=310}
  }
  MaterialExpressionScalarParameter_50 {
   class=MaterialExpressionScalarParameter
   properties {defaultValue=0.5, group="-       EchoFrost Ice_Display", parameterName=IceDisplay_Box_SphereMask_Falloff}
  }
  MaterialExpressionScalarParameter_51 {
   class=MaterialExpressionScalarParameter
   properties {defaultValue=0.8, group="-       EchoFrost Ice", parameterName=* Ice - BaseColor - Layers - Contrast, sortPriority=50}
  }
  MaterialExpressionScalarParameter_52 {
   class=MaterialExpressionScalarParameter
   properties {defaultValue=1000, group="-       EchoFrost Ice", parameterName=* Ice - Internal Light - Core Darken - Distance}
  }
  MaterialExpressionScalarParameter_53 {
   class=MaterialExpressionScalarParameter
   properties {defaultValue=3, group="-       EchoFrost Ice", parameterName=* Ice - Internal Light - Core Darken - Power}
  }
  MaterialExpressionScalarParameter_54 {
   class=MaterialExpressionScalarParameter
   properties {defaultValue=0.8, group="-       EchoFrost Ice", parameterName=* Ice - Internal Light - Sun - Power}
  }
  MaterialExpressionScalarParameter_55 {
   class=MaterialExpressionScalarParameter
   properties {defaultValue=0.05, group="-       EchoFrost Ice", parameterName=* Ice - Internal Light - Core Darken - Min}
  }
  MaterialExpressionScalarParameter_56 {
   class=MaterialExpressionScalarParameter
   properties {group="-       EchoFrost Ice_Display", parameterName=IceDisplay_InnerGlow_CracksTint, sortPriority=30}
  }
  MaterialExpressionScalarParameter_57 {
   class=MaterialExpressionScalarParameter
   properties {defaultValue=0.3, group="-       EchoFrost Ice", parameterName=* Internal Light - Global Emissive, sortPriority=100}
  }
  MaterialExpressionScalarParameter_58 {
   class=MaterialExpressionScalarParameter
   properties {defaultValue=2000, group="-       EchoFrost Ice", parameterName=* Internal Aurora - Noise - Initial Scale, sortPriority=310}
  }
  MaterialExpressionScalarParameter_59 {
   class=MaterialExpressionScalarParameter
   properties {defaultValue=0.05, group="-       EchoFrost Ice", parameterName=* Internal Aurora - Flowmap Speed, sortPriority=310}
  }
  MaterialExpressionScalarParameter_6 {
   class=MaterialExpressionScalarParameter
   properties {group="-       EchoFrost Ice", parameterName=* Internal Layers - Air Bubbles - Min, sortPriority=200}
  }
  MaterialExpressionScalarParameter_60 {
   class=MaterialExpressionScalarParameter
   properties {defaultValue=0.4, group="-       EchoFrost Ice", parameterName=* Internal Aurora - Flowmap Scale, sliderMax=1, sortPriority=310}
  }
  MaterialExpressionScalarParameter_61 {
   class=MaterialExpressionScalarParameter
   properties {defaultValue=0.066, group="-       EchoFrost Ice", parameterName=* Internal Aurora - Flowmap Intensity, sortPriority=310}
  }
  MaterialExpressionScalarParameter_62 {
   class=MaterialExpressionScalarParameter
   properties {defaultValue=3, group="-       EchoFrost Ice", parameterName=* Internal Aurora - Scale U, sortPriority=310}
  }
  MaterialExpressionScalarParameter_63 {
   class=MaterialExpressionScalarParameter
   properties {defaultValue=0.2, group="-       EchoFrost Ice", parameterName=* Internal Layers - Parallax - DepthMinimum, sortPriority=304}
  }
  MaterialExpressionScalarParameter_64 {
   class=MaterialExpressionScalarParameter
   properties {defaultValue=-0.14, group="-       EchoFrost Ice", parameterName=* Internal Layers - Parallax - Bias, sortPriority=309}
  }
  MaterialExpressionScalarParameter_65 {
   class=MaterialExpressionScalarParameter
   properties {defaultValue=0.01, group="-       EchoFrost Ice", parameterName=* Internal Aurora - Speed, sortPriority=310}
  }
  MaterialExpressionScalarParameter_66 {
   class=MaterialExpressionScalarParameter
   properties {defaultValue=0.05, group="-       EchoFrost Ice", parameterName=* Internal Aurora - Contrast, sortPriority=310}
  }
  MaterialExpressionScalarParameter_67 {
   class=MaterialExpressionScalarParameter
   properties {group="-       EchoFrost Ice_Display", parameterName=IceDisplay_Box_SphereMask_Contrast}
  }
  MaterialExpressionScalarParameter_69 {
   class=MaterialExpressionScalarParameter
   properties {defaultValue=500, group="-       EchoFrost Ice_Display", parameterName=IceDisplay_FlatTrophy_Depht_FadeDistance}
  }
  MaterialExpressionScalarParameter_7 {
   class=MaterialExpressionScalarParameter
   properties {defaultValue=0.2, group="-       EchoFrost Ice", parameterName=* Ice - Internal Light - Sky Sample - Roughness}
  }
  MaterialExpressionScalarParameter_70 {
   class=MaterialExpressionScalarParameter
   properties {defaultValue=0.2, group="-       EchoFrost Ice", parameterName="* Ice - BaseColor - Tint - Blend (0=Multiply 1=Linear)", sortPriority=50}
  }
  MaterialExpressionScalarParameter_71 {
   class=MaterialExpressionScalarParameter
   properties {defaultValue=2, group="-       EchoFrost Ice_Display", parameterName=IceDisplay_Opacity_Edges}
  }
  MaterialExpressionScalarParameter_72 {class=MaterialExpressionScalarParameter, properties {group="-       EchoFrost Ice_Display", parameterName=IceDisplay_Opacity_Middle}}
  MaterialExpressionScalarParameter_73 {
   class=MaterialExpressionScalarParameter
   properties {defaultValue=15, group="-       EchoFrost Ice", parameterName=* Internal Aurora - Scale V, sortPriority=310}
  }
  MaterialExpressionScalarParameter_74 {
   class=MaterialExpressionScalarParameter
   properties {group="-       EchoFrost Ice", parameterName=* Ice - Internal Light - Z up Normal light - Bias}
  }
  MaterialExpressionScalarParameter_75 {
   class=MaterialExpressionScalarParameter
   properties {defaultValue=1, group="-       EchoFrost Ice", parameterName=* Ice - Internal Light - Z up Normal light - Intensity}
  }
  MaterialExpressionScalarParameter_76 {
   class=MaterialExpressionScalarParameter
   properties {defaultValue=8, desc=8, group="-       EchoFrost Ice", parameterName=* Internal Aurora - Strength, sortPriority=310}
  }
  MaterialExpressionScalarParameter_77 {
   class=MaterialExpressionScalarParameter
   properties {defaultValue=0.5, group="-       EchoFrost Ice", parameterName=* Ice - Specular - Surface Noise}
  }
  MaterialExpressionScalarParameter_78 {class=MaterialExpressionScalarParameter, properties {group="-       EchoFrost Ice_Display", parameterName=IceDisplay_Depht_Fresnel}}
  MaterialExpressionScalarParameter_79 {
   class=MaterialExpressionScalarParameter
   properties {defaultValue=0.32, group="-       EchoFrost Ice_Display", parameterName=IceDisplay_Column_InnerGlow_GradientMiddle, sortPriority=50}
  }
  MaterialExpressionScalarParameter_8 {
   class=MaterialExpressionScalarParameter
   properties {defaultValue=3, group="-       EchoFrost Ice", parameterName=* Ice - Internal Light - Sky Sample - Contrast}
  }
  MaterialExpressionScalarParameter_80 {
   class=MaterialExpressionScalarParameter
   properties {defaultValue=0.2, group="-       EchoFrost Ice_Display", parameterName=IceDisplay_Column_InnerGlow_GradientBase, sortPriority=50}
  }
  MaterialExpressionScalarParameter_81 {
   class=MaterialExpressionScalarParameter
   properties {defaultValue=1.1, group="-       EchoFrost Ice_Display", parameterName=IceDisplay_Column_InnerGlow_GradientTop, sortPriority=50}
  }
  MaterialExpressionScalarParameter_82 {
   class=MaterialExpressionScalarParameter
   properties {defaultValue=1, group="-       EchoFrost Ice_Display", parameterName=IceDisplay_Depht_FresnelPower}
  }
  MaterialExpressionScalarParameter_83 {
   class=MaterialExpressionScalarParameter
   properties {defaultValue=9000, group="-       EchoFrost Ice", parameterName=* Internal Layers - Parallax - Fade Distance, sortPriority=303}
  }
  MaterialExpressionScalarParameter_84 {
   class=MaterialExpressionScalarParameter
   properties {defaultValue=50, group="-       EchoFrost Ice_Display", parameterName=IceDisplay_Depht_ParallaxOpacity}
  }
  MaterialExpressionScalarParameter_85 {
   class=MaterialExpressionScalarParameter
   properties {defaultValue=-0.3, group="-       EchoFrost Ice_Display", parameterName=IceDisplay_FlatTrophy_Depht}
  }
  MaterialExpressionScalarParameter_86 {
   class=MaterialExpressionScalarParameter
   properties {defaultValue=2, group="-       EchoFrost Ice_Display", parameterName=IceDisplay_FlatTrophy_Contrast}
  }
  MaterialExpressionScalarParameter_87 {class=MaterialExpressionScalarParameter, properties {group="-       EchoFrost Ice_Display", parameterName=IceDisplay_FlatTrophy-Close}}
  MaterialExpressionScalarParameter_88 {
   class=MaterialExpressionScalarParameter
   properties {defaultValue=-1, group="-       EchoFrost Ice", parameterName=* Internal Layers - Distortion from Normal, sortPriority=200}
  }
  MaterialExpressionScalarParameter_89 {
   class=MaterialExpressionScalarParameter
   properties {defaultValue=600, group="-       EchoFrost Ice", parameterName=* Internal Aurora - Object Offset Strength, sortPriority=310}
  }
  MaterialExpressionScalarParameter_9 {
   class=MaterialExpressionScalarParameter
   properties {defaultValue=1, group="-       EchoFrost Ice", parameterName=* Internal Layers - Air Bubbles - Max, sortPriority=200}
  }
  MaterialExpressionScalarParameter_90 {class=MaterialExpressionScalarParameter, properties {group="-       EchoFrost Ice_Display", parameterName=IceDisplay_FlatTrophy-Far}}
  MaterialExpressionScalarParameter_99 {
   class=MaterialExpressionScalarParameter
   properties {defaultValue=0.8, group="-       EchoFrost Ice", parameterName=* Internal Layers - Parallax - DepthMaximum, sortPriority=305}
  }
  MaterialExpressionSceneDepth_4 {class=MaterialExpressionSceneDepth, inputs [UVs {from="", output=""}], properties {}}
  MaterialExpressionScreenPosition_3 {class=MaterialExpressionScreenPosition, inputs [], properties {}}
  MaterialExpressionScreenPosition_5 {class=MaterialExpressionScreenPosition, inputs [], properties {}}
  MaterialExpressionSetMaterialAttributes_0 {
   class=MaterialExpressionSetMaterialAttributes
   inputs [
    MaterialAttributes {from="", output=""}
    EmissiveColor {from=MaterialExpressionNamedRerouteUsage_5, output=""}
   ]
   properties {attributeSetTypes [#9]}
  }
  MaterialExpressionSetMaterialAttributes_1 {
   class=MaterialExpressionSetMaterialAttributes
   inputs [
    MaterialAttributes {from=MaterialExpressionGetMaterialAttributes_7, output=MaterialAttributes}
    BaseColor {from=MaterialExpressionQualitySwitch_6, output=""}
    Roughness {from=MaterialExpressionQualitySwitch_11, output=""}
    EmissiveColor {from=MaterialExpressionQualitySwitch_12, output=""}
    SubsurfaceColor {from=MaterialExpressionQualitySwitch_13, output=""}
   ]
   properties {attributeSetTypes [
     #3
     #5
     #9
     #10
    ]}
  }
  MaterialExpressionSetMaterialAttributes_2 {class=MaterialExpressionSetMaterialAttributes, inputs [MaterialAttributes {from="", output=""}]}
  MaterialExpressionSetMaterialAttributes_4 {
   class=MaterialExpressionSetMaterialAttributes
   inputs [
    MaterialAttributes {from=MaterialExpressionNamedRerouteUsage_41, output=""}
    BaseColor {from=MaterialExpressionMaterialProxyReplace_2, output=""}
    Specular {from=MaterialExpressionReroute_32, output=""}
    Roughness {from=MaterialExpressionReroute_17, output=""}
    SubsurfaceColor {from=MaterialExpressionReroute_28, output=""}
    Opacity {from=MaterialExpressionReroute_36, output=""}
    EmissiveColor {from=MaterialExpressionReroute_26, output=""}
    Normal {from=MaterialExpressionReroute_37, output=""}
   ]
   properties {attributeSetTypes [
     #3
     #4
     #5
     #10
     #2
     #9
     #6
    ]}
  }
  MaterialExpressionSetMaterialAttributes_5 {
   class=MaterialExpressionSetMaterialAttributes
   inputs [
    MaterialAttributes {from="", output=""}
    BaseColor {from=MaterialExpressionReroute_5, output=""}
    Metallic {from=MaterialExpressionConstant_3, output=""}
    Specular {from=MaterialExpressionReroute_30, output=""}
    Roughness {from=MaterialExpressionReroute_66, output=""}
    SubsurfaceColor {from="", output=""}
    Normal {from="", output=""}
    Opacity {from="", output=""}
    EmissiveColor {from="", output=""}
    OpacityMask {from=MaterialExpressionGetMaterialAttributes_0, output=OpacityMask}
   ]
   properties {attributeSetTypes [
     #3
     #7
     #4
     #5
     #10
     #6
     #2
     #9
     #1
    ]}
  }
  MaterialExpressionShaderStageSwitch_0 {
   class=MaterialExpressionShaderStageSwitch
   inputs [
    PixelShader {from=MaterialExpressionVertexInterpolator_1, output=PS}
    VertexShader {from=MaterialExpressionReroute_7, output=""}
   ]
   properties {}
  }
  MaterialExpressionShaderStageSwitch_2 {
   class=MaterialExpressionShaderStageSwitch
   inputs [
    PixelShader {from=MaterialExpressionVertexInterpolator_3, output=PS}
    VertexShader {from=MaterialExpressionReroute_12, output=""}
   ]
   properties {}
  }
  MaterialExpressionShaderStageSwitch_3 {
   class=MaterialExpressionShaderStageSwitch
   inputs [
    PixelShader {from=MaterialExpressionVertexInterpolator_4, output=PS}
    VertexShader {from=MaterialExpressionReroute_13, output=""}
   ]
   properties {}
  }
  MaterialExpressionShadingPathSwitch_0 {
   class=MaterialExpressionShadingPathSwitch
   inputs [
    Default {from=MaterialExpressionStaticSwitchParameter_4, output=""}
    Deferred {from="", output=""}
    Forward {from="", output=""}
    Mobile {from=MaterialExpressionReroute_52, output=""}
   ]
   properties {}
  }
  MaterialExpressionShadingPathSwitch_1 {
   class=MaterialExpressionShadingPathSwitch
   inputs [
    Default {from=MaterialExpressionStaticSwitchParameter_0, output=""}
    Deferred {from="", output=""}
    Forward {from="", output=""}
    Mobile {from=MaterialExpressionReroute_27, output=""}
   ]
   properties {}
  }
  MaterialExpressionShadingPathSwitch_2 {
   class=MaterialExpressionShadingPathSwitch
   inputs [
    Default {from=MaterialExpressionQualitySwitch_2, output=""}
    Deferred {from="", output=""}
    Forward {from="", output=""}
    Mobile {from=MaterialExpressionReroute_72, output=""}
   ]
   properties {}
  }
  MaterialExpressionShadingPathSwitch_3 {
   class=MaterialExpressionShadingPathSwitch
   inputs [
    Default {from=MaterialExpressionQualitySwitch_3, output=""}
    Deferred {from="", output=""}
    Forward {from="", output=""}
    Mobile {from="", output=""}
   ]
   properties {}
  }
  MaterialExpressionShadingPathSwitch_5 {
   class=MaterialExpressionShadingPathSwitch
   inputs [
    Default {from=MaterialExpressionWorldPosition_4, output=Z}
    Deferred {from="", output=""}
    Forward {from="", output=""}
    Mobile {from=MaterialExpressionTruncateLWC_1, output=""}
   ]
   properties {}
  }
  MaterialExpressionSign_0 {class=MaterialExpressionSign, inputs [None {from=MaterialExpressionComponentMask_7, output=""}], properties {}}
  MaterialExpressionSine_1 {class=MaterialExpressionSine, inputs [None {from=MaterialExpressionDotProduct_3}]}
  MaterialExpressionSine_2 {class=MaterialExpressionSine, inputs [None {from=MaterialExpressionTime_0}]}
  MaterialExpressionSine_3 {class=MaterialExpressionSine, inputs [None {from=MaterialExpressionDotProduct_5}]}
  MaterialExpressionSkyAtmosphereLightDirection_0 {class=MaterialExpressionSkyAtmosphereLightDirection, inputs [], properties {}}
  MaterialExpressionSkyLightEnvMapSample_0 {
   class=MaterialExpressionSkyLightEnvMapSample
   inputs [
    Direction {from=MaterialExpressionReflectionVectorWS_0, output=""}
    Roughness {from=MaterialExpressionScalarParameter_7, output=""}
   ]
   properties {}
  }
  MaterialExpressionSmoothStep_0 {
   class=MaterialExpressionSmoothStep
   inputs [
    Min {from=MaterialExpressionScalarParameter_42, output=""}
    Max {from=MaterialExpressionScalarParameter_46, output=""}
    Value {from=MaterialExpressionLinearInterpolate_32, output=""}
   ]
   properties {}
  }
  MaterialExpressionSmoothStep_1 {
   class=MaterialExpressionSmoothStep
   inputs [
    Min {from=MaterialExpressionScalarParameter_48, output=""}
    Max {from=MaterialExpressionScalarParameter_41, output=""}
    Value {from=MaterialExpressionMultiply_33, output=""}
   ]
   properties {}
  }
  MaterialExpressionSmoothStep_2 {
   class=MaterialExpressionSmoothStep
   inputs [
    Min {from="", output=""}
    Max {from="", output=""}
    Value {from="", output=""}
   ]
   properties {}
  }
  MaterialExpressionSmoothStep_7 {
   class=MaterialExpressionSmoothStep
   inputs [
    Min {from=MaterialExpressionScalarParameter_81, output=""}
    Max {from=MaterialExpressionScalarParameter_79, output=""}
    Value {from=MaterialExpressionReroute_70, output=""}
   ]
   properties {constMin=1}
  }
  MaterialExpressionSmoothStep_8 {
   class=MaterialExpressionSmoothStep
   inputs [
    Min {from=MaterialExpressionScalarParameter_80, output=""}
    Max {from=MaterialExpressionScalarParameter_79, output=""}
    Value {from=MaterialExpressionReroute_70, output=""}
   ]
   properties {}
  }
  MaterialExpressionSphereMask_0 {
   class=MaterialExpressionSphereMask
   inputs [
    A {from=MaterialExpressionPixelDepth_0, output=R}
    B {from=MaterialExpressionConstant_5, output=""}
    Radius {from=MaterialExpressionScalarParameter_83, output=""}
    Hardness {from="", output=""}
   ]
   properties {hardnessPercent=10}
  }
  MaterialExpressionSphereMask_1 {
   class=MaterialExpressionSphereMask
   inputs [
    A {from=MaterialExpressionShaderStageSwitch_3, output=""}
    B {from=MaterialExpressionVectorParameter_11, output=RGB}
    Radius {from=MaterialExpressionScalarParameter_49, output=""}
    Hardness {from="", output=""}
   ]
   properties {hardnessPercent=0}
  }
  MaterialExpressionSquareRoot_0 {class=MaterialExpressionSquareRoot, inputs [None {from=MaterialExpressionAbs_2, output=""}], properties {}}
  MaterialExpressionStaticBoolParameter_1 {
   class=MaterialExpressionStaticBoolParameter
   inputs []
   properties {
    desc="Affects,\r\n* Basecolor Fresnel is masked by Thickness\r\n* Internal Aurora\r\n* SSS Color Strength is masked by Thickness\r\n* SSS Opacity is masked by Thickness\r\n"
    group="-       EchoFrost Ice"
    parameterName=* Ice - Use Thickness
    sortPriority=-1
   }
  }
  MaterialExpressionStaticBoolParameter_3 {
   class=MaterialExpressionStaticBoolParameter
   inputs []
   properties {
    desc="Affects,\r\n* Basecolor Fresnel is masked by Thickness\r\n* Internal Aurora\r\n* SSS Color Strength is masked by Thickness\r\n* SSS Opacity is masked by Thickness\r\n"
    group="-       EchoFrost Ice"
    parameterName=* Ice - Use Thickness
    sortPriority=-1
   }
  }
  MaterialExpressionStaticBoolParameter_4 {
   class=MaterialExpressionStaticBoolParameter
   inputs []
   properties {
    desc="Affects,\r\n* Basecolor Fresnel is masked by Thickness\r\n* Internal Aurora\r\n* SSS Color Strength is masked by Thickness\r\n* SSS Opacity is masked by Thickness\r\n"
    group="-       EchoFrost Ice"
    parameterName=* Ice - Use Thickness
    sortPriority=-1
   }
  }
  MaterialExpressionStaticBool_1 {class=MaterialExpressionStaticBool, properties {value=true}}
  MaterialExpressionStaticSwitchParameter_0 {
   class=MaterialExpressionStaticSwitchParameter
   inputs [
    True {from=MaterialExpressionAdd_22, output=""}
    False {from=MaterialExpressionReroute_47, output=""}
   ]
   properties {defaultValue=true, group="-       EchoFrost Ice", parameterName=* Ice - Use Internal Aurora, sortPriority=-1}
  }
  MaterialExpressionStaticSwitchParameter_1 {
   class=MaterialExpressionStaticSwitchParameter
   inputs [
    True {from=MaterialExpressionMaterialFunctionCall_26, output=Result}
    False {from=MaterialExpressionConstant_0, output=""}
   ]
   properties {defaultValue=true, group="-       EchoFrost Ice", parameterName=* Ice - Use Internal Aurora Virtual Plane, sortPriority=-1}
  }
  MaterialExpressionStaticSwitchParameter_2 {
   class=MaterialExpressionStaticSwitchParameter
   inputs [
    True {from=MaterialExpressionNamedRerouteUsage_68, output=""}
    False {from=MaterialExpressionReroute_71, output=""}
   ]
   properties {group="-       EchoFrost Ice_Display", parameterName=IceDisplay_IsTransluscent, sortPriority=0}
  }
  MaterialExpressionStaticSwitchParameter_3 {
   class=MaterialExpressionStaticSwitchParameter
   inputs [
    True {from=MaterialExpressionStaticSwitchParameter_2, output=""}
    False {from=MaterialExpressionReroute_71, output=""}
   ]
   properties {group="-       EchoFrost Ice_Display", parameterName=Use Ice Displays, sortPriority=-10}
  }
  MaterialExpressionStaticSwitchParameter_4 {
   class=MaterialExpressionStaticSwitchParameter
   inputs [
    True {from=MaterialExpressionAdd_16, output=""}
    False {from=MaterialExpressionReroute_51, output=""}
   ]
   properties {defaultValue=true, group="-       EchoFrost Ice", parameterName=* Ice - Use Internal Aurora, sortPriority=-1}
  }
  MaterialExpressionStaticSwitchParameter_5 {
   class=MaterialExpressionStaticSwitchParameter
   inputs [
    True {from=MaterialExpressionAdd_21, output=""}
    False {from=MaterialExpressionReroute_58, output=""}
   ]
   properties {group="-       EchoFrost Ice_Display", parameterName=Use Ice Displays, sortPriority=-10}
  }
  MaterialExpressionStaticSwitchParameter_6 {
   class=MaterialExpressionStaticSwitchParameter
   inputs [
    True {from=MaterialExpressionAdd_29, output=""}
    False {from=MaterialExpressionReroute_61, output=""}
   ]
   properties {group="-       EchoFrost Ice_Display", parameterName=Use Ice Displays, sortPriority=-10}
  }
  MaterialExpressionStaticSwitchParameter_7 {
   class=MaterialExpressionStaticSwitchParameter
   inputs [
    True {from=MaterialExpressionLinearInterpolate_27, output=""}
    False {from=MaterialExpressionReroute_68, output=""}
   ]
   properties {group="-       EchoFrost Ice_Display", parameterName=IceDisplay_IsTransluscent, sortPriority=0}
  }
  MaterialExpressionStaticSwitch_2 {
   class=MaterialExpressionStaticSwitch
   inputs [
    True {from=MaterialExpressionLinearInterpolate_0, output=""}
    False {from=MaterialExpressionMultiply_56, output=""}
    Value {from=MaterialExpressionStaticBoolParameter_1, output=""}
   ]
   properties {}
  }
  MaterialExpressionStaticSwitch_3 {
   class=MaterialExpressionStaticSwitch
   inputs [
    True {from=MaterialExpressionMultiply_0, output=""}
    False {from=MaterialExpressionScalarParameter_39, output=""}
    Value {from=MaterialExpressionStaticBoolParameter_3, output=""}
   ]
   properties {}
  }
  MaterialExpressionStaticSwitch_4 {
   class=MaterialExpressionStaticSwitch
   inputs [
    True {from=MaterialExpressionLinearInterpolate_11, output=""}
    False {from=MaterialExpressionReroute_34, output=""}
    Value {from=MaterialExpressionStaticBoolParameter_4, output=""}
   ]
   properties {}
  }
  MaterialExpressionSubtract_0 {
   class=MaterialExpressionSubtract
   inputs [
    A {from=MaterialExpressionNamedRerouteUsage_79}
    B {from=MaterialExpressionScalarParameter_12}
   ]
  }
  MaterialExpressionSubtract_1 {
   class=MaterialExpressionSubtract
   inputs [
    A {from=MaterialExpressionVectorParameter_3, output=RGB}
    B {from=MaterialExpressionObjectPositionWS_0}
   ]
  }
  MaterialExpressionSubtract_10 {
   class=MaterialExpressionSubtract
   inputs [
    A {from=MaterialExpressionConvert_2}
    B {from=MaterialExpressionMax_6}
   ]
  }
  MaterialExpressionSubtract_2 {
   class=MaterialExpressionSubtract
   inputs [
    A {from=MaterialExpressionConvert_2}
    B {from=MaterialExpressionMax_4}
   ]
  }
  MaterialExpressionSubtract_3 {
   class=MaterialExpressionSubtract
   inputs [
    A {from=MaterialExpressionLocalPosition_0, output=XYZ}
    B {from=MaterialExpressionMaterialFunctionCall_15, output=Local Bounds Minimum}
   ]
  }
  MaterialExpressionSubtract_4 {
   class=MaterialExpressionSubtract
   inputs [
    A {from=MaterialExpressionLocalPosition_1, output=XYZ}
    B {from=MaterialExpressionVectorParameter_4, output=RGB}
   ]
  }
  MaterialExpressionSubtract_7 {
   class=MaterialExpressionSubtract
   inputs [
    A {from=MaterialExpressionShadingPathSwitch_5}
    B {from=MaterialExpressionAdd_31}
   ]
  }
  MaterialExpressionSubtract_8 {
   class=MaterialExpressionSubtract
   inputs [
    A {from=MaterialExpressionSceneDepth_4, output=R}
    B {from=MaterialExpressionPixelDepth_3, output=R}
   ]
  }
  MaterialExpressionSubtract_9 {
   class=MaterialExpressionSubtract
   inputs [
    A {from=MaterialExpressionConvert_2}
    B {from=MaterialExpressionMax_5}
   ]
  }
  MaterialExpressionSwitch_1 {
   class=MaterialExpressionSwitch
   inputs [
    SwitchValue {from=MaterialExpressionScalarParameter_15, output=""}
    Default {from="", output=""}
    Column {from=MaterialExpressionPower_14, output=""}
    Window {from=MaterialExpressionSmoothStep_1, output=""}
    Box {from=MaterialExpressionMaterialFunctionCall_3, output=Result}
   ]
   properties {inputs [
     {
      inputName=Column
      input {
       expression=&self:MaterialExpressionPower_14
       inputName=None
       mask=0
       maskA=0
       maskB=0
       maskG=0
       maskR=0
       outputIndex=0
      }
     }
     {
      inputName=Window
      input {
       expression=&self:MaterialExpressionSmoothStep_1
       inputName=None
       mask=0
       maskA=0
       maskB=0
       maskG=0
       maskR=0
       outputIndex=0
      }
     }
     {
      inputName=Box
      input {
       expression=&self:MaterialExpressionMaterialFunctionCall_3
       inputName=None
       mask=0
       maskA=0
       maskB=0
       maskG=0
       maskR=0
       outputIndex=0
      }
     }
    ]}
  }
  MaterialExpressionTextureCoordinate_0 {class=MaterialExpressionTextureCoordinate}
  MaterialExpressionTextureCoordinate_1 {class=MaterialExpressionTextureCoordinate}
  MaterialExpressionTextureCoordinate_2 {class=MaterialExpressionTextureCoordinate}
  MaterialExpressionTextureCoordinate_3 {class=MaterialExpressionTextureCoordinate}
  MaterialExpressionTextureObjectParameter_0 {
   class=MaterialExpressionTextureObjectParameter
   inputs [
    UVs {from="", output=""}
    Apply View MipBias {from="", output=""}
   ]
   properties {
    group="-       EchoFrost Ice"
    parameterName=* Internal Layers - Distortion - Normal
    samplerType=SAMPLERTYPE_Normal
    sortPriority=200
    texture=/Game/Environments/Asteria/Terrain/Textures/T_Asteria_RockCliff_02_N
   }
  }
  MaterialExpressionTextureObjectParameter_1 {
   class=MaterialExpressionTextureObjectParameter
   inputs [
    UVs {from="", output=""}
    Apply View MipBias {from="", output=""}
   ]
   properties {
    group="-       EchoFrost Ice"
    parameterName=* Internal Layers - Distortion - Height
    sortPriority=200
    texture=/Game/Environments/Asteria/Terrain/Textures/T_Asteria_RockCliff_02_D
   }
  }
  MaterialExpressionTextureObjectParameter_3 {
   class=MaterialExpressionTextureObjectParameter
   inputs [
    UVs {from="", output=""}
    Apply View MipBias {from="", output=""}
   ]
   properties {
    group="-       EchoFrost Ice"
    parameterName=* Internal Layers - Air Bubbles - Texture
    samplerType=SAMPLERTYPE_Masks
    sortPriority=200
    texture=/Hera/Sets/EchoFrost/Textures/T_EchoFrost_Ice_Layers_MSK
   }
  }
  MaterialExpressionTextureObjectParameter_4 {
   class=MaterialExpressionTextureObjectParameter
   inputs [
    UVs {from="", output=""}
    Apply View MipBias {from="", output=""}
   ]
   properties {
    group="-       EchoFrost Ice"
    parameterName=* Internal Layers - Cracks - Texture
    samplerType=SAMPLERTYPE_Masks
    sortPriority=200
    texture=/Hera/Sets/EchoFrost/Textures/T_EchoFrost_Ice_Cracks_MSK
   }
  }
  MaterialExpressionTextureObjectParameter_5 {
   class=MaterialExpressionTextureObjectParameter
   inputs [
    UVs {from="", output=""}
    Apply View MipBias {from="", output=""}
   ]
   properties {
    group="-       EchoFrost Ice"
    parameterName=* Internal Layers - Cracks - Texture
    samplerType=SAMPLERTYPE_Masks
    sortPriority=200
    texture=&T_EchoFrost_Ice_Cracks_MSK
   }
  }
  MaterialExpressionTextureObject_4 {
   class=MaterialExpressionTextureObject
   inputs []
   properties {texture=/Game/Weapons/FORT_Melee/Pickaxe_ChicleVeil/Textures/T_ChicleVeil_Crystal}
  }
  MaterialExpressionTextureSampleParameter2D_0 {
   class=MaterialExpressionTextureSampleParameter2D
   inputs [
    UVs {from="", output=""}
    Apply View MipBias {from="", output=""}
   ]
   properties {
    group="-       EchoFrost Ice"
    parameterName=Normal
    samplerType=SAMPLERTYPE_Normal
    texture=&FlatNormalBlue
   }
  }
  MaterialExpressionTextureSampleParameter2D_2 {
   class=MaterialExpressionTextureSampleParameter2D
   inputs [
    UVs {from="", output=""}
    Apply View MipBias {from="", output=""}
   ]
   properties {
    group="-       EchoFrost Ice"
    parameterName=Specular
    samplerType=SAMPLERTYPE_Masks
    texture=/Game/Global/Textures/Default/Blanks/Temp_specular_Mask
   }
  }
  MaterialExpressionTextureSampleParameter2D_3 {
   class=MaterialExpressionTextureSampleParameter2D
   inputs [
    UVs {from="", output=""}
    Apply View MipBias {from="", output=""}
   ]
   properties {
    group="-       EchoFrost Ice"
    parameterName=Diffuse
    samplerType=SAMPLERTYPE_Masks
    texture=/Game/Global/Textures/Default/Blanks/T_Blank_Mask_Blue_R0G0B1A1
   }
  }
  MaterialExpressionTextureSampleParameter2D_4 {
   class=MaterialExpressionTextureSampleParameter2D
   inputs [
    UVs {from=MaterialExpressionNamedRerouteUsage_95, output=""}
    "DDX(UVs)" {from=MaterialExpressionDDX_1, output=""}
    "DDY(UVs)" {from=MaterialExpressionDDY_1, output=""}
    Apply View MipBias {from="", output=""}
   ]
   properties {
    group="-       EchoFrost Ice"
    mipValueMode=TMVM_Derivative
    parameterName=* Internal Layers - Parallax - Color Texture
    sortPriority=301
    texture=/Game/Global/Textures/Default/Masks_and_Gradients/T_MF_Crystal
   }
  }
  MaterialExpressionTextureSampleParameter2D_5 {
   class=MaterialExpressionTextureSampleParameter2D
   inputs [
    UVs {from=MaterialExpressionLinearInterpolate_29, output=""}
    "DDX(UVs)" {from=MaterialExpressionDDX_0, output=""}
    "DDY(UVs)" {from=MaterialExpressionDDY_0, output=""}
    Apply View MipBias {from="", output=""}
   ]
   properties {
    group="-       EchoFrost Ice"
    mipValueMode=TMVM_Derivative
    parameterName=* Internal Layers - Parallax - DepthTexture
    samplerType=SAMPLERTYPE_Masks
    sortPriority=300
    texture=/Hera/Sets/EchoFrost/Textures/T_EchoFrost_IceSurface_Facets_A_S
   }
  }
  MaterialExpressionTextureSample_1 {
   class=MaterialExpressionTextureSample
   inputs [
    UVs {from=MaterialExpressionPanner_0}
    Tex {from=""}
    Apply View MipBias {from=""}
   ]
   properties {samplerType=SAMPLERTYPE_LinearColor, texture=/Game/Effects/Fort_Effects/Textures/Plasma/T_Turbulence_AdditiveWarp_Linear}
  }
  MaterialExpressionTextureSample_11 {
   class=MaterialExpressionTextureSample
   inputs [
    UVs {from=MaterialExpressionMaterialFunctionCall_25, output=UV2}
    Tex {from=MaterialExpressionTextureObject_4}
    Apply View MipBias {from=""}
   ]
   properties {texture=/Game/Effects/Fort_Effects/Textures/Tile/T_ShieldCaustics_Tile_Chromatic}
  }
  MaterialExpressionTextureSample_12 {
   class=MaterialExpressionTextureSample
   inputs [
    UVs {from=MaterialExpressionMaterialFunctionCall_25, output=UV1}
    Tex {from=MaterialExpressionTextureObject_4}
    Apply View MipBias {from=""}
   ]
   properties {texture=&T_ShieldCaustics_Tile_Chromatic}
  }
  MaterialExpressionTextureSample_3 {
   class=MaterialExpressionTextureSample
   inputs [
    UVs {from=MaterialExpressionReroute_62}
    Tex {from=MaterialExpressionTextureObjectParameter_5}
    Apply View MipBias {from=""}
   ]
   properties {samplerSource=SSM_Wrap_WorldGroupSettings, samplerType=SAMPLERTYPE_Masks, texture=&T_EchoFrost_Ice_Cracks_MSK}
  }
  MaterialExpressionTextureSample_4 {
   class=MaterialExpressionTextureSample
   inputs [
    UVs {from=MaterialExpressionMultiply_4}
    Tex {from=MaterialExpressionTextureObjectParameter_4}
    Apply View MipBias {from=""}
   ]
   properties {samplerType=SAMPLERTYPE_Masks, texture=/Game/Environments/Asteria/Terrain/Textures/T_Asteria_Ice_Cracks_M}
  }
  MaterialExpressionTextureSample_5 {
   class=MaterialExpressionTextureSample
   inputs [
    UVs {from=MaterialExpressionReroute_59}
    Tex {from=MaterialExpressionTextureObjectParameter_3}
    Apply View MipBias {from=""}
   ]
   properties {texture=/Game/Environments/Helios/Sets/GlacierHotel/Textures/T_GlacierHotel_Roof_Snow_Gravel_A_B_D}
  }
  MaterialExpressionTextureSample_6 {
   class=MaterialExpressionTextureSample
   inputs [
    UVs {from=MaterialExpressionMaterialFunctionCall_2, output=UVs}
    Tex {from=MaterialExpressionTextureObjectParameter_0}
    "DDX(UVs)" {from=MaterialExpressionMaterialFunctionCall_2, output=UVs}
    "DDY(UVs)" {from=MaterialExpressionMaterialFunctionCall_2, output=UVs}
    Apply View MipBias {from=""}
   ]
   properties {mipValueMode=TMVM_Derivative, samplerType=SAMPLERTYPE_Normal, texture=/Game/Environments/Asteria/Terrain/Textures/T_Asteria_RockBase_01_D}
  }
  MaterialExpressionTextureSample_8 {
   class=MaterialExpressionTextureSample
   inputs [
    UVs {from=MaterialExpressionMaterialFunctionCall_2, output=UVs}
    Tex {from=MaterialExpressionTextureObjectParameter_1}
    "DDX(UVs)" {from=MaterialExpressionMaterialFunctionCall_2, output=UVs}
    "DDY(UVs)" {from=MaterialExpressionMaterialFunctionCall_2, output=UVs}
    Apply View MipBias {from=""}
   ]
   properties {mipValueMode=TMVM_Derivative, texture=&T_Asteria_RockCliff_02_D}
  }
  MaterialExpressionTextureSample_9 {
   class=MaterialExpressionTextureSample
   inputs [
    UVs {from=MaterialExpressionMaterialFunctionCall_25, output=UV3}
    Tex {from=MaterialExpressionTextureObject_4}
    Apply View MipBias {from=""}
   ]
   properties {texture=&T_ShieldCaustics_Tile_Chromatic}
  }
  MaterialExpressionTime_0 {class=MaterialExpressionTime, properties {period=1}}
  MaterialExpressionTransform_0 {
   class=MaterialExpressionTransform
   inputs [None {from=MaterialExpressionConstant3Vector_2}]
   properties {transformSourceType=TRANSFORMSOURCE_View}
  }
  MaterialExpressionTransform_1 {
   class=MaterialExpressionTransform
   inputs [None {from=MaterialExpressionVertexNormalWS_2}]
   properties {transformSourceType=TRANSFORMSOURCE_World, transformType=TRANSFORM_Camera}
  }
  MaterialExpressionTransform_10 {class=MaterialExpressionTransform, inputs [None {from=MaterialExpressionMaterialFunctionCall_35, output=Result}]}
  MaterialExpressionTransform_4 {class=MaterialExpressionTransform, inputs [None {from=MaterialExpressionGetMaterialAttributes_4, output=Normal}]}
  MaterialExpressionTransform_6 {
   class=MaterialExpressionTransform
   inputs [None {from=MaterialExpressionMaterialFunctionCall_12, output=World Normals}]
   properties {transformSourceType=TRANSFORMSOURCE_World, transformType=TRANSFORM_Tangent}
  }
  MaterialExpressionTruncateLWC_1 {class=MaterialExpressionTruncateLWC, inputs [None {from=MaterialExpressionWorldPosition_4, output=Z}], properties {}}
  MaterialExpressionTwoSidedSign_3 {class=MaterialExpressionTwoSidedSign, inputs [], properties {}}
  MaterialExpressionVectorParameter_0 {
   class=MaterialExpressionVectorParameter
   properties {
    group="-       EchoFrost Ice_Display"
    parameterName=IceDisplay_UplightTrophyColor
    sortPriority=30
    defaultValue=<rgba 0.4189 0.4055 0.6771 1>
   }
  }
  MaterialExpressionVectorParameter_1 {
   class=MaterialExpressionVectorParameter
   properties {
    group="-       EchoFrost Ice"
    parameterName=* Internal Aurora - Gradient - Bottom Color
    sortPriority=351
    defaultValue=<rgba 0.7758 0.1301 0.8796 1>
   }
  }
  MaterialExpressionVectorParameter_10 {
   class=MaterialExpressionVectorParameter
   properties {group="-       EchoFrost Ice", parameterName=* Ice - BaseColor - Tint, sortPriority=50, defaultValue=<rgba 0.049 0.2706 0.4087 1>}
  }
  MaterialExpressionVectorParameter_11 {
   class=MaterialExpressionVectorParameter
   properties {group="-       EchoFrost Ice_Display", parameterName=IceDisplay_Box_SphereMask_Position, defaultValue=<rgba 0 0 50 1>}
  }
  MaterialExpressionVectorParameter_12 {class=MaterialExpressionVectorParameter, properties {parameterName=IceDisplay_Box_BoxMask_Axis, defaultValue=<rgba 50 100 100 1>}}
  MaterialExpressionVectorParameter_13 {
   class=MaterialExpressionVectorParameter
   properties {group="-       EchoFrost Ice", parameterName=* Ice - SSS - Towards, sortPriority=55, defaultValue=<rgba 0.1042 0.5957 1 1>}
  }
  MaterialExpressionVectorParameter_14 {
   class=MaterialExpressionVectorParameter
   properties {
    group="-       EchoFrost Ice"
    parameterName="• Ice - Thickness - Baked Channel Weights"
    sortPriority=80
    channelNames {a="", b=Thickness, g=Curvature, r=AO}
    defaultValue=<rgba 0 0 1 0>
   }
  }
  MaterialExpressionVectorParameter_15 {
   class=MaterialExpressionVectorParameter
   properties {group="-       EchoFrost Ice", parameterName=* Ice - SSS - BackScatter, sortPriority=55, defaultValue=<rgba 0.1448 0.3779 0.4271 1>}
  }
  MaterialExpressionVectorParameter_16 {
   class=MaterialExpressionVectorParameter
   properties {group="-       EchoFrost Ice_Display", parameterName=IceDisplay_Box_SphereMask_StretchAxis, defaultValue=<rgba 2.5 1 1 1>}
  }
  MaterialExpressionVectorParameter_19 {
   class=MaterialExpressionVectorParameter
   properties {group="-       EchoFrost Ice_Display", parameterName=IceDisplay_InnerGlow_Color1, sortPriority=30, defaultValue=<rgba 0.0549 0 0.1 1>}
  }
  MaterialExpressionVectorParameter_2 {
   class=MaterialExpressionVectorParameter
   properties {
    group="-       EchoFrost Ice"
    parameterName=* Internal Aurora - Gradient - Top Color
    sortPriority=350
    defaultValue=<rgba 0.1475 0.7454 0.206 1>
   }
  }
  MaterialExpressionVectorParameter_20 {
   class=MaterialExpressionVectorParameter
   properties {group="-       EchoFrost Ice_Display", parameterName=IceDisplay_InnerGlow_Color2, sortPriority=30, defaultValue=<rgba 0.2 0.054 0.0667 1>}
  }
  MaterialExpressionVectorParameter_22 {
   class=MaterialExpressionVectorParameter
   properties {
    group="-       EchoFrost Ice"
    parameterName="EchoFrost - Scalability (Medium)"
    sortPriority=0
    channelNames {a=SSS Multiply, b=Emissive Multiply, g=Roughness Multiply, r=BaseColor Multiply}
    defaultValue=<rgba 1 1 1 1>
   }
  }
  MaterialExpressionVectorParameter_23 {
   class=MaterialExpressionVectorParameter
   properties {
    group="-       EchoFrost Ice"
    parameterName="EchoFrost - Scalability (Low)"
    sortPriority=0
    channelNames {a=SSS Multiply, b=Emissive Multiply, g=Roughness Multiply, r=BaseColor Multiply}
    defaultValue=<rgba 1 1 1 1>
   }
  }
  MaterialExpressionVectorParameter_24 {
   class=MaterialExpressionVectorParameter
   properties {
    group="-       EchoFrost Ice"
    parameterName="EchoFrost - Scalability (Medium)"
    sortPriority=0
    channelNames {a=SSS Multiply, b=Emissive Multiply, g=Roughness Multiply, r=BaseColor Multiply}
    defaultValue=<rgba 1 1 1 1>
   }
  }
  MaterialExpressionVectorParameter_25 {
   class=MaterialExpressionVectorParameter
   properties {
    group="-       EchoFrost Ice"
    parameterName="EchoFrost - Scalability (High)"
    sortPriority=0
    channelNames {a=SSS Multiply, b=Emissive Multiply, g=Roughness Multiply, r=BaseColor Multiply}
    defaultValue=<rgba 1 1 1 1>
   }
  }
  MaterialExpressionVectorParameter_3 {
   class=MaterialExpressionVectorParameter
   properties {
    group="-       EchoFrost Ice"
    parameterName=* Ice - Internal Light - Shard - Location WS
    defaultValue=<rgba -61131.8828 -27260.8809 2912.9233 0>
   }
  }
  MaterialExpressionVectorParameter_4 {class=MaterialExpressionVectorParameter, properties {parameterName=IceDisplay_Box_UV_Offset, defaultValue=<rgba -101.5 -38 6 1>}}
  MaterialExpressionVectorParameter_5 {class=MaterialExpressionVectorParameter, properties {parameterName=IceDisplay_Box_UV_Falloff, defaultValue=<rgba 205 77 92 1>}}
  MaterialExpressionVectorParameter_6 {
   class=MaterialExpressionVectorParameter
   properties {group="-       EchoFrost Ice", parameterName=* Internal Layers - Distortion - Scale, sortPriority=200, defaultValue=<rgba 600 600 600 1>}
  }
  MaterialExpressionVectorParameter_7 {
   class=MaterialExpressionVectorParameter
   properties {group="-       EchoFrost Ice_Display", parameterName=IceDisplay_FlatTrophyColor, sortPriority=30, defaultValue=<rgba 0.0239 0.02 0.1 1>}
  }
  MaterialExpressionVectorParameter_8 {
   class=MaterialExpressionVectorParameter
   properties {
    group="-       EchoFrost Ice"
    parameterName=* Internal Layers - Air Bubbles - BaseColor
    sortPriority=200
    defaultValue=<rgba 0.1802 0.3374 0.4271 1>
   }
  }
  MaterialExpressionVectorParameter_9 {
   class=MaterialExpressionVectorParameter
   properties {group="-       EchoFrost Ice", parameterName=* Internal Layers - Cracks - BaseColor, sortPriority=200, defaultValue=<rgba 0.6198 0.7685 1 1>}
  }
  MaterialExpressionVertexColor_0 {class=MaterialExpressionVertexColor, inputs [], properties {}}
  MaterialExpressionVertexInterpolator_0 {class=MaterialExpressionVertexInterpolator, inputs [VS {from=MaterialExpressionComponentMask_2, output=""}], properties {}}
  MaterialExpressionVertexInterpolator_1 {class=MaterialExpressionVertexInterpolator, inputs [VS {from=MaterialExpressionReroute_7, output=""}], properties {}}
  MaterialExpressionVertexInterpolator_3 {class=MaterialExpressionVertexInterpolator, inputs [VS {from=MaterialExpressionReroute_12, output=""}], properties {}}
  MaterialExpressionVertexInterpolator_4 {class=MaterialExpressionVertexInterpolator, inputs [VS {from=MaterialExpressionReroute_13, output=""}], properties {}}
  MaterialExpressionVertexNormalWS_0 {class=MaterialExpressionVertexNormalWS, inputs [], properties {}}
  MaterialExpressionVertexNormalWS_1 {class=MaterialExpressionVertexNormalWS, inputs [], properties {}}
  MaterialExpressionVertexNormalWS_2 {class=MaterialExpressionVertexNormalWS, inputs [], properties {}}
  MaterialExpressionVertexNormalWS_3 {class=MaterialExpressionVertexNormalWS, inputs [], properties {}}
  MaterialExpressionVertexNormalWS_5 {class=MaterialExpressionVertexNormalWS, inputs [], properties {}}
  MaterialExpressionVertexNormalWS_7 {class=MaterialExpressionVertexNormalWS, inputs [], properties {}}
  MaterialExpressionVertexNormalWS_9 {class=MaterialExpressionVertexNormalWS, inputs [], properties {}}
  MaterialExpressionViewProperty_0 {class=MaterialExpressionViewProperty, inputs [], properties {property=MEVP_TemporalSampleIndex}}
  MaterialExpressionWorldPosition_0 {class=MaterialExpressionWorldPosition, inputs [], properties {worldPositionShaderOffset=WPT_ExcludeAllShaderOffsets}}
  MaterialExpressionWorldPosition_1 {class=MaterialExpressionWorldPosition, inputs [], properties {}}
  MaterialExpressionWorldPosition_2 {class=MaterialExpressionWorldPosition, inputs [], properties {}}
  MaterialExpressionWorldPosition_3 {class=MaterialExpressionWorldPosition, inputs [], properties {worldPositionShaderOffset=WPT_ExcludeAllShaderOffsets}}
  MaterialExpressionWorldPosition_4 {class=MaterialExpressionWorldPosition, inputs [], properties {}}
 }}
