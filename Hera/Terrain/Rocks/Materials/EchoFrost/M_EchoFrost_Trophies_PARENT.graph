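# Parent material for EchoFrost ice trophies, a Nanite-enabled Default Lit surface built on material attributes.
# MF_EchoFrost_IceTrophies supplies the look, with constants in place of baked curvature, AO and painted mask inputs.
# An optional "Is Destructible" switch adds destructible damage before the result passes through MF_CreativeMoveTool.
# A "Depth Fade" static switch multiplies opacity by the saturated, scaled gap between scene depth and pixel depth.
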Material /Hera/Terrain/Rocks/Materials/EchoFrost/M_EchoFrost_Trophies_PARENT
graph {
 properties {
  bAutomaticallySetUsageInEditor=false
  bUseMaterialAttributes=true
  bUsedWithNanite=true
  translucencyLightingMode=TLM_Surface
  usedShadingModels=Default Lit
 }
 nodes [
  n1 = Constant()
  n2 = MaterialFunctionCall!(Input MaterialAttributes, "\\ --- Dependencies --- /", Baked Curvature=@n1, Baked AO=@n1, Painted Mask=@n1) {materialFunction=/Hera/Sets/EchoFrost/Materials/MaterialFunctions/MF_EchoFrost_IceTrophies}
  n3 = StaticBoolParameter() {
   desc="Only enable if this asset can be destroyed\r\n"
   group="-        Rock/Cliff Global Features"
   parameterName=Is Destructible
   sortPriority=-5
  }
  n4 = MaterialFunctionCall(MA In=StaticSwitch(
    True=MaterialFunctionCall(Material_In=@n2.MaterialAttributes) {materialFunction=&MF_FN_DestructibleDamage}.Material_Out
    False=@n2.MaterialAttributes
    Value=@n3
   )) {materialFunction=&MF_CreativeMoveTool}
  n5 = GetMaterialAttributes(None=@n4.MA Out) {attributeGetTypes [#1]}
  n6 = SceneDepth()
  n7 = PixelDepth()
  n8 = Saturate(None=Divide(A=Subtract(A=@n6.R, B=@n7.R)) {constB=40})
  StaticSwitchParameter(
   True=SetMaterialAttributes!(MaterialAttributes=@n5.MaterialAttributes, Opacity=Multiply(A=@n5.Opacity, B=@n8)) {attributeSetTypes [#1]}
   False=@n4.MA Out
  ) {parameterName=Depth Fade}
 ]
}
