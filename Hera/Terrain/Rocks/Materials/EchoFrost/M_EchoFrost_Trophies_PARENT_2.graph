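# Translucent unlit parent material that fakes frosted, glassy ice, apparently for EchoFrost trophies.
# A local-space projection along the view ray forms a fake inner glow that grades frost-masked colors, darkened by a depth fade that also drives opacity.
# Cubemap reflections and bump-offset crystal parallax add surface and interior detail.
# Color and normals pass through MF_FN_DestructibleDamage, then quality-switched brightness and inverse eye adaptation adjust the color.
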Material /Hera/Terrain/Rocks/Materials/EchoFrost/M_EchoFrost_Trophies_PARENT_2
graph {
 properties {
  bAutomaticallySetUsageInEditor=false
  bEnableNewHLSLGeneratorOverridden=true
  blendMode=BLEND_Translucent
  shadingModel=MSM_Unlit
  usedShadingModels=Unlit
 }
 nodes [
  Constant() {r=1}
  n1 = VectorParameter() {parameterName=Color Frost, defaultValue=<rgba 0.4102 0.6466 0.75 1>}
  n2 = VectorParameter() {parameterName=Color 1, defaultValue=<rgba 0.0116 0.0729 0.1406 1>}
  n3 = ScalarParameter() {defaultValue=0.7218, parameterName=Frost Mask Low}
  n4 = ScalarParameter() {defaultValue=2.8744, parameterName=Frost Mask High}
  n5 = NamedRerouteUsage()
  n6 = Power(Base=Saturate(None=LinearInterpolate(A=@n3, B=@n4, Alpha=@n5))) {constExponent=0.5}
  n7 = VectorParameter() {parameterName=Light Offset, defaultValue=<rgba 0 0 56.932 1>}
  n8 = VectorParameter() {parameterName=Light Scale, defaultValue=<rgba 11.3249 7.9795 8.2484 1>}
  n9 = Divide(A=Subtract(A=LocalPosition().XYZ, B=@n7.RGB), B=@n8.RGB)
  n10 = Add(
   A=@n7.RGB
   B=Transform(None=ActorPositionWS() {originType=CameraRelative}) {transformSourceType=TRANSFORMSOURCE_World, transformType=TRANSFORM_Local}
  )
  n11 = DotProduct(A=@n9, B=@n10)
  n12 = Transform(None=WorldPosition() {worldPositionShaderOffset=WPT_CameraRelative}.XYZ) {transformSourceType=TRANSFORMSOURCE_World, transformType=TRANSFORM_Local}
  n13 = DotProduct(A=@n10, B=@n12)
  n14 = Subtract(A=@n9, B=Multiply(A=Divide(A=@n11, B=@n13), B=@n12))
  MaterialExpressionPower_3 = Power(Base=Abs(None=Divide(A=@n14, B=@n8.RGB))) {constExponent=2.3}
  n15 = Convert(None=@MaterialExpressionPower_3) {
   nodeName=Break Float3
   convertInputs [{
     type=Vector3
     expressionInput {
      expression=&self:MaterialExpressionPower_3
      inputName=None
      mask=0
      maskA=0
      maskB=0
      maskG=0
      maskR=0
      outputIndex=0
     }
    }]
   convertOutputs [{type=Scalar}, {type=Scalar}, {type=Scalar}]
  }
  n16 = Add(A=@n15, B=@n15)
  n17 = OneMinus(None=Saturate(None=Add(A=@n16, B=@n15)))
  n18 = Multiply(A=@n17, B=@n17)
  n19 = Multiply(A=@n18, B=@n18)
  n20 = LinearInterpolate(
   A=LinearInterpolate(
    A=LinearInterpolate(A=@n1.RGB, B=@n2.RGB, Alpha=@n6)
    B=VectorParameter() {parameterName=Color 2, defaultValue=<rgba 0 0.153 0.4479 1>}.RGB
    Alpha=@n19
   )
   B=VectorParameter() {parameterName=Color 3, defaultValue=<rgba 0.1065 0.3925 0.6198 1>}.RGB
   Alpha=Power(Base=@n19) {constExponent=4}
  )
  n21 = VertexNormalWS()
  n22 = Abs(None=ComponentMask(None=Transform(None=@n21) {transformSourceType=TRANSFORMSOURCE_World, transformType=TRANSFORM_Local}) {r=true})
  n23 = Divide(
   A=Subtract(A=SceneDepth().R, B=PixelDepth().R)
   B=LinearInterpolate(
    A=ScalarParameter() {defaultValue=175, parameterName=Depth Distance Y}
    B=ScalarParameter() {defaultValue=500, parameterName=Depth Distance X}
    Alpha=Saturate(None=@n22)
   )
  )
  n24 = ScalarParameter() {defaultValue=0.228, parameterName=Depth Offset}
  n25 = OneMinus(None=Saturate(None=Add(A=@n23, B=@n24)))
  n26 = OneMinus(None=Multiply(A=@n25, B=@n25))
  n27 = ReflectionVectorWS(CustomWorldNormal=Transform(None=NamedRerouteUsage()))
  n28 = LinearInterpolate(
   A=ScalarParameter() {defaultValue=-0.2868, parameterName=Ref Mask Low}
   B=ScalarParameter() {defaultValue=2.9278, parameterName=Ref Mask High}
   Alpha=NamedRerouteDeclaration(None=TextureSample() {samplerType=SAMPLERTYPE_LinearColor, texture=/Hera/Props/EchoFrost/EchoFrost_Stand_B/Textures/T_EchoFrost_StandGlassExterior_M}.R) {name=Frost Mask, nodeColor=<rgba 0.7109 1 0 1>}
  )
  n29 = Multiply(
   A=Multiply(
    A=TextureSample(UVs=@n27) {samplerType=SAMPLERTYPE_LinearColor, texture=/Hera/Sets/EchoFrost/Textures/T_EchoFrost_PillarCubeMap}.RGB
    B=Saturate(None=@n28)
   )
   B=VectorParameter() {parameterName=Reflection Color, defaultValue=<rgba 1 1 1 1>}.RGB
  )
  n30 = ScalarParameter() {defaultValue=0.65, parameterName=Reflection Intensity}
  n31 = Add(
   A=Multiply(A=NamedRerouteDeclaration(None=@n20) {name=Final Gradings, nodeColor=<rgba 0 0.25 1 1>}, B=@n26)
   B=Add(A=Multiply(A=@n29, B=@n30) {constB=0.5}, B=NamedRerouteUsage())
  )
  n32 = NamedRerouteUsage()
  n33 = GetMaterialAttributes(None=MaterialFunctionCall(Material_In=SetMaterialAttributes!(MaterialAttributes, Base Color=@n31, Normal=@n32) {attributeSetTypes [#1, #2]}) {materialFunction=&MF_FN_DestructibleDamage}.Material_Out) {attributeGetTypes [#1, #2, #3]}
  n34 = Multiply(A=@n33.Base Color) {constB=1.4}
  EyeAdaptationInverse(
   LightValueInput=QualitySwitch(Default=@n33.Base Color, Low=@n34, High=@n33.Base Color, Medium=@n34, Epic=@n33.Base Color)
   AlphaInput=ScalarParameter() {defaultValue=0.75, parameterName=EyeAdaptationAlpha}
  )
  n35 = Multiply(A=TextureCoordinate(), B=ScalarParameter() {defaultValue=0.35, parameterName=Normals Tiling 1})
  NamedRerouteDeclaration(None=LinearInterpolate(
    A=Constant3Vector() {constant=<rgba 0 0 1 0>}
    B=TextureSample(UVs=@n35) {samplerType=SAMPLERTYPE_Normal, texture=/Hera/Sets/EchoFrost/Textures/T_EchoFrost_IceSurface_Flat_A_N}.RGB
   ) {constAlpha=0.7}) {name=Normals, nodeColor=<rgba 1 0 0.5625 1>}
  n36 = TextureCoordinate()
  n37 = ScalarParameter() {defaultValue=1, parameterName=Parallax Mask Tiling}
  NamedRerouteDeclaration(None=TextureSample(UVs=Multiply(A=@n36, B=@n37)) {samplerType=SAMPLERTYPE_LinearGrayscale, texture=/Game/Global/Textures/Default/Masks_and_Gradients/T_MF_Crystal2}.R) {name=Parallax Mask, nodeColor=<rgba 0.8047 1 0 1>}
  n38 = BumpOffset(
   Coordinate=Multiply(A=TextureCoordinate(), B=ScalarParameter() {defaultValue=1, parameterName=Parallax Pattern Tiling})
   Height=LinearInterpolate(Alpha=NamedRerouteUsage())
  ) {heightRatio=1}
  n39 = Desaturation(
   None=Add(
    A=TextureSample(UVs=@n38) {texture=/Game/Global/Textures/Default/Masks_and_Gradients/T_MF_Crystal}.RGB
    B=ScalarParameter() {defaultValue=-0.1, parameterName=Parallax Color Offset}
   ) {constB=-0.15}
   Fraction=ScalarParameter() {parameterName=Parallax Color Desat}
  )
  n40 = NamedRerouteUsage()
  NamedRerouteDeclaration(None=Multiply(A=Multiply(A=@n39, B=@n40), B=ScalarParameter() {defaultValue=2, parameterName=Parallax Color Intensity})) {name=Parallax Pattern, nodeColor=<rgba 0 0.0859 1 1>}
  NamedRerouteDeclaration(None=@n33.Normal) {name=Cracked Normals, nodeColor=<rgba 0.9219 1 0 1>}
  n41 = LinearInterpolate(
   A=ScalarParameter() {defaultValue=0.2, parameterName=Opacity 1}
   B=ScalarParameter() {defaultValue=1.085, parameterName=Opacity 2}
   Alpha=@n26
  )
  NamedRerouteDeclaration(None=@n41) {name=Opacity, nodeColor=<rgba 0.7578 0 1 1>}
  NamedRerouteUsage()
  Reroute(None=@n41)
 ]
}
